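# Level Sequence for authoring the first-person outro of the cooked base grenade throw on the medium male player character.
# It binds one skeletal mesh actor, animated through a Fortnite_Biped_CtrlRig Control Rig track; its skeletal animation track is empty.
# Body controls carry dense, near per-frame keys over frames 0 to 30, with both FK and IK leg controls keyed.
# FK/IK, space and stretch switches stay constant throughout.
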LevelSequence /Game/Animation/Game/MainPlayer/Combat/Gadgets/Medium/Male/BaseGrenade/BaseGrenadeThrow_Cook_Outro_CMM_1P_2/Driving_BaseGrenadeThrow_Cook_Outro_CMM_1P_2
bindings ["BaseGrenadeThrow_Cook_Outro_CMM_1P --> BaseGrenadeThrow_Cook_Outro_CMM_1P" {
  id=#1
  possessed_class=SkeletalMeshActor
  tracks [
   {
    type=MovieScene3DTransformTrack
    sections [{
      type=MovieScene3DTransformSection
      channels [
       Location.X {type=MovieSceneScriptingDoubleChannel, default=0}
       Location.Y {type=MovieSceneScriptingDoubleChannel, default=0}
       Location.Z {type=MovieSceneScriptingDoubleChannel, default=0}
       Rotation.X {type=MovieSceneScriptingDoubleChannel, default=0}
       Rotation.Y {type=MovieSceneScriptingDoubleChannel, default=0}
       Rotation.Z {type=MovieSceneScriptingDoubleChannel, default=0}
       Scale.X {type=MovieSceneScriptingDoubleChannel, default=1}
       Scale.Y {type=MovieSceneScriptingDoubleChannel, default=1}
       Scale.Z {type=MovieSceneScriptingDoubleChannel, default=1}
      ]
     }]
   }
   {type=MovieSceneSkeletalAnimationTrack, sections []}
   {type=MovieSceneSpawnTrack, sections [{type=MovieSceneSpawnSection, channels [None {type=MovieSceneScriptingBoolChannel, default=true}]}]}
   {
    type=MovieSceneControlRigParameterTrack
    display_name=Fortnite_Biped_CtrlRig
    sections [{
      type=MovieSceneControlRigParameterSection
      channels [
       leg_l_stretch_switch {type=MovieSceneScriptingBoolChannel, default=false, const [false,31]}
       leg_r_stretch_switch {type=MovieSceneScriptingBoolChannel, default=false, const [false,31]}
       arm_l_fk_local {type=MovieSceneScriptingBoolChannel, default=true, const [true,31]}
       Auto Interpolate_3 {type=MovieSceneScriptingBoolChannel, default=true, const [true,31]}
       Auto Interpolate {type=MovieSceneScriptingBoolChannel, default=true, const [true,31]}
       arm_l_stretch_switch {type=MovieSceneScriptingBoolChannel, default=false, const [false,31]}
       Keep Aligned {type=MovieSceneScriptingBoolChannel, default=false, const [false,31]}
       arm_r_fk_local {type=MovieSceneScriptingBoolChannel, default=true, const [true,31]}
       Auto Interpolate_4 {type=MovieSceneScriptingBoolChannel, default=true, const [true,31]}
       Auto Interpolate_2 {type=MovieSceneScriptingBoolChannel, default=true, const [true,31]}
       arm_r_stretch_switch {type=MovieSceneScriptingBoolChannel, default=false, const [false,31]}
       Keep Aligned_2 {type=MovieSceneScriptingBoolChannel, default=false, const [false,31]}
       head_local {type=MovieSceneScriptingBoolChannel, default=true, const [true,31]}
       Stretch {type=MovieSceneScriptingBoolChannel, default=false, const [false,31]}
       head_stretch {type=MovieSceneScriptingBoolChannel, default=false, const [false,31]}
       spine_stretch_switch {type=MovieSceneScriptingBoolChannel, default=false, const [false,31]}
       arm_l_fk_ik_switch {type=MovieSceneScriptingBoolChannel, default=false, const [false,31]}
       leg_l_fk_ik_switch {type=MovieSceneScriptingBoolChannel, default=true, const [true,31]}
       arm_r_fk_ik_switch {type=MovieSceneScriptingBoolChannel, default=false, const [false,31]}
       leg_r_fk_ik_switch {type=MovieSceneScriptingBoolChannel, default=true, const [true,31]}
       ShowBodyControls {type=MovieSceneScriptingBoolChannel, default=true, const [true,31]}
       spine_fk_ik_switch {type=MovieSceneScriptingBoolChannel, default=false, const [false,31]}
       neck_fk_ik_switch {type=MovieSceneScriptingBoolChannel, default=false, const [false,31]}
       space {type=MovieSceneScriptingByteChannel, default=1, const [1,31]}
       global_ctrl.Location.X {type=MovieSceneScriptingFloatChannel, default=0, const [0,31]}
       global_ctrl.Location.Y {type=MovieSceneScriptingFloatChannel, default=0, const [0,31]}
       global_ctrl.Location.Z {type=MovieSceneScriptingFloatChannel, default=0, const [0,31]}
       global_ctrl.Rotation.X {type=MovieSceneScriptingFloatChannel, default=0, const [0,31]}
       global_ctrl.Rotation.Y {type=MovieSceneScriptingFloatChannel, default=0, const [0,31]}
       global_ctrl.Rotation.Z {type=MovieSceneScriptingFloatChannel, default=0, const [0,31]}
       root_ctrl.Location.X {type=MovieSceneScriptingFloatChannel, default=0, const [0,31]}
       root_ctrl.Location.Y {type=MovieSceneScriptingFloatChannel, default=0, const [0,31]}
       root_ctrl.Location.Z {type=MovieSceneScriptingFloatChannel, default=0, const [0,31]}
       root_ctrl.Rotation.X {type=MovieSceneScriptingFloatChannel, default=0, const [0,31]}
       root_ctrl.Rotation.Y {type=MovieSceneScriptingFloatChannel, default=0, const [0,31]}
       root_ctrl.Rotation.Z {type=MovieSceneScriptingFloatChannel, default=0, const [0,31]}
       root_ctrl.Scale.X {type=MovieSceneScriptingFloatChannel, default=1, const [1,31]}
       root_ctrl.Scale.Y {type=MovieSceneScriptingFloatChannel, default=1, const [1,31]}
       root_ctrl.Scale.Z {type=MovieSceneScriptingFloatChannel, default=1, const [1,31]}
       body_offset_ctrl.Location.X {type=MovieSceneScriptingFloatChannel, default=0, const [0,31]}
       body_offset_ctrl.Location.Y {type=MovieSceneScriptingFloatChannel, default=0, const [0,31]}
       body_offset_ctrl.Location.Z {type=MovieSceneScriptingFloatChannel, default=0, const [0,31]}
       body_offset_ctrl.Rotation.X {type=MovieSceneScriptingFloatChannel, default=0, const [0,31]}
       body_offset_ctrl.Rotation.Y {type=MovieSceneScriptingFloatChannel, default=0, const [0,31]}
       body_offset_ctrl.Rotation.Z {type=MovieSceneScriptingFloatChannel, default=0, const [0,31]}
       body_offset_ctrl.Scale.X {type=MovieSceneScriptingFloatChannel, default=1, const [1,31]}
       body_offset_ctrl.Scale.Y {type=MovieSceneScriptingFloatChannel, default=1, const [1,31]}
       body_offset_ctrl.Scale.Z {type=MovieSceneScriptingFloatChannel, default=1, const [1,31]}
       thigh_l_fk_ctrl.Location.X {type=MovieSceneScriptingFloatChannel, default=0, thinned_from=31, keys [[0,0],[30,0]]}
       thigh_l_fk_ctrl.Location.Y {type=MovieSceneScriptingFloatChannel, default=0, thinned_from=31, keys [[0,0],[30,0]]}
       thigh_l_fk_ctrl.Location.Z {type=MovieSceneScriptingFloatChannel, default=0, thinned_from=31, keys [[0,0],[30,0]]}
       thigh_l_fk_ctrl.Rotation.X {
        type=MovieSceneScriptingFloatChannel
        default=17.8858
        thinned_from=31
        keys [[0,17.8858],[1,18.3068],[2,16.8088],[3,13.7114],[4,9.6419],[5,5.2796],[6,0.9973],[7,-2.3393],[8,-4.6911],[9,-6.0991],[10,-6.5004],[11,-6.6936],[12,-6.8168],[13,-6.8125],[14,-6.405],[15,-5.3736],[16,-3.4861],[17,-1.0615],[18,1.2693],[19,3.2002],[20,4.4169],[21,5.1101],[22,5.6229],[23,5.982],[24,6.0592],[25,5.97],[26,5.8306],[27,5.7578],[30,5.7578]]
       }
       thigh_l_fk_ctrl.Rotation.Y {
        type=MovieSceneScriptingFloatChannel
        default=8.2844
        thinned_from=31
        keys [[0,8.2844],[1,9.703],[2,10.2604],[3,9.8322],[4,8.5593],[5,6.6675],[6,4.4418],[7,2.5358],[8,1.1529],[9,0.3266],[10,-0.0839],[11,-0.4062],[12,-0.579],[13,-0.4421],[14,0.1034],[15,1.0417],[16,2.372],[17,4.3319],[18,6.6666],[19,8.4629],[20,9.2338],[21,9.3641],[22,9.1007],[23,8.7249],[24,8.3721],[25,8.076],[26,7.8719],[27,7.796],[30,7.796]]
       }
       thigh_l_fk_ctrl.Rotation.Z {
        type=MovieSceneScriptingFloatChannel
        default=-17.0181
        thinned_from=31
        keys [[0,-17.0181],[1,-16.0069],[2,-16.4912],[3,-18.2437],[4,-21.1356],[5,-25.0574],[6,-29.9638],[7,-34.3479],[8,-37.5586],[9,-40.0659],[10,-41.1725],[11,-40.9413],[13,-39.7716],[14,-38.9841],[15,-37.8028],[16,-36.0464],[17,-33.7285],[18,-31.1504],[19,-28.7045],[20,-26.5569],[21,-24.6337],[22,-23.1035],[23,-21.8815],[24,-20.8195],[25,-19.9815],[26,-19.4317],[27,-19.2344],[30,-19.2344]]
       }
       thigh_l_fk_ctrl.Scale.X {type=MovieSceneScriptingFloatChannel, default=1, const [1,31]}
       thigh_l_fk_ctrl.Scale.Y {type=MovieSceneScriptingFloatChannel, default=1, const [1,31]}
       thigh_l_fk_ctrl.Scale.Z {type=MovieSceneScriptingFloatChannel, default=1, const [1,31]}
       calf_l_fk_ctrl.Location.X {
        type=MovieSceneScriptingFloatChannel
        default=0.1087
        thinned_from=31
        keys [[0,0.1087],[1,0.1147],[2,0.1175],[3,0.1168],[4,0.1133],[5,0.1069],[6,0.0974],[7,0.0854],[8,0.0731],[9,0.0646],[10,0.0618],[11,0.0604],[12,0.0585],[13,0.0559],[14,0.0538],[15,0.0532],[16,0.0545],[17,0.0602],[18,0.068],[19,0.0736],[20,0.076],[21,0.0767],[22,0.0765],[23,0.076],[26,0.0741],[27,0.0738],[30,0.0738]]
       }
       calf_l_fk_ctrl.Location.Y {
        type=MovieSceneScriptingFloatChannel
        default=0.0077
        thinned_from=31
        keys [[0,0.0077],[1,0.0098],[2,0.0108],[3,0.0106],[4,0.0093],[5,0.0071],[7,0.0013],[8,-0.001],[9,-0.0021],[14,-0.0031],[16,-0.0031],[17,-0.0026],[19,-0.0009],[20,-0.0005],[30,-0.0009]]
       }
       calf_l_fk_ctrl.Location.Z {
        type=MovieSceneScriptingFloatChannel
        default=0.0037
        thinned_from=31
        keys [[0,0.0037],[2,0.004],[5,0.0036],[9,0.0023],[16,0.0019],[20,0.0027],[30,0.0026]]
       }
       calf_l_fk_ctrl.Rotation.X {
        type=MovieSceneScriptingFloatChannel
        default=0.9232
        thinned_from=31
        keys [[0,0.9232],[1,0.968],[2,0.9887],[3,0.9841],[4,0.9574],[5,0.9098],[6,0.8368],[7,0.7429],[8,0.6439],[9,0.5739],[10,0.551],[11,0.5392],[12,0.5235],[13,0.5012],[14,0.484],[15,0.4783],[16,0.49],[17,0.5372],[18,0.602],[19,0.6484],[20,0.6676],[21,0.673],[22,0.6717],[23,0.6678],[25,0.6564],[26,0.6518],[27,0.6499],[30,0.6499]]
       }
       calf_l_fk_ctrl.Rotation.Y {
        type=MovieSceneScriptingFloatChannel
        default=0.212
        thinned_from=31
        keys [[0,0.212],[1,0.2156],[2,0.217],[3,0.2167],[4,0.2148],[5,0.2108],[6,0.2031],[7,0.1907],[8,0.1746],[9,0.1614],[10,0.1568],[11,0.1543],[12,0.151],[13,0.1461],[14,0.1423],[15,0.141],[16,0.1436],[17,0.1539],[18,0.1669],[19,0.1754],[20,0.1787],[21,0.1796],[22,0.1794],[26,0.176],[30,0.1757]]
       }
       calf_l_fk_ctrl.Rotation.Z {
        type=MovieSceneScriptingFloatChannel
        default=26.045
        thinned_from=31
        keys [[0,26.045],[1,27.5327],[2,28.233],[3,28.0752],[4,27.1785],[5,25.6041],[6,23.2625],[7,20.3526],[8,17.3913],[9,15.3608],[10,14.7056],[11,14.3696],[12,13.9233],[13,13.2938],[14,12.8115],[15,12.6515],[16,12.9797],[17,14.3117],[18,16.1714],[19,17.5256],[20,18.0916],[21,18.2526],[22,18.2129],[23,18.0982],[25,17.7601],[26,17.6247],[27,17.5704],[30,17.5704]]
       }
       calf_l_fk_ctrl.Scale.X {type=MovieSceneScriptingFloatChannel, default=1, const [1,31]}
       calf_l_fk_ctrl.Scale.Y {type=MovieSceneScriptingFloatChannel, default=1, const [1,31]}
       calf_l_fk_ctrl.Scale.Z {type=MovieSceneScriptingFloatChannel, default=1, const [1,31]}
       foot_l_fk_ctrl.Location.X {type=MovieSceneScriptingFloatChannel, default=0, thinned_from=31, keys [[0,0],[30,0]]}
       foot_l_fk_ctrl.Location.Y {type=MovieSceneScriptingFloatChannel, default=0, thinned_from=31, keys [[0,0],[30,0]]}
       foot_l_fk_ctrl.Location.Z {type=MovieSceneScriptingFloatChannel, default=0, thinned_from=31, keys [[0,0],[30,0]]}
       foot_l_fk_ctrl.Rotation.X {
        type=MovieSceneScriptingFloatChannel
        default=-2.4949
        thinned_from=31
        keys [[0,-2.4949],[1,-2.1644],[2,-2.4036],[3,-3.0314],[4,-3.9077],[5,-4.8986],[6,-6.2402],[7,-8.0895],[8,-10.0485],[9,-11.6844],[10,-12.7814],[11,-13.3759],[12,-13.5475],[13,-13.4187],[14,-13.1017],[15,-12.6644],[16,-12.1781],[17,-11.483],[19,-9.5577],[20,-8.5689],[21,-7.5528],[22,-6.722],[24,-5.3522],[25,-4.7637],[26,-4.3497],[27,-4.1932],[30,-4.1932]]
       }
       foot_l_fk_ctrl.Rotation.Y {
        type=MovieSceneScriptingFloatChannel
        default=-4.9183
        thinned_from=31
        keys [[0,-4.9183],[1,-5.0789],[2,-5.0294],[3,-4.8502],[4,-4.6597],[5,-4.5665],[6,-4.6634],[8,-5.0172],[9,-5.1485],[12,-5.26],[14,-5.3002],[15,-5.2892],[17,-5.1843],[18,-5.1666],[19,-5.1199],[21,-4.9358],[22,-4.8275],[23,-4.7533],[24,-4.7416],[26,-4.7973],[27,-4.8137],[30,-4.8137]]
       }
       foot_l_fk_ctrl.Rotation.Z {
        type=MovieSceneScriptingFloatChannel
        default=-0.8631
        thinned_from=31
        keys [[0,-0.8631],[1,-2.2499],[2,-2.8933],[3,-2.8779],[4,-2.2979],[5,-1.2252],[6,0.4698],[7,2.599],[8,4.7651],[9,6.419],[10,7.2458],[12,8.3734],[13,9.0026],[14,9.5268],[15,9.8018],[16,9.687],[17,8.8964],[18,7.7341],[19,6.8215],[20,6.3389],[21,6.0645],[22,5.9034],[23,5.7882],[24,5.7094],[25,5.6605],[26,5.635],[30,5.6274]]
       }
       foot_l_fk_ctrl.Scale.X {type=MovieSceneScriptingFloatChannel, default=1, const [1,31]}
       foot_l_fk_ctrl.Scale.Y {type=MovieSceneScriptingFloatChannel, default=1, const [1,31]}
       foot_l_fk_ctrl.Scale.Z {type=MovieSceneScriptingFloatChannel, default=1, const [1,31]}
       ball_l_fk_ctrl.Location.X {type=MovieSceneScriptingFloatChannel, default=0, const [0,31]}
       ball_l_fk_ctrl.Location.Y {type=MovieSceneScriptingFloatChannel, default=0, const [0,31]}
       ball_l_fk_ctrl.Location.Z {type=MovieSceneScriptingFloatChannel, default=0, const [0,31]}
       ball_l_fk_ctrl.Rotation.X {type=MovieSceneScriptingFloatChannel, default=0, const [0,31]}
       ball_l_fk_ctrl.Rotation.Y {type=MovieSceneScriptingFloatChannel, default=0, const [0,31]}
       ball_l_fk_ctrl.Rotation.Z {type=MovieSceneScriptingFloatChannel, default=0, const [0,31]}
       ball_l_fk_ctrl.Scale.X {type=MovieSceneScriptingFloatChannel, default=1, const [1,31]}
       ball_l_fk_ctrl.Scale.Y {type=MovieSceneScriptingFloatChannel, default=1, const [1,31]}
       ball_l_fk_ctrl.Scale.Z {type=MovieSceneScriptingFloatChannel, default=1, const [1,31]}
       foot_l_ik_ctrl.Location.X {type=MovieSceneScriptingFloatChannel, default=6.416, thinned_from=31, keys [[0,6.416],[30,6.416]]}
       foot_l_ik_ctrl.Location.Y {type=MovieSceneScriptingFloatChannel, default=21.9206, thinned_from=31, keys [[0,21.9206],[30,21.9206]]}
       foot_l_ik_ctrl.Location.Z {
        type=MovieSceneScriptingFloatChannel
        default=-0.0001
        thinned_from=31
        keys [[0,-0.0002],[8,0],[11,0.0003],[12,-0.0001],[16,-0.0001],[17,0.0003],[23,-0.0001],[24,0],[25,-0.0003],[26,-0.0001],[30,-0.0001]]
       }
       foot_l_ik_ctrl.Rotation.X {
        type=MovieSceneScriptingFloatChannel
        default=0.027
        thinned_from=31
        keys [[0,0.027],[4,0.0269],[7,0.027],[9,0.0277],[10,0.0277],[11,0.0281],[12,0.027],[16,0.027],[17,0.0281],[18,0.0275],[20,0.0275],[21,0.0272],[22,0.0272],[23,0.0274],[24,0.0271],[25,0.0275],[30,0.0274]]
       }
       foot_l_ik_ctrl.Rotation.Y {type=MovieSceneScriptingFloatChannel, default=0.1388, thinned_from=31, keys [[0,0.1388],[24,0.1389],[25,0.1383],[26,0.1389],[30,0.1389]]}
       foot_l_ik_ctrl.Rotation.Z {type=MovieSceneScriptingFloatChannel, default=10.5353, thinned_from=31, keys [[0,10.5353],[30,10.5358]]}
       foot_l_ik_ctrl.Scale.X {type=MovieSceneScriptingFloatChannel, default=1, const [1,31]}
       foot_l_ik_ctrl.Scale.Y {type=MovieSceneScriptingFloatChannel, default=1, const [1,31]}
       foot_l_ik_ctrl.Scale.Z {type=MovieSceneScriptingFloatChannel, default=1, const [1,31]}
       ball_l_ik_ctrl.Location.X {type=MovieSceneScriptingFloatChannel, default=0, thinned_from=31, keys [[0,0],[4,-0.0001],[26,0],[30,0.0001]]}
       ball_l_ik_ctrl.Location.Y {type=MovieSceneScriptingFloatChannel, default=0, thinned_from=31, keys [[0,0],[30,0]]}
       ball_l_ik_ctrl.Location.Z {type=MovieSceneScriptingFloatChannel, default=0, thinned_from=31, keys [[0,0],[30,0]]}
       ball_l_ik_ctrl.Rotation.X {type=MovieSceneScriptingFloatChannel, default=0, thinned_from=31, keys [[0,0],[30,-0.0001]]}
       ball_l_ik_ctrl.Rotation.Y {type=MovieSceneScriptingFloatChannel, default=0, thinned_from=31, keys [[0,0],[30,0]]}
       ball_l_ik_ctrl.Rotation.Z {type=MovieSceneScriptingFloatChannel, default=0, thinned_from=31, keys [[0,0],[26,0],[27,-0.0004],[30,-0.0004]]}
       ball_l_ik_ctrl.Scale.X {type=MovieSceneScriptingFloatChannel, default=1, const [1,31]}
       ball_l_ik_ctrl.Scale.Y {type=MovieSceneScriptingFloatChannel, default=1, const [1,31]}
       ball_l_ik_ctrl.Scale.Z {type=MovieSceneScriptingFloatChannel, default=1, const [1,31]}
       foot_bk1_l_ctrl.Location.X {type=MovieSceneScriptingFloatChannel, default=0, const [0,31]}
       foot_bk1_l_ctrl.Location.Y {type=MovieSceneScriptingFloatChannel, default=0, const [0,31]}
       foot_bk1_l_ctrl.Location.Z {type=MovieSceneScriptingFloatChannel, default=0, const [0,31]}
       foot_bk1_l_ctrl.Rotation.X {type=MovieSceneScriptingFloatChannel, default=0, const [0,31]}
       foot_bk1_l_ctrl.Rotation.Y {type=MovieSceneScriptingFloatChannel, default=0, const [0,31]}
       foot_bk1_l_ctrl.Rotation.Z {type=MovieSceneScriptingFloatChannel, default=0, const [0,31]}
       foot_bk1_l_ctrl.Scale.X {type=MovieSceneScriptingFloatChannel, default=1, const [1,31]}
       foot_bk1_l_ctrl.Scale.Y {type=MovieSceneScriptingFloatChannel, default=1, const [1,31]}
       foot_bk1_l_ctrl.Scale.Z {type=MovieSceneScriptingFloatChannel, default=1, const [1,31]}
       foot_roll_l_ctrl.Location.X {type=MovieSceneScriptingFloatChannel, default=0, const [0,31]}
       foot_roll_l_ctrl.Location.Y {type=MovieSceneScriptingFloatChannel, default=0, const [0,31]}
       foot_roll_l_ctrl.Location.Z {type=MovieSceneScriptingFloatChannel, default=0, const [0,31]}
       foot_roll_l_ctrl.Rotation.X {type=MovieSceneScriptingFloatChannel, default=0, const [0,31]}
       foot_roll_l_ctrl.Rotation.Y {type=MovieSceneScriptingFloatChannel, default=0, const [0,31]}
       foot_roll_l_ctrl.Rotation.Z {type=MovieSceneScriptingFloatChannel, default=0, const [0,31]}
       foot_roll_l_ctrl.Scale.X {type=MovieSceneScriptingFloatChannel, default=1, const [1,31]}
       foot_roll_l_ctrl.Scale.Y {type=MovieSceneScriptingFloatChannel, default=1, const [1,31]}
       foot_roll_l_ctrl.Scale.Z {type=MovieSceneScriptingFloatChannel, default=1, const [1,31]}
       roll_blend_l_ctrl {type=MovieSceneScriptingFloatChannel, default=1, const [1,31]}
       tip_l_ctrl.Location.X {type=MovieSceneScriptingFloatChannel, default=0, const [0,31]}
       tip_l_ctrl.Location.Y {type=MovieSceneScriptingFloatChannel, default=0, const [0,31]}
       tip_l_ctrl.Location.Z {type=MovieSceneScriptingFloatChannel, default=0, const [0,31]}
       tip_l_ctrl.Rotation.X {type=MovieSceneScriptingFloatChannel, default=0, const [0,31]}
       tip_l_ctrl.Rotation.Y {type=MovieSceneScriptingFloatChannel, default=0, const [0,31]}
       tip_l_ctrl.Rotation.Z {type=MovieSceneScriptingFloatChannel, default=0, const [0,31]}
       tip_l_ctrl.Scale.X {type=MovieSceneScriptingFloatChannel, default=1, const [1,31]}
       tip_l_ctrl.Scale.Y {type=MovieSceneScriptingFloatChannel, default=1, const [1,31]}
       tip_l_ctrl.Scale.Z {type=MovieSceneScriptingFloatChannel, default=1, const [1,31]}
       heel_l_ctrl.Location.X {type=MovieSceneScriptingFloatChannel, default=0, const [0,31]}
       heel_l_ctrl.Location.Y {type=MovieSceneScriptingFloatChannel, default=0, const [0,31]}
       heel_l_ctrl.Location.Z {type=MovieSceneScriptingFloatChannel, default=0, const [0,31]}
       heel_l_ctrl.Rotation.X {type=MovieSceneScriptingFloatChannel, default=0, const [0,31]}
       heel_l_ctrl.Rotation.Y {type=MovieSceneScriptingFloatChannel, default=0, const [0,31]}
       heel_l_ctrl.Rotation.Z {type=MovieSceneScriptingFloatChannel, default=0, const [0,31]}
       heel_l_ctrl.Scale.X {type=MovieSceneScriptingFloatChannel, default=1, const [1,31]}
       heel_l_ctrl.Scale.Y {type=MovieSceneScriptingFloatChannel, default=1, const [1,31]}
       heel_l_ctrl.Scale.Z {type=MovieSceneScriptingFloatChannel, default=1, const [1,31]}
       leg_l_softness {type=MovieSceneScriptingFloatChannel, default=0, const [0,31]}
       leg_l_pv_ik_ctrl.Location.X {
        type=MovieSceneScriptingFloatChannel
        default=-1.6016
        thinned_from=31
        keys [[0,-1.6016],[1,-1.9083],[2,-1.6836],[3,-1.0776],[4,-0.2451],[5,0.6697],[6,1.8271],[7,3.3564],[8,4.9106],[9,6.138],[10,6.9252],[11,7.3157],[12,7.3829],[13,7.2192],[14,6.9152],[16,6.1531],[17,5.6151],[18,4.8592],[20,3.321],[21,2.5413],[22,1.9201],[23,1.408],[24,0.8902],[25,0.4358],[26,0.1133],[27,-0.009],[30,-0.009]]
       }
       leg_l_pv_ik_ctrl.Location.Y {
        type=MovieSceneScriptingFloatChannel
        default=8.813
        thinned_from=31
        keys [[0,8.813],[1,9.2383],[2,9.4341],[3,9.383],[4,9.1097],[5,8.6219],[6,7.8647],[7,6.8488],[8,5.7325],[9,4.8911],[10,4.5398],[11,4.356],[12,4.1946],[13,4.0148],[14,3.9085],[15,3.9213],[16,4.097],[17,4.6276],[18,5.3465],[19,5.8802],[20,6.1453],[21,6.2698],[22,6.3065],[23,6.3047],[26,6.2164],[27,6.2031],[30,6.2031]]
       }
       leg_l_pv_ik_ctrl.Location.Z {
        type=MovieSceneScriptingFloatChannel
        default=10.5925
        thinned_from=31
        keys [[0,10.5925],[1,11.239],[2,11.5797],[3,11.564],[4,11.233],[5,10.6067],[6,9.6697],[8,7.4124],[9,6.6738],[10,6.4777],[11,6.3766],[12,6.1974],[13,5.9195],[14,5.6894],[15,5.5855],[16,5.6824],[17,6.182],[18,6.8888],[19,7.3889],[20,7.5564],[21,7.55],[22,7.4733],[23,7.3751],[25,7.138],[26,7.0497],[27,7.0149],[30,7.0149]]
       }
       leg_l_pv_ik_ctrl.Rotation.X {
        type=MovieSceneScriptingFloatChannel
        default=-13.485
        thinned_from=31
        keys [[0,-13.485],[1,-14.2355],[2,-14.6266],[3,-14.6056],[4,-14.2205],[5,-13.4906],[6,-12.388],[8,-9.6691],[9,-8.7578],[10,-8.5122],[11,-8.3854],[12,-8.1623],[13,-7.8163],[14,-7.5292],[15,-7.4001],[16,-7.5236],[17,-8.1505],[18,-9.0283],[19,-9.6442],[20,-9.8517],[21,-9.8475],[22,-9.757],[23,-9.6399],[25,-9.3551],[26,-9.2486],[27,-9.2067],[30,-9.2067]]
       }
       leg_l_pv_ik_ctrl.Rotation.Y {
        type=MovieSceneScriptingFloatChannel
        default=0.0227
        thinned_from=31
        keys [[0,0.0227],[1,0.0259],[2,0.0477],[3,0.0775],[4,0.1081],[5,0.1342],[6,0.1644],[7,0.2085],[8,0.2556],[9,0.2954],[10,0.323],[11,0.3363],[12,0.3373],[13,0.3296],[14,0.3171],[16,0.288],[17,0.2705],[18,0.247],[19,0.2213],[20,0.1906],[21,0.1557],[22,0.1258],[23,0.0999],[24,0.0728],[25,0.0482],[26,0.0304],[27,0.0236],[30,0.0236]]
       }
       leg_l_pv_ik_ctrl.Rotation.Z {
        type=MovieSceneScriptingFloatChannel
        default=1.8796
        thinned_from=31
        keys [[0,1.8796],[1,2.2196],[2,1.9653],[3,1.2875],[4,0.3506],[5,-0.6955],[6,-2.0521],[7,-3.8983],[8,-5.8483],[9,-7.4309],[10,-8.4333],[11,-8.9357],[12,-9.0475],[13,-8.8846],[14,-8.5363],[15,-8.0755],[16,-7.5783],[17,-6.8441],[18,-5.8374],[19,-4.8574],[21,-2.9908],[22,-2.2542],[23,-1.6486],[24,-1.0363],[25,-0.4985],[26,-0.1163],[27,0.0288],[30,0.0288]]
       }
       leg_l_pv_ik_ctrl.Scale.X {type=MovieSceneScriptingFloatChannel, default=1, const [1,31]}
       leg_l_pv_ik_ctrl.Scale.Y {type=MovieSceneScriptingFloatChannel, default=1, const [1,31]}
       leg_l_pv_ik_ctrl.Scale.Z {type=MovieSceneScriptingFloatChannel, default=1, const [1,31]}
       thigh_r_fk_ctrl.Location.X {type=MovieSceneScriptingFloatChannel, default=0, thinned_from=31, keys [[0,0],[30,0]]}
       thigh_r_fk_ctrl.Location.Y {type=MovieSceneScriptingFloatChannel, default=0, thinned_from=31, keys [[0,0],[30,0]]}
       thigh_r_fk_ctrl.Location.Z {type=MovieSceneScriptingFloatChannel, default=0, thinned_from=31, keys [[0,0],[30,0]]}
       thigh_r_fk_ctrl.Rotation.X {
        type=MovieSceneScriptingFloatChannel
        default=4.2325
        thinned_from=31
        keys [[0,4.2325],[1,0.9081],[2,-0.4647],[3,0.5975],[4,3.1849],[5,6.348],[6,9.0494],[7,11.0623],[8,12.6854],[9,13.9772],[10,14.6097],[11,14.7595],[12,15.2631],[13,16.1706],[14,16.874],[15,17.1972],[16,16.9457],[17,16.7898],[18,17.1973],[19,17.7797],[20,18.7483],[21,19.9974],[22,20.8125],[23,21.0792],[24,21.1939],[25,21.2124],[30,21.168]]
       }
       thigh_r_fk_ctrl.Rotation.Y {
        type=MovieSceneScriptingFloatChannel
        default=14.5507
        thinned_from=31
        keys [[0,14.5507],[1,12.7887],[2,11.2889],[3,10.3435],[4,9.7145],[5,9.2411],[6,8.6895],[7,8.0713],[8,7.6067],[9,7.3754],[10,7.4876],[11,7.7501],[12,8.0354],[13,8.2305],[14,8.3248],[15,8.3464],[16,8.3111],[17,7.95],[18,7.2362],[19,6.7084],[20,6.7373],[21,7.1077],[22,7.6872],[23,8.2334],[24,8.6237],[25,8.8823],[26,9.0273],[27,9.0727],[30,9.0727]]
       }
       thigh_r_fk_ctrl.Rotation.Z {
        type=MovieSceneScriptingFloatChannel
        default=-5.2965
        thinned_from=31
        keys [[0,-5.2965],[1,-4.5975],[2,-3.3586],[3,-1.6625],[4,-0.1491],[5,0.4476],[6,-0.8905],[7,-3.4582],[8,-6.2851],[9,-9.0767],[10,-10.6269],[11,-10.9437],[13,-10.5526],[14,-10.299],[15,-9.8679],[16,-9.1634],[17,-7.5028],[18,-4.9846],[19,-2.7055],[20,-1.0694],[21,0.2515],[22,1.2041],[23,1.9768],[24,2.7254],[25,3.3673],[26,3.8171],[27,3.9868],[30,3.9868]]
       }
       thigh_r_fk_ctrl.Scale.X {type=MovieSceneScriptingFloatChannel, default=1, const [1,31]}
       thigh_r_fk_ctrl.Scale.Y {type=MovieSceneScriptingFloatChannel, default=1, const [1,31]}
       thigh_r_fk_ctrl.Scale.Z {type=MovieSceneScriptingFloatChannel, default=1, const [1,31]}
       calf_r_fk_ctrl.Location.X {
        type=MovieSceneScriptingFloatChannel
        default=0.1146
        thinned_from=31
        keys [[0,0.1146],[1,0.1142],[2,0.1119],[3,0.1064],[4,0.0986],[5,0.09],[6,0.0864],[7,0.0896],[8,0.0973],[9,0.1068],[10,0.1152],[11,0.1221],[12,0.1254],[13,0.1258],[14,0.1251],[15,0.1233],[16,0.1208],[17,0.1142],[18,0.103],[19,0.0926],[20,0.0864],[21,0.0828],[22,0.0805],[24,0.0765],[25,0.0748],[26,0.0736],[27,0.0732],[30,0.0732]]
       }
       calf_r_fk_ctrl.Location.Y {
        type=MovieSceneScriptingFloatChannel
        default=0.0095
        thinned_from=31
        keys [[0,0.0095],[1,0.0094],[2,0.0086],[3,0.0067],[5,0.0022],[6,0.0014],[7,0.0021],[8,0.004],[10,0.0098],[11,0.0124],[12,0.0138],[13,0.014],[14,0.0136],[15,0.0129],[16,0.0119],[17,0.0094],[18,0.0057],[19,0.0028],[20,0.0014],[21,0.0006],[25,-0.0008],[30,-0.0011]]
       }
       calf_r_fk_ctrl.Location.Z {
        type=MovieSceneScriptingFloatChannel
        default=0.0038
        thinned_from=31
        keys [[0,0.0038],[3,0.0036],[6,0.003],[12,0.0042],[16,0.004],[20,0.003],[30,0.0025]]
       }
       calf_r_fk_ctrl.Rotation.X {
        type=MovieSceneScriptingFloatChannel
        default=0.9558
        thinned_from=31
        keys [[0,0.9558],[1,0.9525],[2,0.9357],[3,0.8949],[4,0.8361],[5,0.7695],[6,0.7419],[7,0.767],[8,0.8261],[9,0.8973],[10,0.9601],[11,1.0101],[12,1.0343],[13,1.0372],[14,1.0317],[15,1.0191],[16,1.0006],[17,0.9527],[18,0.8695],[19,0.7898],[20,0.7419],[21,0.7132],[22,0.6955],[24,0.6638],[25,0.6502],[26,0.6406],[27,0.637],[30,0.637]]
       }
       calf_r_fk_ctrl.Rotation.Y {
        type=MovieSceneScriptingFloatChannel
        default=0.2158
        thinned_from=31
        keys [[0,0.2158],[1,0.2155],[2,0.2141],[3,0.2104],[4,0.204],[5,0.1954],[6,0.1914],[7,0.195],[9,0.2106],[10,0.2161],[11,0.2195],[12,0.2207],[14,0.2206],[16,0.2189],[17,0.2155],[18,0.2078],[19,0.1982],[20,0.1914],[21,0.187],[24,0.1788],[25,0.1765],[26,0.1748],[27,0.1741],[30,0.1741]]
       }
       calf_r_fk_ctrl.Rotation.Z {
        type=MovieSceneScriptingFloatChannel
        default=27.5152
        thinned_from=31
        keys [[0,27.5152],[1,27.4052],[2,26.8357],[3,25.4741],[4,23.5602],[5,21.4485],[6,20.5893],[7,21.3692],[8,23.238],[9,25.5549],[10,27.6644],[11,29.3944],[12,30.2495],[13,30.3507],[14,30.1568],[15,29.7101],[16,29.0622],[17,27.4113],[18,24.6392],[19,22.0833],[20,20.5903],[21,19.7075],[22,19.1687],[24,18.2094],[25,17.8026],[26,17.518],[27,17.411],[30,17.411]]
       }
       calf_r_fk_ctrl.Scale.X {type=MovieSceneScriptingFloatChannel, default=1, const [1,31]}
       calf_r_fk_ctrl.Scale.Y {type=MovieSceneScriptingFloatChannel, default=1, const [1,31]}
       calf_r_fk_ctrl.Scale.Z {type=MovieSceneScriptingFloatChannel, default=1, const [1,31]}
       foot_r_fk_ctrl.Location.X {type=MovieSceneScriptingFloatChannel, default=0, thinned_from=31, keys [[0,0],[30,0]]}
       foot_r_fk_ctrl.Location.Y {type=MovieSceneScriptingFloatChannel, default=0, thinned_from=31, keys [[0,0],[30,0]]}
       foot_r_fk_ctrl.Location.Z {type=MovieSceneScriptingFloatChannel, default=0, thinned_from=31, keys [[0,0],[30,0]]}
       foot_r_fk_ctrl.Rotation.X {
        type=MovieSceneScriptingFloatChannel
        default=-0.8911
        thinned_from=31
        keys [[0,-0.8911],[1,1.4402],[2,3.6634],[3,4.6799],[4,4.4991],[5,3.1599],[6,1.5865],[7,0.299],[8,-1.2025],[9,-2.3134],[10,-2.8917],[11,-3.3667],[12,-3.6145],[13,-3.6876],[14,-3.6349],[15,-3.3151],[16,-2.5804],[17,-2.1951],[18,-2.498],[19,-2.7432],[20,-2.7033],[21,-2.6447],[23,-2.5722],[24,-2.5089],[25,-2.4386],[26,-2.3817],[27,-2.3584],[30,-2.3584]]
       }
       foot_r_fk_ctrl.Rotation.Y {
        type=MovieSceneScriptingFloatChannel
        default=-9.677
        thinned_from=31
        keys [[0,-9.6771],[1,-9.1077],[2,-8.7774],[3,-8.4945],[4,-8.0582],[5,-7.2441],[6,-5.7253],[7,-3.7221],[8,-1.361],[9,0.2792],[10,0.6222],[11,0.6044],[12,0.1833],[13,-0.5919],[14,-1.5919],[15,-2.7098],[16,-3.816],[17,-4.8137],[18,-5.6953],[19,-6.4637],[20,-7.1862],[21,-7.8087],[22,-8.1943],[23,-8.3877],[24,-8.5226],[25,-8.6117],[26,-8.6625],[30,-8.6788]]
       }
       foot_r_fk_ctrl.Rotation.Z {
        type=MovieSceneScriptingFloatChannel
        default=-16.3757
        thinned_from=31
        keys [[0,-16.3757],[1,-16.0153],[2,-15.5172],[3,-14.7823],[4,-13.9778],[5,-13.3074],[6,-13.1876],[7,-13.4834],[8,-13.9413],[10,-15.15],[11,-15.6915],[12,-16.1636],[13,-16.5604],[14,-16.8589],[15,-17.0481],[16,-17.1243],[17,-17.1039],[18,-16.8386],[19,-16.2829],[20,-15.5446],[21,-14.7654],[22,-14.1902],[23,-13.8068],[24,-13.4692],[25,-13.2013],[26,-13.0252],[27,-12.9622],[30,-12.9622]]
       }
       foot_r_fk_ctrl.Scale.X {type=MovieSceneScriptingFloatChannel, default=1, const [1,31]}
       foot_r_fk_ctrl.Scale.Y {type=MovieSceneScriptingFloatChannel, default=1, const [1,31]}
       foot_r_fk_ctrl.Scale.Z {type=MovieSceneScriptingFloatChannel, default=1, const [1,31]}
       ball_r_fk_ctrl.Location.X {type=MovieSceneScriptingFloatChannel, default=0, thinned_from=31, keys [[0,0],[30,0]]}
       ball_r_fk_ctrl.Location.Y {type=MovieSceneScriptingFloatChannel, default=0, thinned_from=31, keys [[0,0],[30,0]]}
       ball_r_fk_ctrl.Location.Z {type=MovieSceneScriptingFloatChannel, default=0, thinned_from=31, keys [[0,0],[30,0]]}
       ball_r_fk_ctrl.Rotation.X {
        type=MovieSceneScriptingFloatChannel
        default=0.1796
        thinned_from=31
        keys [[0,0.1796],[1,0.4748],[2,0.7118],[3,0.966],[4,1.2681],[5,1.6502],[6,2.2394],[7,2.9515],[8,3.7557],[9,4.2938],[10,4.3509],[11,4.2758],[12,4.0043],[13,3.5397],[14,2.9757],[15,2.3922],[16,1.86],[17,1.3642],[18,0.9172],[19,0.5834],[20,0.3383],[21,0.1549],[22,0.0614],[23,0.031],[24,0.013],[25,0.0039],[26,0],[30,0]]
       }
       ball_r_fk_ctrl.Rotation.Y {
        type=MovieSceneScriptingFloatChannel
        default=-0.4639
        thinned_from=31
        keys [[0,-0.4639],[1,-1.2335],[2,-1.8461],[3,-2.4152],[4,-2.9906],[5,-3.6201],[6,-4.4256],[7,-5.3016],[8,-6.3414],[9,-6.9316],[10,-6.7886],[11,-6.4906],[12,-6.0651],[13,-5.4956],[14,-4.8012],[15,-4.0571],[16,-3.3457],[17,-2.6796],[18,-2.0251],[19,-1.4273],[20,-0.8789],[21,-0.416],[22,-0.1672],[23,-0.086],[24,-0.0365],[25,-0.0109],[26,0],[30,0]]
       }
       ball_r_fk_ctrl.Rotation.Z {
        type=MovieSceneScriptingFloatChannel
        default=-0.8664
        thinned_from=31
        keys [[0,-0.8664],[1,-1.6604],[2,-2.8504],[3,-4.2863],[4,-5.8728],[5,-7.5148],[6,-9.4489],[7,-11.4842],[8,-13.441],[9,-14.8747],[10,-15.5848],[11,-15.9412],[12,-15.6466],[13,-14.7588],[14,-13.5657],[15,-12.1146],[16,-10.4512],[17,-8.1874],[18,-5.5154],[19,-3.3493],[20,-2.0634],[21,-1.2811],[22,-0.7884],[23,-0.4383],[24,-0.2093],[25,-0.0766],[26,0],[30,0]]
       }
       ball_r_fk_ctrl.Scale.X {type=MovieSceneScriptingFloatChannel, default=1, const [1,31]}
       ball_r_fk_ctrl.Scale.Y {type=MovieSceneScriptingFloatChannel, default=1, const [1,31]}
       ball_r_fk_ctrl.Scale.Z {type=MovieSceneScriptingFloatChannel, default=1, const [1,31]}
       foot_r_ik_ctrl.Location.X {
        type=MovieSceneScriptingFloatChannel
        default=-0.8486
        thinned_from=31
        keys [[0,-0.8486],[1,-0.919],[2,-0.7452],[3,-0.3219],[4,0.3758],[5,1.3801],[6,2.6717],[7,3.925],[8,5.002],[9,5.722],[10,6.0836],[11,6.3217],[12,6.3407],[13,6.1588],[14,5.8348],[15,5.3287],[16,4.6006],[17,3.6675],[18,2.6673],[19,1.7609],[20,0.9755],[21,0.304],[22,-0.1505],[23,-0.4184],[24,-0.6022],[25,-0.7158],[26,-0.7755],[27,-0.792],[30,-0.792]]
       }
       foot_r_ik_ctrl.Location.Y {
        type=MovieSceneScriptingFloatChannel
        default=-19.665
        thinned_from=31
        keys [[0,-19.665],[1,-19.4818],[2,-19.4371],[3,-19.5432],[4,-19.7959],[6,-20.5169],[7,-20.6794],[8,-20.6731],[10,-20.5307],[11,-20.5026],[12,-20.5538],[15,-20.9638],[17,-21.1172],[18,-21.1275],[19,-21.0127],[20,-20.764],[21,-20.4563],[22,-20.2135],[24,-19.9589],[26,-19.8508],[30,-19.8393]]
       }
       foot_r_ik_ctrl.Location.Z {
        type=MovieSceneScriptingFloatChannel
        default=0.2138
        thinned_from=31
        keys [[0,0.2138],[1,0.4119],[2,0.698],[3,1.0358],[4,1.4014],[5,1.7711],[6,2.1951],[7,2.6281],[8,3.03],[9,3.3161],[10,3.4581],[11,3.5299],[12,3.4729],[13,3.2965],[14,3.0543],[15,2.7526],[16,2.3987],[17,1.904],[18,1.3019],[19,0.8001],[20,0.4957],[21,0.3076],[22,0.1888],[23,0.1051],[24,0.0501],[25,0.0186],[26,0.0037],[27,0.0001],[30,0.0001]]
       }
       foot_r_ik_ctrl.Rotation.X {
        type=MovieSceneScriptingFloatChannel
        default=0.936
        thinned_from=31
        keys [[0,0.936],[1,1.8024],[2,3.0441],[3,4.5223],[4,6.1451],[5,7.8194],[6,9.7902],[7,11.8634],[8,13.8723],[9,15.3268],[10,16.0068],[11,16.3257],[12,15.9941],[13,15.0667],[14,13.8282],[15,12.3328],[16,10.6336],[17,8.3432],[18,5.6453],[19,3.4513],[20,2.1347],[21,1.3242],[22,0.8179],[23,0.466],[24,0.2359],[25,0.1031],[26,0.0422],[27,0.0274],[30,0.0274]]
       }
       foot_r_ik_ctrl.Rotation.Y {
        type=MovieSceneScriptingFloatChannel
        default=0.5216
        thinned_from=31
        keys [[0,0.5216],[1,1.2147],[2,1.7046],[3,2.1175],[4,2.5117],[5,2.9433],[6,3.5149],[7,4.1528],[8,4.9699],[9,5.4029],[10,5.1867],[11,4.8526],[12,4.4575],[13,3.9793],[14,3.4095],[15,2.8201],[16,2.2901],[17,1.869],[18,1.5018],[19,1.1426],[20,0.7473],[21,0.3871],[22,0.2045],[23,0.1677],[24,0.1474],[25,0.1398],[26,0.1381],[27,0.1389],[30,0.1389]]
       }
       foot_r_ik_ctrl.Rotation.Z {
        type=MovieSceneScriptingFloatChannel
        default=-36.751
        thinned_from=31
        keys [[0,-36.751],[1,-37.1011],[2,-36.3411],[3,-34.3194],[4,-30.8676],[5,-25.8147],[6,-19.3091],[7,-12.916],[8,-7.0816],[9,-3.1338],[10,-1.473],[11,-0.5434],[12,-0.8102],[13,-2.1632],[14,-4.2389],[15,-7.1147],[16,-10.877],[17,-15.3327],[18,-19.8607],[19,-23.9415],[20,-27.6433],[21,-30.938],[22,-33.1864],[23,-34.4783],[24,-35.3742],[25,-35.9366],[26,-36.2288],[27,-36.3118],[30,-36.3118]]
       }
       foot_r_ik_ctrl.Scale.X {type=MovieSceneScriptingFloatChannel, default=1, const [1,31]}
       foot_r_ik_ctrl.Scale.Y {type=MovieSceneScriptingFloatChannel, default=1, const [1,31]}
       foot_r_ik_ctrl.Scale.Z {type=MovieSceneScriptingFloatChannel, default=1, const [1,31]}
       ball_r_ik_ctrl.Location.X {type=MovieSceneScriptingFloatChannel, default=-0.0078, thinned_from=31, keys [[0,0],[30,0]]}
       ball_r_ik_ctrl.Location.Y {type=MovieSceneScriptingFloatChannel, default=0.0105, thinned_from=31, keys [[0,0],[30,0]]}
       ball_r_ik_ctrl.Location.Z {type=MovieSceneScriptingFloatChannel, default=-0.0206, thinned_from=31, keys [[0,0],[30,0]]}
       ball_r_ik_ctrl.Rotation.X {
        type=MovieSceneScriptingFloatChannel
        default=-0.9262
        thinned_from=31
        keys [[0,-0.9054],[1,-1.7615],[2,-2.9943],[3,-4.4648],[4,-6.0809],[5,-7.7488],[6,-9.7026],[7,-11.7424],[8,-13.6905],[9,-15.1045],[10,-15.805],[11,-16.1562],[12,-15.8654],[13,-14.9858],[14,-13.7942],[15,-12.3335],[16,-10.6507],[17,-8.3624],[18,-5.6601],[19,-3.4589],[20,-2.1337],[21,-1.314],[22,-0.8006],[23,-0.4444],[24,-0.2118],[25,-0.0772],[26,0],[30,0]]
       }
       ball_r_ik_ctrl.Rotation.Y {
        type=MovieSceneScriptingFloatChannel
        default=-0.3235
        thinned_from=31
        keys [[0,-0.3805],[1,-1.0799],[2,-1.5957],[3,-2.0576],[4,-2.5284],[5,-3.0707],[6,-3.814],[7,-4.6708],[8,-5.7445],[9,-6.372],[10,-6.2132],[11,-5.8881],[12,-5.4207],[13,-4.7991],[14,-4.0601],[15,-3.3006],[16,-2.6172],[17,-2.052],[18,-1.5655],[19,-1.13],[20,-0.6875],[21,-0.2941],[22,-0.0916],[23,-0.0438],[24,-0.0163],[25,-0.0035],[26,0],[30,0]]
       }
       ball_r_ik_ctrl.Rotation.Z {
        type=MovieSceneScriptingFloatChannel
        default=-0.1827
        thinned_from=31
        keys [[0,-0.1769],[1,-0.4505],[2,-0.6394],[3,-0.8154],[4,-1.0053],[5,-1.2348],[6,-1.5872],[7,-1.9831],[8,-2.373],[9,-2.6034],[10,-2.6179],[11,-2.5874],[12,-2.4682],[13,-2.2444],[14,-1.955],[15,-1.6398],[16,-1.3385],[17,-1.045],[18,-0.7591],[19,-0.5193],[20,-0.3169],[21,-0.1503],[22,-0.0609],[23,-0.0312],[24,-0.0132],[25,-0.0039],[26,0],[30,0]]
       }
       ball_r_ik_ctrl.Scale.X {type=MovieSceneScriptingFloatChannel, default=1, const [1,31]}
       ball_r_ik_ctrl.Scale.Y {type=MovieSceneScriptingFloatChannel, default=1, const [1,31]}
       ball_r_ik_ctrl.Scale.Z {type=MovieSceneScriptingFloatChannel, default=1, const [1,31]}
       foot_bk1_r_ctrl.Location.X {type=MovieSceneScriptingFloatChannel, default=0, const [0,31]}
       foot_bk1_r_ctrl.Location.Y {type=MovieSceneScriptingFloatChannel, default=0, const [0,31]}
       foot_bk1_r_ctrl.Location.Z {type=MovieSceneScriptingFloatChannel, default=0, const [0,31]}
       foot_bk1_r_ctrl.Rotation.X {type=MovieSceneScriptingFloatChannel, default=0, const [0,31]}
       foot_bk1_r_ctrl.Rotation.Y {type=MovieSceneScriptingFloatChannel, default=0, const [0,31]}
       foot_bk1_r_ctrl.Rotation.Z {type=MovieSceneScriptingFloatChannel, default=0, const [0,31]}
       foot_bk1_r_ctrl.Scale.X {type=MovieSceneScriptingFloatChannel, default=1, const [1,31]}
       foot_bk1_r_ctrl.Scale.Y {type=MovieSceneScriptingFloatChannel, default=1, const [1,31]}
       foot_bk1_r_ctrl.Scale.Z {type=MovieSceneScriptingFloatChannel, default=1, const [1,31]}
       foot_roll_r_ctrl.Location.X {type=MovieSceneScriptingFloatChannel, default=0, const [0,31]}
       foot_roll_r_ctrl.Location.Y {type=MovieSceneScriptingFloatChannel, default=0, const [0,31]}
       foot_roll_r_ctrl.Location.Z {type=MovieSceneScriptingFloatChannel, default=0, const [0,31]}
       foot_roll_r_ctrl.Rotation.X {type=MovieSceneScriptingFloatChannel, default=0, const [0,31]}
       foot_roll_r_ctrl.Rotation.Y {type=MovieSceneScriptingFloatChannel, default=0, const [0,31]}
       foot_roll_r_ctrl.Rotation.Z {type=MovieSceneScriptingFloatChannel, default=0, const [0,31]}
       foot_roll_r_ctrl.Scale.X {type=MovieSceneScriptingFloatChannel, default=1, const [1,31]}
       foot_roll_r_ctrl.Scale.Y {type=MovieSceneScriptingFloatChannel, default=1, const [1,31]}
       foot_roll_r_ctrl.Scale.Z {type=MovieSceneScriptingFloatChannel, default=1, const [1,31]}
       roll_blend_r_ctrl {type=MovieSceneScriptingFloatChannel, default=1, const [1,31]}
       tip_r_ctrl.Location.X {type=MovieSceneScriptingFloatChannel, default=0, const [0,31]}
       tip_r_ctrl.Location.Y {type=MovieSceneScriptingFloatChannel, default=0, const [0,31]}
       tip_r_ctrl.Location.Z {type=MovieSceneScriptingFloatChannel, default=0, const [0,31]}
       tip_r_ctrl.Rotation.X {type=MovieSceneScriptingFloatChannel, default=0, const [0,31]}
       tip_r_ctrl.Rotation.Y {type=MovieSceneScriptingFloatChannel, default=0, const [0,31]}
       tip_r_ctrl.Rotation.Z {type=MovieSceneScriptingFloatChannel, default=0, const [0,31]}
       tip_r_ctrl.Scale.X {type=MovieSceneScriptingFloatChannel, default=1, const [1,31]}
       tip_r_ctrl.Scale.Y {type=MovieSceneScriptingFloatChannel, default=1, const [1,31]}
       tip_r_ctrl.Scale.Z {type=MovieSceneScriptingFloatChannel, default=1, const [1,31]}
       heel_r_ctrl.Location.X {type=MovieSceneScriptingFloatChannel, default=0, const [0,31]}
       heel_r_ctrl.Location.Y {type=MovieSceneScriptingFloatChannel, default=0, const [0,31]}
       heel_r_ctrl.Location.Z {type=MovieSceneScriptingFloatChannel, default=0, const [0,31]}
       heel_r_ctrl.Rotation.X {type=MovieSceneScriptingFloatChannel, default=0, const [0,31]}
       heel_r_ctrl.Rotation.Y {type=MovieSceneScriptingFloatChannel, default=0, const [0,31]}
       heel_r_ctrl.Rotation.Z {type=MovieSceneScriptingFloatChannel, default=0, const [0,31]}
       heel_r_ctrl.Scale.X {type=MovieSceneScriptingFloatChannel, default=1, const [1,31]}
       heel_r_ctrl.Scale.Y {type=MovieSceneScriptingFloatChannel, default=1, const [1,31]}
       heel_r_ctrl.Scale.Z {type=MovieSceneScriptingFloatChannel, default=1, const [1,31]}
       leg_r_softness {type=MovieSceneScriptingFloatChannel, default=0, const [0,31]}
       leg_r_pv_ik_ctrl.Location.X {
        type=MovieSceneScriptingFloatChannel
        default=-0.6093
        thinned_from=31
        keys [[0,-0.6093],[1,-2.8059],[2,-4.8303],[3,-5.6463],[4,-5.2914],[5,-3.8681],[6,-2.3988],[7,-1.3875],[8,-0.2976],[9,0.4737],[10,0.8372],[11,1.1482],[12,1.3309],[13,1.4235],[14,1.4342],[15,1.2343],[16,0.6845],[17,0.5741],[18,1.2015],[19,1.7053],[20,1.7875],[21,1.7523],[22,1.707],[23,1.6813],[24,1.6352],[25,1.5851],[26,1.5454],[27,1.5293],[30,1.5293]]
       }
       leg_r_pv_ik_ctrl.Location.Y {
        type=MovieSceneScriptingFloatChannel
        default=9.2329
        thinned_from=31
        keys [[0,9.2329],[1,9.2122],[2,8.9735],[3,8.5216],[4,7.9757],[5,7.4117],[6,7.1888],[7,7.4344],[8,7.9813],[9,8.6335],[10,9.2159],[11,9.6764],[12,9.8967],[13,9.9177],[14,9.8647],[15,9.757],[16,9.6121],[17,9.1586],[18,8.3279],[19,7.5342],[20,7.0722],[21,6.8007],[22,6.6352],[24,6.3369],[25,6.2106],[26,6.1221],[27,6.0888],[30,6.0888]]
       }
       leg_r_pv_ik_ctrl.Location.Z {
        type=MovieSceneScriptingFloatChannel
        default=11.3527
        thinned_from=31
        keys [[0,11.3527],[1,11.0988],[2,10.6557],[3,9.9673],[4,9.1427],[5,8.3381],[6,8.0964],[7,8.531],[8,9.4574],[9,10.5664],[10,11.5574],[11,12.3803],[12,12.7929],[13,12.8486],[14,12.7598],[15,12.5342],[16,12.1832],[17,11.417],[18,10.2242],[19,9.1366],[20,8.4897],[21,8.1022],[22,7.8645],[24,7.4444],[25,7.2652],[26,7.1397],[27,7.0925],[30,7.0925]]
       }
       leg_r_pv_ik_ctrl.Rotation.X {
        type=MovieSceneScriptingFloatChannel
        default=-14.3593
        thinned_from=31
        keys [[0,-14.3593],[1,-14.0772],[2,-13.5747],[3,-12.7758],[5,-10.8332],[6,-10.5342],[7,-11.0508],[8,-12.1472],[9,-13.4439],[10,-14.5881],[11,-15.5276],[12,-15.995],[13,-16.0577],[14,-15.9571],[15,-15.7021],[16,-15.3052],[17,-14.4279],[18,-13.0413],[19,-11.7575],[20,-10.9854],[21,-10.5198],[22,-10.2329],[24,-9.7232],[25,-9.5048],[26,-9.3515],[27,-9.2938],[30,-9.2938]]
       }
       leg_r_pv_ik_ctrl.Rotation.Y {
        type=MovieSceneScriptingFloatChannel
        default=0.094
        thinned_from=31
        keys [[0,0.094],[1,-0.025],[2,-0.1501],[3,-0.2189],[4,-0.2221],[5,-0.1598],[6,-0.0826],[7,-0.0184],[8,0.0588],[9,0.124],[10,0.1681],[11,0.2064],[12,0.2273],[13,0.2331],[14,0.2307],[15,0.215],[16,0.1796],[17,0.152],[18,0.1491],[19,0.147],[20,0.1376],[21,0.1287],[22,0.1224],[23,0.1175],[24,0.1119],[25,0.1066],[26,0.1028],[27,0.1013],[30,0.1013]]
       }
       leg_r_pv_ik_ctrl.Rotation.Z {
        type=MovieSceneScriptingFloatChannel
        default=0.7614
        thinned_from=31
        keys [[0,0.7614],[1,3.2274],[2,5.5134],[3,6.4709],[4,6.1188],[5,4.5267],[6,2.8311],[7,1.6494],[8,0.3919],[9,-0.4732],[10,-0.8619],[11,-1.1889],[12,-1.3806],[13,-1.4819],[14,-1.4966],[15,-1.2806],[16,-0.6758],[17,-0.5688],[18,-1.3123],[19,-1.9315],[20,-2.0534],[21,-2.0279],[22,-1.9842],[23,-1.9622],[24,-1.9159],[25,-1.8633],[26,-1.8207],[27,-1.8033],[30,-1.8033]]
       }
       leg_r_pv_ik_ctrl.Scale.X {type=MovieSceneScriptingFloatChannel, default=1, const [1,31]}
       leg_r_pv_ik_ctrl.Scale.Y {type=MovieSceneScriptingFloatChannel, default=1, const [1,31]}
       leg_r_pv_ik_ctrl.Scale.Z {type=MovieSceneScriptingFloatChannel, default=1, const [1,31]}
       upperarm_l_fk_ctrl.Location.X {type=MovieSceneScriptingFloatChannel, default=0, thinned_from=31, keys [[0,0],[30,0]]}
       upperarm_l_fk_ctrl.Location.Y {type=MovieSceneScriptingFloatChannel, default=0, thinned_from=31, keys [[0,0],[30,0]]}
       upperarm_l_fk_ctrl.Location.Z {type=MovieSceneScriptingFloatChannel, default=0, thinned_from=31, keys [[0,0],[30,0]]}
       upperarm_l_fk_ctrl.Rotation.X {
        type=MovieSceneScriptingFloatChannel
        default=40.712
        keys [[0,40.712],[1,34.8115],[2,28.0461],[3,11.0003],[4,-5.9402],[5,-16.8602],[6,-25.1473],[8,-28.9345],[11,-21.6718],[20,-4.9017],[27,17.6136],[28,23.5762],[29,27.4218],[30,28.0629]]
       }
       upperarm_l_fk_ctrl.Rotation.Y {
        type=MovieSceneScriptingFloatChannel
        default=27.7377
        thinned_from=26
        keys [[0,27.7377],[1,33.5362],[2,39.0255],[3,40.1406],[4,37.1706],[5,33.7134],[6,29.0504],[7,26.6768],[8,25.5551],[9,23.7355],[11,20.01],[12,18.7712],[13,19.7908],[14,22.3001],[15,24.117],[16,24.5955],[17,23.77],[18,23.0852],[19,22.5464],[20,22.1264],[21,21.7514],[22,20.1993],[27,7.7236],[29,6.9166],[30,6.8907]]
       }
       upperarm_l_fk_ctrl.Rotation.Z {
        type=MovieSceneScriptingFloatChannel
        default=37.6096
        keys [[0,37.6096],[1,39.1038],[2,31.7113],[3,11.844],[4,-2.9037],[5,-5.1217],[6,1.3674],[7,8.9898],[8,9.6215],[9,4.9296],[10,0.4431],[11,-4.8618],[12,-10.1887],[13,-13.6567],[14,-14.2322],[15,-13.0455],[16,-10.7314],[17,-6.5022],[18,-1.2053],[19,1.1252],[20,0.1764],[21,0.3752],[22,2.1611],[23,4.4178],[24,4.9045],[25,3.8831],[26,2.8358],[27,1.9793],[28,1.3995],[29,1.1264],[30,1.3427]]
       }
       upperarm_l_fk_ctrl.Scale.X {type=MovieSceneScriptingFloatChannel, default=1, const [1,31]}
       upperarm_l_fk_ctrl.Scale.Y {type=MovieSceneScriptingFloatChannel, default=1, const [1,31]}
       upperarm_l_fk_ctrl.Scale.Z {type=MovieSceneScriptingFloatChannel, default=1, const [1,31]}
       lowerarm_l_fk_ctrl.Location.X {type=MovieSceneScriptingFloatChannel, default=0, thinned_from=31, keys [[0,0],[30,0]]}
       lowerarm_l_fk_ctrl.Location.Y {type=MovieSceneScriptingFloatChannel, default=0, thinned_from=31, keys [[0,0],[30,0]]}
       lowerarm_l_fk_ctrl.Location.Z {type=MovieSceneScriptingFloatChannel, default=0, thinned_from=31, keys [[0,0],[30,0]]}
       lowerarm_l_fk_ctrl.Rotation.X {
        type=MovieSceneScriptingFloatChannel
        default=-3.3393
        thinned_from=31
        keys [[0,-3.3393],[1,-3.4173],[2,-3.3724],[3,-3.2901],[4,-3.2744],[5,-3.2918],[6,-3.284],[7,-3.3325],[8,-3.3681],[9,-3.3687],[10,-3.3462],[12,-3.2493],[13,-3.2407],[14,-3.2733],[15,-3.3602],[16,-3.5185],[17,-3.7064],[18,-3.8696],[19,-3.937],[20,-3.8932],[21,-3.8598],[22,-3.8546],[24,-3.9036],[25,-3.918],[26,-3.9214],[28,-3.9116],[29,-3.9193],[30,-3.9384]]
       }
       lowerarm_l_fk_ctrl.Rotation.Y {
        type=MovieSceneScriptingFloatChannel
        default=1.656
        thinned_from=31
        keys [[0,1.656],[1,1.7781],[2,1.7122],[3,1.555],[4,1.5162],[5,1.559],[6,1.5405],[7,1.6435],[8,1.7053],[9,1.7063],[10,1.6684],[11,1.5675],[12,1.4432],[13,1.4138],[14,1.5132],[15,1.6923],[16,1.8977],[17,2.0526],[18,2.1413],[19,2.1684],[21,2.1369],[22,2.1345],[24,2.1556],[25,2.1612],[28,2.1588],[30,2.1689]]
       }
       lowerarm_l_fk_ctrl.Rotation.Z {
        type=MovieSceneScriptingFloatChannel
        default=47.555
        keys [[0,47.555],[1,40.1489],[2,44.2258],[3,53.2997],[4,55.4329],[5,53.0729],[6,54.0986],[7,48.2798],[8,44.639],[9,44.578],[10,46.8276],[11,52.5988],[12,59.3798],[13,60.9444],[14,55.5963],[15,45.4181],[16,32.1402],[17,19.6794],[18,10.1815],[19,6.4679],[20,8.8696],[21,10.7285],[22,11.0187],[23,9.7989],[24,8.2958],[25,7.5071],[26,7.3178],[27,7.6409],[28,7.8573],[29,7.4349],[30,6.3905]]
       }
       lowerarm_l_fk_ctrl.Scale.X {type=MovieSceneScriptingFloatChannel, default=1, const [1,31]}
       lowerarm_l_fk_ctrl.Scale.Y {type=MovieSceneScriptingFloatChannel, default=1, const [1,31]}
       lowerarm_l_fk_ctrl.Scale.Z {type=MovieSceneScriptingFloatChannel, default=1, const [1,31]}
       hand_l_fk_ctrl.Location.X {type=MovieSceneScriptingFloatChannel, default=0, thinned_from=31, keys [[0,0],[30,0]]}
       hand_l_fk_ctrl.Location.Y {type=MovieSceneScriptingFloatChannel, default=0, thinned_from=31, keys [[0,0],[30,0]]}
       hand_l_fk_ctrl.Location.Z {type=MovieSceneScriptingFloatChannel, default=0, thinned_from=31, keys [[0,0],[30,0]]}
       hand_l_fk_ctrl.Rotation.X {
        type=MovieSceneScriptingFloatChannel
        default=-6.2729
        keys [[0,-6.2729],[1,-5.1299],[2,-5.254],[3,-5.4429],[4,-5.1432],[5,-4.181],[6,-0.7555],[7,-5.045],[8,-8.6535],[9,-9.9108],[10,-11.314],[11,-11.6536],[12,-8.6407],[13,-4.14],[14,1.1681],[15,7.9318],[16,14.4334],[17,18.9509],[18,20.633],[19,20.6287],[20,21.5146],[21,21.9083],[22,20.7027],[23,18.2975],[24,17.9794],[25,16.991],[26,11.8816],[27,5.7792],[28,0.3829],[29,-3.5393],[30,-5.2154]]
       }
       hand_l_fk_ctrl.Rotation.Y {
        type=MovieSceneScriptingFloatChannel
        default=9.1462
        keys [[0,9.1462],[1,13.8548],[2,10.2569],[3,5.4151],[4,6.0194],[5,7.0809],[6,4.0154],[7,4.1066],[8,4.3387],[9,6.3106],[10,10.0335],[11,12.6177],[12,11.4392],[13,9.5614],[14,10.9994],[15,15.1913],[16,20.9997],[17,22.7068],[18,20.1262],[19,17.6235],[20,15.6833],[21,14.6643],[22,14.8227],[23,15.8051],[24,16.6947],[25,15.832],[26,12.9175],[27,8.867],[28,4.9702],[29,2.351],[30,1.7837]]
       }
       hand_l_fk_ctrl.Rotation.Z {
        type=MovieSceneScriptingFloatChannel
        default=-10.7666
        keys [[0,-10.7666],[1,-9.295],[2,-11.3363],[3,-12.8495],[4,-11.2288],[5,-10.2063],[6,-10.687],[7,-10.7075],[8,-8.5154],[9,-5.3844],[10,-4.0275],[11,-4.5553],[12,-5.2745],[13,-7.4229],[14,-9.2793],[15,-8.3365],[16,-6.0682],[17,-4.2691],[18,-3.5939],[19,-4.0647],[20,-4.7009],[21,-4.8648],[22,-5.0715],[23,-5.516],[24,-4.7273],[25,-4.7001],[26,-7.2054],[27,-9.8468],[28,-11.563],[29,-12.3446],[30,-12.579]]
       }
       hand_l_fk_ctrl.Scale.X {type=MovieSceneScriptingFloatChannel, default=1, const [1,31]}
       hand_l_fk_ctrl.Scale.Y {type=MovieSceneScriptingFloatChannel, default=1, const [1,31]}
       hand_l_fk_ctrl.Scale.Z {type=MovieSceneScriptingFloatChannel, default=1, const [1,31]}
       thumb_01_l_ctrl.Location.X {type=MovieSceneScriptingFloatChannel, default=0, thinned_from=31, keys [[0,0],[30,0]]}
       thumb_01_l_ctrl.Location.Y {type=MovieSceneScriptingFloatChannel, default=0, thinned_from=31, keys [[0,0],[30,0]]}
       thumb_01_l_ctrl.Location.Z {type=MovieSceneScriptingFloatChannel, default=0, thinned_from=31, keys [[0,0],[30,0]]}
       thumb_01_l_ctrl.Rotation.X {
        type=MovieSceneScriptingFloatChannel
        default=5.4807
        thinned_from=31
        keys [[0,5.4807],[1,5.7286],[2,5.2552],[3,4.0072],[4,2.2428],[6,-1.8024],[7,-3.5668],[8,-4.8148],[9,-5.2882],[12,-5.2882],[13,-5.0044],[14,-4.2341],[15,-3.0989],[16,-1.7205],[18,1.2797],[19,2.6582],[20,3.7934],[21,4.5637],[22,4.8475],[23,4.8009],[24,4.6778],[25,4.5038],[27,4.1049],[28,3.9309],[29,3.8078],[30,3.7611]]
       }
       thumb_01_l_ctrl.Rotation.Y {
        type=MovieSceneScriptingFloatChannel
        default=0.8396
        thinned_from=31
        keys [[0,0.8396],[1,0.4311],[2,0.6947],[3,1.3896],[4,2.372],[6,4.6245],[7,5.6069],[8,6.3018],[9,6.5654],[12,6.5654],[13,6.9253],[14,7.9022],[15,9.3418],[16,11.0899],[18,14.8947],[19,16.6428],[20,18.0824],[21,19.0593],[22,19.4192],[23,18.7426],[24,16.9589],[25,14.437],[27,8.6553],[28,6.1335],[29,4.3497],[30,3.6731]]
       }
       thumb_01_l_ctrl.Rotation.Z {
        type=MovieSceneScriptingFloatChannel
        default=-0.1513
        thinned_from=31
        keys [[0,-0.1513],[1,1.0401],[2,0.5168],[3,-0.863],[4,-2.8136],[6,-7.2859],[7,-9.2365],[8,-10.6163],[9,-11.1396],[12,-11.1396],[13,-10.936],[14,-10.3835],[15,-9.5692],[16,-8.5805],[18,-6.4285],[19,-5.4398],[20,-4.6255],[21,-4.073],[22,-3.8694],[23,-4.0648],[24,-4.5798],[25,-5.3078],[27,-6.977],[28,-7.7051],[29,-8.2201],[30,-8.4154]]
       }
       thumb_01_l_ctrl.Scale.X {type=MovieSceneScriptingFloatChannel, default=1, const [1,31]}
       thumb_01_l_ctrl.Scale.Y {type=MovieSceneScriptingFloatChannel, default=1, const [1,31]}
       thumb_01_l_ctrl.Scale.Z {type=MovieSceneScriptingFloatChannel, default=1, const [1,31]}
       thumb_02_l_ctrl.Location.X {type=MovieSceneScriptingFloatChannel, default=0, thinned_from=31, keys [[0,0],[30,0]]}
       thumb_02_l_ctrl.Location.Y {type=MovieSceneScriptingFloatChannel, default=0, thinned_from=31, keys [[0,0],[30,0]]}
       thumb_02_l_ctrl.Location.Z {type=MovieSceneScriptingFloatChannel, default=0, thinned_from=31, keys [[0,0],[30,0]]}
       thumb_02_l_ctrl.Rotation.X {type=MovieSceneScriptingFloatChannel, default=0, thinned_from=31, keys [[0,0],[30,0]]}
       thumb_02_l_ctrl.Rotation.Y {type=MovieSceneScriptingFloatChannel, default=0, thinned_from=31, keys [[0,0],[30,0]]}
       thumb_02_l_ctrl.Rotation.Z {
        type=MovieSceneScriptingFloatChannel
        default=17.8105
        thinned_from=31
        keys [[0,17.8105],[1,19.8675],[2,19.4192],[3,18.2373],[4,16.5664],[6,12.7354],[7,11.0644],[8,9.8825],[9,9.4342],[12,9.4342],[13,9.6209],[14,10.1274],[15,10.874],[16,11.7805],[18,13.7534],[19,14.6599],[20,15.4065],[21,15.9131],[22,16.0997],[23,15.5601],[24,14.1375],[25,12.1264],[27,7.5153],[28,5.5041],[29,4.0815],[30,3.5419]]
       }
       thumb_02_l_ctrl.Scale.X {type=MovieSceneScriptingFloatChannel, default=1, const [1,31]}
       thumb_02_l_ctrl.Scale.Y {type=MovieSceneScriptingFloatChannel, default=1, const [1,31]}
       thumb_02_l_ctrl.Scale.Z {type=MovieSceneScriptingFloatChannel, default=1, const [1,31]}
       thumb_03_l_ctrl.Location.X {type=MovieSceneScriptingFloatChannel, default=0, thinned_from=31, keys [[0,0],[30,0]]}
       thumb_03_l_ctrl.Location.Y {type=MovieSceneScriptingFloatChannel, default=0, thinned_from=31, keys [[0,0],[30,0]]}
       thumb_03_l_ctrl.Location.Z {type=MovieSceneScriptingFloatChannel, default=0, thinned_from=31, keys [[0,0],[30,0]]}
       thumb_03_l_ctrl.Rotation.X {type=MovieSceneScriptingFloatChannel, default=0, thinned_from=31, keys [[0,0],[30,0]]}
       thumb_03_l_ctrl.Rotation.Y {type=MovieSceneScriptingFloatChannel, default=0, thinned_from=31, keys [[0,0],[30,0]]}
       thumb_03_l_ctrl.Rotation.Z {
        type=MovieSceneScriptingFloatChannel
        default=0
        thinned_from=31
        keys [[0,0],[1,0],[2,-2.4423],[3,-8.8811],[4,-17.9842],[6,-38.8547],[7,-47.9578],[8,-54.3966],[9,-56.8389],[12,-56.8389],[13,-56.0365],[14,-53.8134],[15,-50.4459],[16,-46.2101],[17,-41.3824],[19,-31.0559],[20,-26.1095],[21,-21.676],[22,-18.0317],[24,-11.6618],[25,-8.6259],[26,-5.8581],[27,-3.4851],[28,-1.6333],[29,-0.4294],[30,0]]
       }
       thumb_03_l_ctrl.Scale.X {type=MovieSceneScriptingFloatChannel, default=1, const [1,31]}
       thumb_03_l_ctrl.Scale.Y {type=MovieSceneScriptingFloatChannel, default=1, const [1,31]}
       thumb_03_l_ctrl.Scale.Z {type=MovieSceneScriptingFloatChannel, default=1, const [1,31]}
       index_metacarpal_l_ctrl.Location.X {type=MovieSceneScriptingFloatChannel, default=0, thinned_from=31, keys [[0,0],[30,0]]}
       index_metacarpal_l_ctrl.Location.Y {type=MovieSceneScriptingFloatChannel, default=0, thinned_from=31, keys [[0,0],[30,0]]}
       index_metacarpal_l_ctrl.Location.Z {type=MovieSceneScriptingFloatChannel, default=0, thinned_from=31, keys [[0,0],[30,0]]}
       index_metacarpal_l_ctrl.Rotation.X {type=MovieSceneScriptingFloatChannel, default=0, thinned_from=31, keys [[0,0],[30,0]]}
       index_metacarpal_l_ctrl.Rotation.Y {
        type=MovieSceneScriptingFloatChannel
        default=0
        thinned_from=31
        keys [[0,0],[1,0],[2,0.1266],[3,0.4603],[4,0.9321],[6,2.0138],[7,2.4856],[8,2.8193],[9,2.9459],[12,2.9459],[13,2.9263],[14,2.8707],[15,2.7841],[16,2.6713],[17,2.5373],[18,2.3869],[19,2.225],[22,1.7196],[23,1.5204],[24,1.2668],[25,0.9837],[26,0.6961],[27,0.429],[28,0.2073],[29,0.056],[30,0]]
       }
       index_metacarpal_l_ctrl.Rotation.Z {type=MovieSceneScriptingFloatChannel, default=0, thinned_from=31, keys [[0,0],[30,0]]}
       index_metacarpal_l_ctrl.Scale.X {type=MovieSceneScriptingFloatChannel, default=1, const [1,31]}
       index_metacarpal_l_ctrl.Scale.Y {type=MovieSceneScriptingFloatChannel, default=1, const [1,31]}
       index_metacarpal_l_ctrl.Scale.Z {type=MovieSceneScriptingFloatChannel, default=1, const [1,31]}
       index_01_l_ctrl.Location.X {type=MovieSceneScriptingFloatChannel, default=0, thinned_from=31, keys [[0,0],[30,0]]}
       index_01_l_ctrl.Location.Y {type=MovieSceneScriptingFloatChannel, default=0, thinned_from=31, keys [[0,0],[30,0]]}
       index_01_l_ctrl.Location.Z {type=MovieSceneScriptingFloatChannel, default=0, thinned_from=31, keys [[0,0],[30,0]]}
       index_01_l_ctrl.Rotation.X {
        type=MovieSceneScriptingFloatChannel
        default=0
        thinned_from=31
        keys [[0,0],[1,0],[2,0.0164],[3,0.0597],[4,0.1209],[6,0.2611],[7,0.3223],[8,0.3656],[9,0.382],[12,0.382],[13,0.4301],[14,0.5607],[15,0.7531],[16,0.9868],[18,1.4955],[19,1.7292],[20,1.9217],[21,2.0522],[22,2.1004],[23,2.0101],[24,1.7722],[25,1.4358],[27,0.6646],[28,0.3282],[29,0.0902],[30,0]]
       }
       index_01_l_ctrl.Rotation.Y {
        type=MovieSceneScriptingFloatChannel
        default=0
        thinned_from=31
        keys [[0,0],[1,0],[2,0.0552],[3,0.2006],[4,0.4062],[6,0.8777],[7,1.0833],[8,1.2287],[9,1.2839],[12,1.2839],[13,1.2207],[14,1.0491],[15,0.7962],[16,0.4891],[18,-0.1792],[19,-0.4863],[20,-0.7392],[21,-0.9108],[22,-0.974],[23,-0.9322],[24,-0.8218],[25,-0.6659],[27,-0.3082],[28,-0.1522],[29,-0.0418],[30,0]]
       }
       index_01_l_ctrl.Rotation.Z {
        type=MovieSceneScriptingFloatChannel
        default=27.9551
        thinned_from=31
        keys [[0,27.9551],[1,31.9852],[2,29.0726],[3,21.394],[4,10.5379],[6,-14.3515],[7,-25.2076],[8,-32.8862],[9,-35.7989],[12,-35.7989],[13,-34.3913],[14,-30.5706],[15,-24.9403],[16,-18.1033],[18,-3.223],[19,3.6139],[20,9.2443],[21,13.0649],[22,14.4725],[23,13.8507],[24,12.2112],[25,9.8933],[27,4.5792],[28,2.2613],[29,0.6219],[30,0]]
       }
       index_01_l_ctrl.Scale.X {type=MovieSceneScriptingFloatChannel, default=1, const [1,31]}
       index_01_l_ctrl.Scale.Y {type=MovieSceneScriptingFloatChannel, default=1, const [1,31]}
       index_01_l_ctrl.Scale.Z {type=MovieSceneScriptingFloatChannel, default=1, const [1,31]}
       index_02_l_ctrl.Location.X {type=MovieSceneScriptingFloatChannel, default=0, thinned_from=31, keys [[0,0],[30,0]]}
       index_02_l_ctrl.Location.Y {type=MovieSceneScriptingFloatChannel, default=0, thinned_from=31, keys [[0,0],[30,0]]}
       index_02_l_ctrl.Location.Z {type=MovieSceneScriptingFloatChannel, default=0, thinned_from=31, keys [[0,0],[30,0]]}
       index_02_l_ctrl.Rotation.X {type=MovieSceneScriptingFloatChannel, default=0, thinned_from=31, keys [[0,0],[30,0]]}
       index_02_l_ctrl.Rotation.Y {type=MovieSceneScriptingFloatChannel, default=0, thinned_from=31, keys [[0,0],[30,0]]}
       index_02_l_ctrl.Rotation.Z {
        type=MovieSceneScriptingFloatChannel
        default=17.5873
        thinned_from=31
        keys [[0,17.5873],[1,20.1227],[2,16.1295],[3,5.6017],[4,-9.2822],[6,-43.4066],[7,-58.2906],[8,-68.8183],[9,-72.8116],[12,-72.8116],[13,-70.8959],[14,-65.6855],[15,-57.9853],[16,-48.6004],[18,-27.9965],[19,-18.3874],[20,-10.3137],[21,-4.5803],[22,-1.9922],[23,-1.3346],[24,-0.8404],[25,-0.4864],[26,-0.249],[27,-0.105],[28,-0.0311],[29,0],[30,0]]
       }
       index_02_l_ctrl.Scale.X {type=MovieSceneScriptingFloatChannel, default=1, const [1,31]}
       index_02_l_ctrl.Scale.Y {type=MovieSceneScriptingFloatChannel, default=1, const [1,31]}
       index_02_l_ctrl.Scale.Z {type=MovieSceneScriptingFloatChannel, default=1, const [1,31]}
       index_03_l_ctrl.Location.X {type=MovieSceneScriptingFloatChannel, default=0, thinned_from=31, keys [[0,0],[30,0]]}
       index_03_l_ctrl.Location.Y {type=MovieSceneScriptingFloatChannel, default=0, thinned_from=31, keys [[0,0],[30,0]]}
       index_03_l_ctrl.Location.Z {type=MovieSceneScriptingFloatChannel, default=0, thinned_from=31, keys [[0,0],[30,0]]}
       index_03_l_ctrl.Rotation.X {type=MovieSceneScriptingFloatChannel, default=0, thinned_from=31, keys [[0,0],[30,0]]}
       index_03_l_ctrl.Rotation.Y {type=MovieSceneScriptingFloatChannel, default=0, thinned_from=31, keys [[0,0],[30,0]]}
       index_03_l_ctrl.Rotation.Z {
        type=MovieSceneScriptingFloatChannel
        default=-1.8875
        thinned_from=31
        keys [[0,-1.8875],[1,-4.1673],[2,-8.3201],[3,-13.1484],[4,-18.2732],[5,-23.3158],[6,-27.8973],[7,-31.6388],[8,-34.1616],[9,-35.0867],[12,-35.0867],[13,-34.087],[14,-31.3737],[15,-27.375],[16,-22.5195],[18,-11.9517],[19,-7.0962],[20,-3.0975],[21,-0.3842],[22,0.6155],[23,0.5891],[24,0.5193],[25,0.4208],[27,0.1947],[28,0.0962],[29,0.0264],[30,0]]
       }
       index_03_l_ctrl.Scale.X {type=MovieSceneScriptingFloatChannel, default=1, const [1,31]}
       index_03_l_ctrl.Scale.Y {type=MovieSceneScriptingFloatChannel, default=1, const [1,31]}
       index_03_l_ctrl.Scale.Z {type=MovieSceneScriptingFloatChannel, default=1, const [1,31]}
       middle_metacarpal_l_ctrl.Location.X {type=MovieSceneScriptingFloatChannel, default=0, const [0,31]}
       middle_metacarpal_l_ctrl.Location.Y {type=MovieSceneScriptingFloatChannel, default=0, const [0,31]}
       middle_metacarpal_l_ctrl.Location.Z {type=MovieSceneScriptingFloatChannel, default=0, const [0,31]}
       middle_metacarpal_l_ctrl.Rotation.X {type=MovieSceneScriptingFloatChannel, default=0, const [0,31]}
       middle_metacarpal_l_ctrl.Rotation.Y {type=MovieSceneScriptingFloatChannel, default=0, const [0,31]}
       middle_metacarpal_l_ctrl.Rotation.Z {type=MovieSceneScriptingFloatChannel, default=0, const [0,31]}
       middle_metacarpal_l_ctrl.Scale.X {type=MovieSceneScriptingFloatChannel, default=1, const [1,31]}
       middle_metacarpal_l_ctrl.Scale.Y {type=MovieSceneScriptingFloatChannel, default=1, const [1,31]}
       middle_metacarpal_l_ctrl.Scale.Z {type=MovieSceneScriptingFloatChannel, default=1, const [1,31]}
       middle_01_l_ctrl.Location.X {type=MovieSceneScriptingFloatChannel, default=0, thinned_from=31, keys [[0,0],[30,0]]}
       middle_01_l_ctrl.Location.Y {type=MovieSceneScriptingFloatChannel, default=0, thinned_from=31, keys [[0,0],[30,0]]}
       middle_01_l_ctrl.Location.Z {type=MovieSceneScriptingFloatChannel, default=0, thinned_from=31, keys [[0,0],[30,0]]}
       middle_01_l_ctrl.Rotation.X {
        type=MovieSceneScriptingFloatChannel
        default=4.8669
        thinned_from=31
        keys [[0,4.8669],[1,5.5686],[2,5.0367],[3,3.6344],[4,1.6518],[6,-2.8936],[7,-4.8762],[8,-6.2785],[9,-6.8104],[22,-6.8104],[23,-6.5178],[24,-5.7463],[25,-4.6556],[27,-2.1549],[28,-1.0641],[29,-0.2926],[30,0]]
       }
       middle_01_l_ctrl.Rotation.Y {
        type=MovieSceneScriptingFloatChannel
        default=3.2659
        thinned_from=31
        keys [[0,3.2659],[1,3.7367],[2,3.6543],[3,3.4369],[4,3.1295],[6,2.4249],[7,2.1175],[8,1.9001],[9,1.8177],[22,1.8177],[23,1.7396],[24,1.5337],[25,1.2426],[27,0.5751],[28,0.284],[29,0.0781],[30,0]]
       }
       middle_01_l_ctrl.Rotation.Z {
        type=MovieSceneScriptingFloatChannel
        default=8.6317
        thinned_from=31
        keys [[0,8.6317],[1,9.8761],[2,7.1083],[3,-0.1887],[4,-10.5051],[6,-34.1574],[7,-44.4738],[8,-51.7708],[9,-54.5386],[12,-54.5386],[13,-53.6175],[14,-51.0741],[15,-47.2383],[16,-42.4398],[17,-37.0084],[19,-25.5666],[20,-20.2158],[21,-15.5515],[22,-11.9034],[23,-9.0722],[24,-6.6349],[25,-4.5865],[26,-2.9218],[27,-1.6359],[28,-0.7237],[29,-0.1801],[30,0]]
       }
       middle_01_l_ctrl.Scale.X {type=MovieSceneScriptingFloatChannel, default=1, const [1,31]}
       middle_01_l_ctrl.Scale.Y {type=MovieSceneScriptingFloatChannel, default=1, const [1,31]}
       middle_01_l_ctrl.Scale.Z {type=MovieSceneScriptingFloatChannel, default=1, const [1,31]}
       middle_02_l_ctrl.Location.X {type=MovieSceneScriptingFloatChannel, default=0, thinned_from=31, keys [[0,0],[30,0]]}
       middle_02_l_ctrl.Location.Y {type=MovieSceneScriptingFloatChannel, default=0, thinned_from=31, keys [[0,0],[30,0]]}
       middle_02_l_ctrl.Location.Z {type=MovieSceneScriptingFloatChannel, default=0, thinned_from=31, keys [[0,0],[30,0]]}
       middle_02_l_ctrl.Rotation.X {type=MovieSceneScriptingFloatChannel, default=0, thinned_from=31, keys [[0,0],[30,0]]}
       middle_02_l_ctrl.Rotation.Y {type=MovieSceneScriptingFloatChannel, default=0, thinned_from=31, keys [[0,0],[30,0]]}
       middle_02_l_ctrl.Rotation.Z {
        type=MovieSceneScriptingFloatChannel
        default=4.2054
        thinned_from=31
        keys [[0,4.2054],[1,4.8117],[2,2.1433],[3,-4.8917],[4,-14.8375],[6,-37.6404],[7,-47.5863],[8,-54.6213],[9,-57.2897],[12,-57.2897],[13,-56.2378],[14,-53.3371],[15,-48.9704],[16,-43.5202],[17,-37.369],[18,-30.8996],[19,-24.4946],[20,-18.5364],[21,-13.4079],[22,-9.4915],[23,-6.6468],[24,-4.4278],[25,-2.7585],[26,-1.563],[27,-0.7653],[28,-0.2895],[29,-0.0597],[30,0]]
       }
       middle_02_l_ctrl.Scale.X {type=MovieSceneScriptingFloatChannel, default=1, const [1,31]}
       middle_02_l_ctrl.Scale.Y {type=MovieSceneScriptingFloatChannel, default=1, const [1,31]}
       middle_02_l_ctrl.Scale.Z {type=MovieSceneScriptingFloatChannel, default=1, const [1,31]}
       middle_03_l_ctrl.Location.X {type=MovieSceneScriptingFloatChannel, default=0, thinned_from=31, keys [[0,0],[30,0]]}
       middle_03_l_ctrl.Location.Y {type=MovieSceneScriptingFloatChannel, default=0, thinned_from=31, keys [[0,0],[30,0]]}
       middle_03_l_ctrl.Location.Z {type=MovieSceneScriptingFloatChannel, default=0, thinned_from=31, keys [[0,0],[30,0]]}
       middle_03_l_ctrl.Rotation.X {type=MovieSceneScriptingFloatChannel, default=0, thinned_from=31, keys [[0,0],[30,0]]}
       middle_03_l_ctrl.Rotation.Y {type=MovieSceneScriptingFloatChannel, default=0, thinned_from=31, keys [[0,0],[30,0]]}
       middle_03_l_ctrl.Rotation.Z {
        type=MovieSceneScriptingFloatChannel
        default=0
        thinned_from=31
        keys [[0,0],[1,0],[2,-2.524],[3,-9.1783],[4,-18.5861],[6,-40.1553],[7,-49.5631],[8,-56.2174],[9,-58.7415],[12,-58.7415],[13,-56.832],[14,-51.6492],[15,-44.0114],[16,-34.7369],[18,-14.5512],[19,-5.2767],[20,2.3611],[21,7.5439],[22,9.4533],[23,9.0471],[24,7.9763],[25,6.4622],[27,2.9911],[28,1.4771],[29,0.4062],[30,0]]
       }
       middle_03_l_ctrl.Scale.X {type=MovieSceneScriptingFloatChannel, default=1, const [1,31]}
       middle_03_l_ctrl.Scale.Y {type=MovieSceneScriptingFloatChannel, default=1, const [1,31]}
       middle_03_l_ctrl.Scale.Z {type=MovieSceneScriptingFloatChannel, default=1, const [1,31]}
       ring_metacarpal_l_ctrl.Location.X {type=MovieSceneScriptingFloatChannel, default=0, thinned_from=31, keys [[0,0],[30,0]]}
       ring_metacarpal_l_ctrl.Location.Y {type=MovieSceneScriptingFloatChannel, default=0, thinned_from=31, keys [[0,0],[30,0]]}
       ring_metacarpal_l_ctrl.Location.Z {type=MovieSceneScriptingFloatChannel, default=0, thinned_from=31, keys [[0,0],[30,0]]}
       ring_metacarpal_l_ctrl.Rotation.X {type=MovieSceneScriptingFloatChannel, default=0, thinned_from=31, keys [[0,0],[30,0]]}
       ring_metacarpal_l_ctrl.Rotation.Y {
        type=MovieSceneScriptingFloatChannel
        default=-1.2794
        thinned_from=31
        keys [[0,-1.2794],[1,-1.4638],[2,-1.4509],[3,-1.4168],[4,-1.3687],[6,-1.2583],[7,-1.2102],[8,-1.1761],[9,-1.1632],[22,-1.1632],[23,-1.1132],[24,-0.9815],[25,-0.7952],[27,-0.3681],[28,-0.1818],[29,-0.05],[30,0]]
       }
       ring_metacarpal_l_ctrl.Rotation.Z {type=MovieSceneScriptingFloatChannel, default=0, thinned_from=31, keys [[0,0],[30,0]]}
       ring_metacarpal_l_ctrl.Scale.X {type=MovieSceneScriptingFloatChannel, default=1, const [1,31]}
       ring_metacarpal_l_ctrl.Scale.Y {type=MovieSceneScriptingFloatChannel, default=1, const [1,31]}
       ring_metacarpal_l_ctrl.Scale.Z {type=MovieSceneScriptingFloatChannel, default=1, const [1,31]}
       ring_01_l_ctrl.Location.X {type=MovieSceneScriptingFloatChannel, default=0, thinned_from=31, keys [[0,0],[30,0]]}
       ring_01_l_ctrl.Location.Y {type=MovieSceneScriptingFloatChannel, default=0, thinned_from=31, keys [[0,0],[30,0]]}
       ring_01_l_ctrl.Location.Z {type=MovieSceneScriptingFloatChannel, default=0, thinned_from=31, keys [[0,0],[30,0]]}
       ring_01_l_ctrl.Rotation.X {
        type=MovieSceneScriptingFloatChannel
        default=5.4615
        thinned_from=31
        keys [[0,5.4615],[1,6.2489],[2,5.4159],[3,3.2199],[4,0.1151],[6,-7.0032],[7,-10.108],[8,-12.3041],[9,-13.1371],[12,-13.1371],[13,-13.0329],[14,-12.7399],[15,-12.287],[16,-11.7032],[17,-11.0173],[18,-10.2585],[21,-7.834],[22,-7.073],[23,-6.2103],[24,-5.1468],[26,-2.8067],[27,-1.7248],[28,-0.8315],[29,-0.2241],[30,0]]
       }
       ring_01_l_ctrl.Rotation.Y {
        type=MovieSceneScriptingFloatChannel
        default=-1.9417
        thinned_from=31
        keys [[0,-1.9417],[1,-2.2216],[2,-2.145],[3,-1.943],[4,-1.6574],[6,-1.0027],[7,-0.7171],[8,-0.5151],[9,-0.4385],[12,-0.4385],[13,-0.3029],[14,0.0652],[15,0.6076],[16,1.2663],[18,2.6998],[19,3.3584],[20,3.9008],[21,4.2689],[22,4.4045],[23,4.2152],[24,3.7163],[25,3.0109],[27,1.3936],[28,0.6882],[29,0.1893],[30,0]]
       }
       ring_01_l_ctrl.Rotation.Z {
        type=MovieSceneScriptingFloatChannel
        default=8.595
        thinned_from=31
        keys [[0,8.595],[1,10.3754],[2,7.2673],[3,-0.9265],[4,-12.511],[6,-39.0706],[7,-50.6551],[8,-58.849],[9,-61.957],[12,-61.957],[13,-61.5288],[14,-60.3201],[15,-58.4453],[16,-56.0187],[17,-53.1544],[18,-49.9669],[19,-46.5702],[21,-39.6066],[22,-36.2682],[23,-32.3955],[24,-27.5503],[25,-22.1914],[26,-16.7778],[27,-11.7685],[28,-7.6222],[29,-4.7978],[30,-3.7544]]
       }
       ring_01_l_ctrl.Scale.X {type=MovieSceneScriptingFloatChannel, default=1, const [1,31]}
       ring_01_l_ctrl.Scale.Y {type=MovieSceneScriptingFloatChannel, default=1, const [1,31]}
       ring_01_l_ctrl.Scale.Z {type=MovieSceneScriptingFloatChannel, default=1, const [1,31]}
       ring_02_l_ctrl.Location.X {type=MovieSceneScriptingFloatChannel, default=0, thinned_from=31, keys [[0,0],[30,0]]}
       ring_02_l_ctrl.Location.Y {type=MovieSceneScriptingFloatChannel, default=0, thinned_from=31, keys [[0,0],[30,0]]}
       ring_02_l_ctrl.Location.Z {type=MovieSceneScriptingFloatChannel, default=0, thinned_from=31, keys [[0,0],[30,0]]}
       ring_02_l_ctrl.Rotation.X {type=MovieSceneScriptingFloatChannel, default=0, thinned_from=31, keys [[0,0],[30,0]]}
       ring_02_l_ctrl.Rotation.Y {type=MovieSceneScriptingFloatChannel, default=0, thinned_from=31, keys [[0,0],[30,0]]}
       ring_02_l_ctrl.Rotation.Z {
        type=MovieSceneScriptingFloatChannel
        default=-3.7544
        thinned_from=31
        keys [[0,-3.7544],[1,-3.7544],[2,-6.3979],[3,-13.3671],[4,-23.2199],[6,-45.8096],[7,-55.6626],[8,-62.6317],[9,-65.2752],[12,-65.2752],[13,-64.053],[14,-60.6924],[15,-55.6518],[16,-49.39],[17,-42.3657],[18,-35.0374],[19,-27.8639],[20,-21.3039],[21,-15.8158],[22,-11.8585],[23,-9.1836],[24,-7.1733],[25,-5.733],[26,-4.7675],[27,-4.1818],[28,-3.8811],[29,-3.7703],[30,-3.7544]]
       }
       ring_02_l_ctrl.Scale.X {type=MovieSceneScriptingFloatChannel, default=1, const [1,31]}
       ring_02_l_ctrl.Scale.Y {type=MovieSceneScriptingFloatChannel, default=1, const [1,31]}
       ring_02_l_ctrl.Scale.Z {type=MovieSceneScriptingFloatChannel, default=1, const [1,31]}
       ring_03_l_ctrl.Location.X {type=MovieSceneScriptingFloatChannel, default=0, thinned_from=31, keys [[0,0],[30,0]]}
       ring_03_l_ctrl.Location.Y {type=MovieSceneScriptingFloatChannel, default=0, thinned_from=31, keys [[0,0],[30,0]]}
       ring_03_l_ctrl.Location.Z {type=MovieSceneScriptingFloatChannel, default=0, thinned_from=31, keys [[0,0],[30,0]]}
       ring_03_l_ctrl.Rotation.X {type=MovieSceneScriptingFloatChannel, default=0, thinned_from=31, keys [[0,0],[30,0]]}
       ring_03_l_ctrl.Rotation.Y {type=MovieSceneScriptingFloatChannel, default=0, const [0,31]}
       ring_03_l_ctrl.Rotation.Z {
        type=MovieSceneScriptingFloatChannel
        default=-3.7545
        thinned_from=31
        keys [[0,-3.7545],[1,-3.7545],[2,-6.2208],[3,-12.7229],[4,-21.9155],[6,-42.9913],[7,-52.184],[8,-58.6861],[9,-61.1524],[12,-61.1524],[13,-59.5887],[14,-55.3407],[15,-49.0728],[16,-41.4495],[18,-24.7948],[19,-17.0924],[20,-10.6924],[21,-6.2596],[22,-4.4584],[23,-4.226],[24,-4.0514],[25,-3.9263],[26,-3.8425],[27,-3.7916],[28,-3.7655],[30,-3.7655]]
       }
       ring_03_l_ctrl.Scale.X {type=MovieSceneScriptingFloatChannel, default=1, const [1,31]}
       ring_03_l_ctrl.Scale.Y {type=MovieSceneScriptingFloatChannel, default=1, const [1,31]}
       ring_03_l_ctrl.Scale.Z {type=MovieSceneScriptingFloatChannel, default=1, const [1,31]}
       pinky_metacarpal_l_ctrl.Location.X {type=MovieSceneScriptingFloatChannel, default=0, thinned_from=31, keys [[0,0],[30,0]]}
       pinky_metacarpal_l_ctrl.Location.Y {type=MovieSceneScriptingFloatChannel, default=0, thinned_from=31, keys [[0,0],[30,0]]}
       pinky_metacarpal_l_ctrl.Location.Z {type=MovieSceneScriptingFloatChannel, default=0, thinned_from=31, keys [[0,0],[30,0]]}
       pinky_metacarpal_l_ctrl.Rotation.X {type=MovieSceneScriptingFloatChannel, default=0, thinned_from=31, keys [[0,0],[30,0]]}
       pinky_metacarpal_l_ctrl.Rotation.Y {
        type=MovieSceneScriptingFloatChannel
        default=0
        thinned_from=31
        keys [[0,0],[1,0],[2,-0.1372],[3,-0.499],[4,-1.0104],[6,-2.183],[7,-2.6944],[8,-3.0562],[9,-3.1934],[22,-3.1934],[23,-3.0562],[24,-2.6944],[25,-2.183],[27,-1.0104],[28,-0.499],[29,-0.1372],[30,0]]
       }
       pinky_metacarpal_l_ctrl.Rotation.Z {type=MovieSceneScriptingFloatChannel, default=0, thinned_from=31, keys [[0,0],[30,0]]}
       pinky_metacarpal_l_ctrl.Scale.X {type=MovieSceneScriptingFloatChannel, default=1, const [1,31]}
       pinky_metacarpal_l_ctrl.Scale.Y {type=MovieSceneScriptingFloatChannel, default=1, const [1,31]}
       pinky_metacarpal_l_ctrl.Scale.Z {type=MovieSceneScriptingFloatChannel, default=1, const [1,31]}
       pinky_01_l_ctrl.Location.X {type=MovieSceneScriptingFloatChannel, default=0, thinned_from=31, keys [[0,0],[30,0]]}
       pinky_01_l_ctrl.Location.Y {type=MovieSceneScriptingFloatChannel, default=0, thinned_from=31, keys [[0,0],[30,0]]}
       pinky_01_l_ctrl.Location.Z {type=MovieSceneScriptingFloatChannel, default=0, thinned_from=31, keys [[0,0],[30,0]]}
       pinky_01_l_ctrl.Rotation.X {
        type=MovieSceneScriptingFloatChannel
        default=0
        thinned_from=31
        keys [[0,0],[1,0],[2,-0.5522],[3,-2.0079],[4,-4.0661],[6,-8.7848],[7,-10.8429],[8,-12.2987],[9,-12.8509],[12,-12.8509],[13,-12.832],[14,-12.7708],[15,-12.6599],[16,-12.492],[17,-12.26],[18,-11.9566],[19,-11.5746],[20,-11.1067],[21,-10.5457],[22,-9.8843],[23,-8.913],[24,-7.5367],[25,-5.9202],[26,-4.2282],[27,-2.6255],[28,-1.2767],[29,-0.3466],[30,0]]
       }
       pinky_01_l_ctrl.Rotation.Y {
        type=MovieSceneScriptingFloatChannel
        default=0
        thinned_from=31
        keys [[0,0],[1,0],[2,-0.1035],[3,-0.3765],[4,-0.7625],[6,-1.6474],[7,-2.0333],[8,-2.3063],[9,-2.4099],[12,-2.4099],[13,-2.2077],[14,-1.6588],[15,-0.85],[16,0.1322],[18,2.2699],[19,3.252],[20,4.0609],[21,4.6097],[22,4.8119],[23,4.6052],[24,4.0601],[25,3.2894],[27,1.5225],[28,0.7519],[29,0.2068],[30,0]]
       }
       pinky_01_l_ctrl.Rotation.Z {
        type=MovieSceneScriptingFloatChannel
        default=-4.4418
        thinned_from=31
        keys [[0,-4.4418],[1,-4.2011],[2,-6.9606],[3,-14.2355],[4,-24.5208],[6,-48.1019],[7,-58.3872],[8,-65.6622],[9,-68.4217],[12,-68.4217],[13,-68.1598],[14,-67.3994],[15,-66.1787],[16,-64.5357],[17,-62.5085],[18,-60.1351],[19,-57.4536],[20,-54.502],[21,-51.3185],[22,-47.9411],[23,-43.4934],[24,-37.511],[25,-30.6495],[26,-23.5649],[27,-16.913],[28,-11.3496],[29,-7.5307],[30,-6.1119]]
       }
       pinky_01_l_ctrl.Scale.X {type=MovieSceneScriptingFloatChannel, default=1, const [1,31]}
       pinky_01_l_ctrl.Scale.Y {type=MovieSceneScriptingFloatChannel, default=1, const [1,31]}
       pinky_01_l_ctrl.Scale.Z {type=MovieSceneScriptingFloatChannel, default=1, const [1,31]}
       pinky_02_l_ctrl.Location.X {type=MovieSceneScriptingFloatChannel, default=0, thinned_from=31, keys [[0,0],[30,0]]}
       pinky_02_l_ctrl.Location.Y {type=MovieSceneScriptingFloatChannel, default=0, thinned_from=31, keys [[0,0],[30,0]]}
       pinky_02_l_ctrl.Location.Z {type=MovieSceneScriptingFloatChannel, default=0, thinned_from=31, keys [[0,0],[30,0]]}
       pinky_02_l_ctrl.Rotation.X {type=MovieSceneScriptingFloatChannel, default=0, thinned_from=31, keys [[0,0],[30,0]]}
       pinky_02_l_ctrl.Rotation.Y {type=MovieSceneScriptingFloatChannel, default=0, thinned_from=31, keys [[0,0],[30,0]]}
       pinky_02_l_ctrl.Rotation.Z {
        type=MovieSceneScriptingFloatChannel
        default=-6.1119
        thinned_from=31
        keys [[0,-6.1119],[1,-6.1119],[2,-8.3945],[3,-14.4123],[4,-22.9203],[6,-42.4264],[7,-50.9344],[8,-56.9522],[9,-59.2348],[12,-59.2348],[13,-58.3983],[14,-56.0856],[15,-52.592],[16,-48.2128],[17,-43.2431],[19,-32.7138],[20,-27.7445],[21,-23.3659],[22,-19.8733],[23,-17.0224],[24,-14.4028],[25,-12.0606],[26,-10.0413],[27,-8.391],[28,-7.1554],[29,-6.3804],[30,-6.1119]]
       }
       pinky_02_l_ctrl.Scale.X {type=MovieSceneScriptingFloatChannel, default=1, const [1,31]}
       pinky_02_l_ctrl.Scale.Y {type=MovieSceneScriptingFloatChannel, default=1, const [1,31]}
       pinky_02_l_ctrl.Scale.Z {type=MovieSceneScriptingFloatChannel, default=1, const [1,31]}
       pinky_03_l_ctrl.Location.X {type=MovieSceneScriptingFloatChannel, default=0, thinned_from=31, keys [[0,0],[30,0]]}
       pinky_03_l_ctrl.Location.Y {type=MovieSceneScriptingFloatChannel, default=0, thinned_from=31, keys [[0,0],[30,0]]}
       pinky_03_l_ctrl.Location.Z {type=MovieSceneScriptingFloatChannel, default=0, thinned_from=31, keys [[0,0],[30,0]]}
       pinky_03_l_ctrl.Rotation.X {type=MovieSceneScriptingFloatChannel, default=0, thinned_from=31, keys [[0,0],[30,0]]}
       pinky_03_l_ctrl.Rotation.Y {type=MovieSceneScriptingFloatChannel, default=0, thinned_from=31, keys [[0,0],[30,0]]}
       pinky_03_l_ctrl.Rotation.Z {
        type=MovieSceneScriptingFloatChannel
        default=-6.1119
        thinned_from=31
        keys [[0,-6.1119],[1,-6.1119],[2,-9.3501],[3,-17.8871],[4,-29.9566],[6,-57.6284],[7,-69.698],[8,-78.235],[9,-81.4732],[12,-81.4732],[13,-80.0092],[14,-75.9795],[15,-69.9273],[16,-62.396],[17,-53.9291],[18,-45.0698],[19,-36.3614],[20,-28.3473],[21,-21.5709],[22,-16.5755],[23,-13.1217],[24,-10.5262],[25,-8.6665],[26,-7.4198],[27,-6.6637],[28,-6.2754],[29,-6.1323],[30,-6.1119]]
       }
       pinky_03_l_ctrl.Scale.X {type=MovieSceneScriptingFloatChannel, default=1, const [1,31]}
       pinky_03_l_ctrl.Scale.Y {type=MovieSceneScriptingFloatChannel, default=1, const [1,31]}
       pinky_03_l_ctrl.Scale.Z {type=MovieSceneScriptingFloatChannel, default=1, const [1,31]}
       hand_l_ik_ctrl.Location.X {
        type=MovieSceneScriptingFloatChannel
        default=-12.9059
        thinned_from=31
        keys [[0,-12.9059],[1,-11.4908],[2,-9.4475],[3,-5.7063],[4,-1.2737],[5,1.9191],[6,0.8927],[7,-2.6382],[8,-6.7424],[9,-10.5106],[10,-12.5787],[11,-13.8057],[12,-13.9484],[13,-12.7363],[14,-10.4245],[16,-4.6741],[17,-1.2092],[18,2.5643],[19,4.8361],[20,5.8026],[21,6.7861],[22,7.6654],[23,8.3197],[24,8.6276],[25,8.638],[26,8.5236],[28,8.1916],[29,8.1082],[30,8.1695]]
       }
       hand_l_ik_ctrl.Location.Y {
        type=MovieSceneScriptingFloatChannel
        default=23.993
        thinned_from=31
        keys [[0,23.993],[1,19.101],[2,12.6765],[3,3.6568],[4,-5.2596],[5,-12.4531],[6,-16.6052],[7,-17.4387],[8,-13.5167],[9,-7.8465],[10,-3.4838],[11,0.3037],[12,2.5893],[13,3.2123],[14,3.0355],[15,2.6681],[16,2.7189],[17,3.2869],[18,4.0132],[19,4.8204],[20,5.548],[21,6.0707],[22,6.3643],[23,6.4043],[24,6.1664],[26,5.3661],[27,5.0008],[28,4.7103],[29,4.5331],[30,4.5078]]
       }
       hand_l_ik_ctrl.Location.Z {
        type=MovieSceneScriptingFloatChannel
        default=50.724
        keys [[0,50.724],[1,56.2502],[2,57.7886],[3,54.6073],[4,48.4682],[5,40.0056],[6,26.329],[7,12.0981],[8,3.2154],[9,-2.0029],[10,-4.1622],[11,-4.5867],[12,-3.773],[13,-3.566],[14,-4.8898],[15,-6.8603],[16,-8.5937],[17,-7.7805],[18,-4.3941],[19,-0.9748],[20,2.0424],[21,5.5151],[22,8.8415],[23,11.4204],[24,12.6504],[25,12.7469],[26,12.4386],[27,11.8912],[28,11.2722],[29,10.7484],[30,10.4864]]
       }
       hand_l_ik_ctrl.Rotation.X {
        type=MovieSceneScriptingFloatChannel
        default=111.4109
        keys [[0,111.4109],[1,118.3751],[2,118.725],[3,103.8922],[4,65.4126],[5,17.3525],[6,-9.8305],[7,-22.1965],[8,-27.8552],[9,-31.5489],[10,-33.7699],[11,-34.2481],[12,-31.9407],[13,-24.6265],[14,-12.8154],[15,0.1723],[16,11.1115],[17,19.1562],[18,25.5316],[19,30.4172],[20,34.8792],[21,39.5374],[22,43.8824],[23,47.2597],[24,48.8505],[25,47.825],[26,44.7096],[27,40.6694],[28,36.7349],[29,33.7782],[30,32.6087]]
       }
       hand_l_ik_ctrl.Rotation.Y {
        type=MovieSceneScriptingFloatChannel
        default=35.8652
        thinned_from=31
        keys [[0,35.8652],[1,40.7624],[2,48.8713],[3,61.1621],[4,69.424],[5,65.072],[6,49.3845],[7,31.7777],[8,18.4984],[9,11.0819],[10,11.2219],[11,14.5284],[12,17.102],[13,17.4962],[14,16.9492],[15,15.6441],[16,14.0656],[17,12.1182],[18,10.3306],[19,10.6803],[20,13.6772],[22,21.6969],[23,24.7076],[24,25.913],[25,25.1289],[26,23.0019],[27,19.9969],[28,16.772],[29,14.1716],[30,13.1072]]
       }
       hand_l_ik_ctrl.Rotation.Z {
        type=MovieSceneScriptingFloatChannel
        default=97.8586
        keys [[0,97.8586],[1,97.3349],[2,91.8077],[3,75.2252],[4,38.0483],[5,-8.2441],[6,-33.9223],[7,-42.9565],[8,-42.0404],[9,-38.4646],[10,-35.7528],[11,-32.5602],[12,-28.2477],[13,-22.6342],[14,-16.0602],[15,-9.296],[16,-3.3017],[17,1.8787],[18,6.1508],[19,9.0399],[20,11.4503],[21,14.3294],[22,17.4317],[23,20.0696],[24,21.2387],[25,20.0681],[26,17.1855],[27,13.8113],[28,10.8804],[29,8.9248],[30,8.214]]
       }
       hand_l_ik_ctrl.Scale.X {type=MovieSceneScriptingFloatChannel, default=1, thinned_from=31, keys [[0,1],[30,1]]}
       hand_l_ik_ctrl.Scale.Y {type=MovieSceneScriptingFloatChannel, default=1, const [1,31]}
       hand_l_ik_ctrl.Scale.Z {type=MovieSceneScriptingFloatChannel, default=1, const [1,31]}
       Softness {type=MovieSceneScriptingFloatChannel, default=0, const [0,31]}
       hand_l_ikorient_ctrl.Rotation.Roll {type=MovieSceneScriptingFloatChannel, default=0, const [0,31]}
       hand_l_ikorient_ctrl.Rotation.Pitch {type=MovieSceneScriptingFloatChannel, default=0, const [0,31]}
       hand_l_ikorient_ctrl.Rotation.Yaw {type=MovieSceneScriptingFloatChannel, default=0, const [0,31]}
       arm_l_pv_ik_ctrl.Location.X {
        type=MovieSceneScriptingFloatChannel
        default=10.0344
        keys [[0,10.0344],[1,11.9324],[2,17.18],[3,22.8321],[4,23.0724],[5,19.5317],[6,11.4471],[7,1.5378],[8,-9.3209],[9,-19.5253],[10,-24.7952],[11,-27.1651],[12,-26.8206],[13,-22.2776],[14,-15.3379],[15,-9.5757],[16,-5.5565],[17,-1.9616],[18,1.8889],[19,4.8062],[20,6.7784],[21,8.9078],[22,11.1911],[23,13.1708],[24,13.6167],[25,13.7567],[26,15.0043],[27,16.3595],[28,17.3378],[29,17.8473],[30,17.9141]]
       }
       arm_l_pv_ik_ctrl.Location.Y {
        type=MovieSceneScriptingFloatChannel
        default=63.1447
        thinned_from=31
        keys [[0,63.1447],[1,64.4828],[2,61.7187],[3,50.6416],[4,37.2325],[5,25.8273],[6,11.623],[7,0.5117],[8,-5.0105],[9,-7.3795],[10,-8.7445],[11,-10.3093],[13,-12.3512],[14,-12.6452],[15,-11.4532],[16,-9.2553],[17,-5.7823],[18,-0.7691],[19,3.4209],[20,5.7748],[21,8.8362],[22,12.5562],[23,16.2092],[24,17.8051],[25,17.9783],[26,18.5595],[27,19.0537],[28,19.3487],[30,19.8257]]
       }
       arm_l_pv_ik_ctrl.Location.Z {
        type=MovieSceneScriptingFloatChannel
        default=13.9242
        keys [[0,13.9242],[1,16.0156],[2,18.2743],[3,17.5558],[4,13.3587],[5,8.9408],[6,4.9728],[7,4.2221],[8,6.2788],[9,9.7787],[10,12.9332],[11,16.4661],[12,18.8852],[13,20.1466],[14,20.3848],[15,18.9597],[16,15.9278],[17,10.9226],[18,5.464],[19,2.4332],[20,1.4774],[21,0.3269],[22,-0.6797],[23,-1.246],[24,-1.6584],[25,-1.6147],[26,-0.752],[27,0.3816],[28,1.337],[29,1.853],[30,1.842]]
       }
       arm_l_pv_ik_ctrl.Rotation.X {
        type=MovieSceneScriptingFloatChannel
        default=3.058
        thinned_from=31
        keys [[0,3.058],[1,-0.7414],[2,2.3722],[3,6.1921],[4,5.6588],[5,4.4147],[6,4.6066],[7,5.0252],[8,8.7012],[10,21.0367],[11,28.7177],[12,35.1384],[13,39.3163],[14,41.4955],[15,40.9827],[16,37.8036],[17,31.1739],[18,22.8837],[19,17.8195],[20,16.1493],[21,13.7584],[22,11.2068],[23,9.1472],[24,7.5364],[25,7.1517],[26,8.4948],[27,10.5535],[28,12.364],[29,13.3497],[30,13.3675]]
       }
       arm_l_pv_ik_ctrl.Rotation.Y {
        type=MovieSceneScriptingFloatChannel
        default=35.7191
        keys [[0,35.7191],[1,41.3091],[2,44.8294],[3,43.2195],[4,37.3826],[5,30.1054],[6,21.1035],[7,18.5151],[8,16.1419],[9,13.6578],[10,12.3525],[11,10.3938],[12,7.363],[13,4.9042],[14,3.1697],[15,1.4546],[16,-0.6693],[17,-3.3916],[18,-5.5213],[19,-6.429],[20,-6.4943],[21,-5.5954],[22,-3.7761],[23,-1.5879],[24,-0.2054],[25,0.4409],[26,1.0637],[27,1.3367],[28,1.2678],[29,1.0514],[30,0.8549]]
       }
       arm_l_pv_ik_ctrl.Rotation.Z {
        type=MovieSceneScriptingFloatChannel
        default=76.9034
        keys [[0,76.9034],[1,71.1898],[2,64.4188],[3,48.5663],[4,28.2644],[5,10.5702],[6,-7.8943],[7,-22.2321],[8,-31.6314],[9,-38.7093],[10,-40.5886],[11,-39.1593],[12,-36.809],[13,-30.8373],[14,-22.4655],[15,-15.7271],[16,-10.3824],[17,-4.0224],[18,4.4322],[19,11.8461],[20,16.6034],[21,21.7697],[22,27.5729],[23,33.0275],[24,34.6123],[25,34.7166],[26,37.1657],[27,39.8199],[28,41.7822],[29,43.0442],[30,43.6766]]
       }
       arm_l_pv_ik_ctrl.Scale.X {type=MovieSceneScriptingFloatChannel, default=1, thinned_from=31, keys [[0,1],[30,1]]}
       arm_l_pv_ik_ctrl.Scale.Y {type=MovieSceneScriptingFloatChannel, default=1, const [1,31]}
       arm_l_pv_ik_ctrl.Scale.Z {type=MovieSceneScriptingFloatChannel, default=1, const [1,31]}
       upperarm_r_fk_ctrl.Location.X {type=MovieSceneScriptingFloatChannel, default=0, thinned_from=31, keys [[0,0],[30,0]]}
       upperarm_r_fk_ctrl.Location.Y {type=MovieSceneScriptingFloatChannel, default=0.0001, thinned_from=31, keys [[0,0.0001],[30,0]]}
       upperarm_r_fk_ctrl.Location.Z {type=MovieSceneScriptingFloatChannel, default=-0.0001, thinned_from=31, keys [[0,-0.0001],[30,-0.0001]]}
       upperarm_r_fk_ctrl.Rotation.X {
        type=MovieSceneScriptingFloatChannel
        default=-39.4084
        thinned_from=30
        keys [[0,-39.4084],[1,-72.3065],[2,-108.1673],[3,-138.0887],[4,-157.4716],[5,-157.5212],[6,-57.7758],[7,-28.9362],[8,-18.4639],[9,-3.8913],[10,17.7791],[12,36.4705],[13,38.8893],[14,37.6283],[15,31.6698],[16,19.6479],[17,3.2734],[18,-3.8349],[19,-5.9359],[20,-6.3538],[21,-6.7166],[22,-6.9233],[25,-7.1624],[26,-7.2237],[27,-7.2639],[28,-7.2735],[29,-7.2624],[30,-6.8624]]
       }
       upperarm_r_fk_ctrl.Rotation.Y {
        type=MovieSceneScriptingFloatChannel
        default=10.9495
        keys [[0,10.9495],[1,9.9598],[2,22.8853],[3,45.7302],[4,58.8494],[5,69.4231],[6,75.1285],[7,51.065],[8,34.6568],[9,13.4948],[10,7.0509],[12,-2.5851],[13,-4.3147],[14,-5.2049],[15,1.1594],[16,10.3201],[17,22.8223],[18,24.7908],[19,24.1802],[20,18.7778],[21,11.9812],[22,5.8082],[23,1.9118],[24,-0.0789],[25,-0.5858],[26,-1.4753],[27,-2.8742],[28,-3.973],[29,-4.7578],[30,-4.8945]]
       }
       upperarm_r_fk_ctrl.Rotation.Z {
        type=MovieSceneScriptingFloatChannel
        default=-48.8845
        thinned_from=31
        keys [[0,-48.8845],[1,-46.2743],[2,-38.075],[3,-37.767],[4,-41.8391],[5,-32.3504],[6,76.3538],[7,103.0437],[8,101.563],[9,95.8848],[10,91.1012],[11,90.0711],[13,86.1035],[14,81.9499],[15,74.0559],[16,58.9657],[17,47.3829],[18,35.9757],[19,27.1769],[20,19.5434],[21,11.8587],[22,5.2786],[23,-1.5161],[24,-8.9251],[25,-16.066],[26,-21.8958],[27,-25.2975],[28,-26.32],[29,-26.3399],[30,-26.2348]]
       }
       upperarm_r_fk_ctrl.Scale.X {type=MovieSceneScriptingFloatChannel, default=1, const [1,31]}
       upperarm_r_fk_ctrl.Scale.Y {type=MovieSceneScriptingFloatChannel, default=1, const [1,31]}
       upperarm_r_fk_ctrl.Scale.Z {type=MovieSceneScriptingFloatChannel, default=1, const [1,31]}
       lowerarm_r_fk_ctrl.Location.X {type=MovieSceneScriptingFloatChannel, default=0, thinned_from=31, keys [[0,0],[30,0]]}
       lowerarm_r_fk_ctrl.Location.Y {type=MovieSceneScriptingFloatChannel, default=0, thinned_from=31, keys [[0,0],[30,0]]}
       lowerarm_r_fk_ctrl.Location.Z {type=MovieSceneScriptingFloatChannel, default=0, thinned_from=31, keys [[0,0],[30,0]]}
       lowerarm_r_fk_ctrl.Rotation.X {
        type=MovieSceneScriptingFloatChannel
        default=-3.4203
        thinned_from=31
        keys [[0,-3.4203],[1,-3.3066],[2,-3.2993],[3,-3.3685],[4,-3.3917],[5,-3.322],[6,-3.3035],[7,-3.4996],[8,-4.1714],[9,-4.174],[10,-4.0685],[11,-4.1183],[12,-4.0907],[13,-3.9056],[14,-3.7372],[15,-3.6091],[16,-3.4903],[17,-3.409],[18,-3.3762],[19,-3.3806],[20,-3.3956],[21,-3.4276],[22,-3.4665],[23,-3.4879],[24,-3.4924],[27,-3.473],[30,-3.4944]]
       }
       lowerarm_r_fk_ctrl.Rotation.Y {
        type=MovieSceneScriptingFloatChannel
        default=1.7824
        thinned_from=31
        keys [[0,1.7824],[1,1.592],[2,1.576],[3,1.706],[4,1.7418],[5,1.6234],[6,1.5853],[7,1.8779],[8,2.2264],[9,2.2267],[10,2.2075],[11,2.2179],[12,2.2124],[13,2.1564],[14,2.0721],[15,1.981],[16,1.8677],[17,1.7668],[18,1.7182],[19,1.7251],[20,1.7475],[21,1.792],[22,1.8405],[23,1.865],[24,1.87],[26,1.8534],[27,1.8482],[28,1.8545],[29,1.8661],[30,1.8722]]
       }
       lowerarm_r_fk_ctrl.Rotation.Z {
        type=MovieSceneScriptingFloatChannel
        default=39.8843
        keys [[0,39.8843],[1,51.2298],[2,52.1278],[3,44.6021],[4,42.4201],[5,49.4413],[6,51.6027],[7,33.5416],[8,-5.8888],[9,-6.0228],[10,-0.5449],[11,-3.1413],[12,-1.7041],[13,8.1868],[14,17.8134],[15,25.8529],[16,34.2469],[17,40.8643],[18,43.8609],[19,43.444],[20,42.0664],[21,39.2659],[22,36.0911],[23,34.4285],[24,34.0871],[25,34.543],[26,35.2214],[27,35.5749],[28,35.1488],[29,34.3576],[30,33.9339]]
       }
       lowerarm_r_fk_ctrl.Scale.X {type=MovieSceneScriptingFloatChannel, default=1, const [1,31]}
       lowerarm_r_fk_ctrl.Scale.Y {type=MovieSceneScriptingFloatChannel, default=1, const [1,31]}
       lowerarm_r_fk_ctrl.Scale.Z {type=MovieSceneScriptingFloatChannel, default=1, const [1,31]}
       hand_r_fk_ctrl.Location.X {type=MovieSceneScriptingFloatChannel, default=0, thinned_from=31, keys [[0,0],[30,0]]}
       hand_r_fk_ctrl.Location.Y {type=MovieSceneScriptingFloatChannel, default=0, thinned_from=31, keys [[0,0],[30,0]]}
       hand_r_fk_ctrl.Location.Z {type=MovieSceneScriptingFloatChannel, default=0, thinned_from=31, keys [[0,0],[30,0]]}
       hand_r_fk_ctrl.Rotation.X {
        type=MovieSceneScriptingFloatChannel
        default=-1.4739
        keys [[0,-1.4739],[1,-2.2893],[2,-1.8115],[3,6.1515],[4,13.6977],[5,7.9431],[6,-7.6048],[7,3.2014],[8,12.1373],[9,7.3711],[10,-1.3769],[11,-4.763],[12,1.0218],[13,7.8609],[14,8.7055],[15,8.7189],[16,8.9835],[17,12.2062],[18,15.6628],[19,16.7576],[20,18.4249],[21,20.8468],[22,21.022],[23,18.6274],[24,16.0259],[25,13.2291],[26,10.3424],[27,7.9552],[28,7.2064],[29,7.7501],[30,8.2404]]
       }
       hand_r_fk_ctrl.Rotation.Y {
        type=MovieSceneScriptingFloatChannel
        default=1.5492
        keys [[0,1.5492],[1,9.4314],[2,11.7468],[3,14.4409],[4,23.049],[5,21.389],[6,22.1646],[7,51.5125],[8,18.9391],[9,10.3892],[10,0.5175],[11,-2.7656],[12,-1.5093],[13,1.8683],[14,1.9097],[15,3.3328],[16,5.0419],[17,4.2055],[18,1.9014],[19,0.8876],[20,3.6533],[21,7.7221],[22,11.0265],[23,10.1748],[24,5.2104],[25,-1.3705],[26,-7.3983],[27,-11.1337],[28,-12.1829],[29,-11.9283],[30,-11.5803]]
       }
       hand_r_fk_ctrl.Rotation.Z {
        type=MovieSceneScriptingFloatChannel
        default=-26.1039
        thinned_from=31
        keys [[0,-26.1039],[1,-28.2798],[2,-9.5766],[3,6.705],[4,14.2871],[5,2.6213],[6,-15.0886],[7,-37.4776],[8,-34.8233],[9,-30.9487],[10,-30.0475],[11,-26.9974],[12,-22.505],[13,-12.6777],[14,-3.0951],[15,-2.6657],[16,-5.3573],[17,-7.0895],[18,-6.3983],[19,-5.8697],[20,-4.8962],[21,-3.9032],[22,-5.8414],[23,-9.7455],[24,-11.8521],[25,-12.1988],[26,-11.5092],[27,-11.0457],[28,-11.1211],[30,-11.1582]]
       }
       hand_r_fk_ctrl.Scale.X {type=MovieSceneScriptingFloatChannel, default=1, const [1,31]}
       hand_r_fk_ctrl.Scale.Y {type=MovieSceneScriptingFloatChannel, default=1, const [1,31]}
       hand_r_fk_ctrl.Scale.Z {type=MovieSceneScriptingFloatChannel, default=1, const [1,31]}
       thumb_01_r_ctrl.Location.X {type=MovieSceneScriptingFloatChannel, default=0, thinned_from=31, keys [[0,0],[30,0]]}
       thumb_01_r_ctrl.Location.Y {type=MovieSceneScriptingFloatChannel, default=0, thinned_from=31, keys [[0,0],[30,0]]}
       thumb_01_r_ctrl.Location.Z {type=MovieSceneScriptingFloatChannel, default=0, thinned_from=31, keys [[0,0],[30,0]]}
       thumb_01_r_ctrl.Rotation.X {type=MovieSceneScriptingFloatChannel, default=15.1661, thinned_from=31, keys [[0,15.1661],[7,15.1661],[8,9.8335],[30,9.8335]]}
       thumb_01_r_ctrl.Rotation.Y {type=MovieSceneScriptingFloatChannel, default=-25.3344, thinned_from=31, keys [[0,-25.3344],[7,-25.3344],[8,1.6409],[30,1.6409]]}
       thumb_01_r_ctrl.Rotation.Z {type=MovieSceneScriptingFloatChannel, default=-2.1593, thinned_from=31, keys [[0,-2.1593],[7,-2.1593],[8,0.7484],[30,0.7484]]}
       thumb_01_r_ctrl.Scale.X {type=MovieSceneScriptingFloatChannel, default=1, const [1,31]}
       thumb_01_r_ctrl.Scale.Y {type=MovieSceneScriptingFloatChannel, default=1, const [1,31]}
       thumb_01_r_ctrl.Scale.Z {type=MovieSceneScriptingFloatChannel, default=1, const [1,31]}
       thumb_02_r_ctrl.Location.X {type=MovieSceneScriptingFloatChannel, default=0, thinned_from=31, keys [[0,0],[30,0]]}
       thumb_02_r_ctrl.Location.Y {type=MovieSceneScriptingFloatChannel, default=0, thinned_from=31, keys [[0,0],[30,0]]}
       thumb_02_r_ctrl.Location.Z {type=MovieSceneScriptingFloatChannel, default=0, thinned_from=31, keys [[0,0],[30,0]]}
       thumb_02_r_ctrl.Rotation.X {type=MovieSceneScriptingFloatChannel, default=0, thinned_from=31, keys [[0,0],[30,0]]}
       thumb_02_r_ctrl.Rotation.Y {type=MovieSceneScriptingFloatChannel, default=0, thinned_from=31, keys [[0,0],[30,0]]}
       thumb_02_r_ctrl.Rotation.Z {type=MovieSceneScriptingFloatChannel, default=0, thinned_from=31, keys [[0,0],[7,0],[8,23.8951],[30,23.8951]]}
       thumb_02_r_ctrl.Scale.X {type=MovieSceneScriptingFloatChannel, default=1, const [1,31]}
       thumb_02_r_ctrl.Scale.Y {type=MovieSceneScriptingFloatChannel, default=1, const [1,31]}
       thumb_02_r_ctrl.Scale.Z {type=MovieSceneScriptingFloatChannel, default=1, const [1,31]}
       thumb_03_r_ctrl.Location.X {type=MovieSceneScriptingFloatChannel, default=0, thinned_from=31, keys [[0,0],[30,0]]}
       thumb_03_r_ctrl.Location.Y {type=MovieSceneScriptingFloatChannel, default=0, thinned_from=31, keys [[0,0],[30,0]]}
       thumb_03_r_ctrl.Location.Z {type=MovieSceneScriptingFloatChannel, default=0, thinned_from=31, keys [[0,0],[30,0]]}
       thumb_03_r_ctrl.Rotation.X {type=MovieSceneScriptingFloatChannel, default=0, thinned_from=31, keys [[0,0],[30,0]]}
       thumb_03_r_ctrl.Rotation.Y {type=MovieSceneScriptingFloatChannel, default=0, thinned_from=31, keys [[0,0],[30,0]]}
       thumb_03_r_ctrl.Rotation.Z {type=MovieSceneScriptingFloatChannel, default=-24.6415, thinned_from=31, keys [[0,-24.6415],[7,-24.6415],[8,7.3553],[30,7.3553]]}
       thumb_03_r_ctrl.Scale.X {type=MovieSceneScriptingFloatChannel, default=1, const [1,31]}
       thumb_03_r_ctrl.Scale.Y {type=MovieSceneScriptingFloatChannel, default=1, const [1,31]}
       thumb_03_r_ctrl.Scale.Z {type=MovieSceneScriptingFloatChannel, default=1, const [1,31]}
       index_metacarpal_r_ctrl.Location.X {type=MovieSceneScriptingFloatChannel, default=0, thinned_from=31, keys [[0,0],[30,0]]}
       index_metacarpal_r_ctrl.Location.Y {type=MovieSceneScriptingFloatChannel, default=0, thinned_from=31, keys [[0,0],[30,0]]}
       index_metacarpal_r_ctrl.Location.Z {type=MovieSceneScriptingFloatChannel, default=0, thinned_from=31, keys [[0,0],[30,0]]}
       index_metacarpal_r_ctrl.Rotation.X {type=MovieSceneScriptingFloatChannel, default=0, thinned_from=31, keys [[0,0],[30,0]]}
       index_metacarpal_r_ctrl.Rotation.Y {type=MovieSceneScriptingFloatChannel, default=0, thinned_from=31, keys [[0,0],[7,0],[8,0.57],[30,0.57]]}
       index_metacarpal_r_ctrl.Rotation.Z {type=MovieSceneScriptingFloatChannel, default=0, thinned_from=31, keys [[0,0],[30,0]]}
       index_metacarpal_r_ctrl.Scale.X {type=MovieSceneScriptingFloatChannel, default=1, const [1,31]}
       index_metacarpal_r_ctrl.Scale.Y {type=MovieSceneScriptingFloatChannel, default=1, const [1,31]}
       index_metacarpal_r_ctrl.Scale.Z {type=MovieSceneScriptingFloatChannel, default=1, const [1,31]}
       index_01_r_ctrl.Location.X {type=MovieSceneScriptingFloatChannel, default=0, thinned_from=31, keys [[0,0],[30,0]]}
       index_01_r_ctrl.Location.Y {type=MovieSceneScriptingFloatChannel, default=0, thinned_from=31, keys [[0,0],[30,0]]}
       index_01_r_ctrl.Location.Z {type=MovieSceneScriptingFloatChannel, default=0, thinned_from=31, keys [[0,0],[30,0]]}
       index_01_r_ctrl.Rotation.X {type=MovieSceneScriptingFloatChannel, default=-1.6158, thinned_from=31, keys [[0,-1.6158],[7,-1.6158],[8,4.6618],[30,4.6618]]}
       index_01_r_ctrl.Rotation.Y {type=MovieSceneScriptingFloatChannel, default=-0.9096, thinned_from=31, keys [[0,-0.9096],[7,-0.9096],[8,-8.9751],[30,-8.9751]]}
       index_01_r_ctrl.Rotation.Z {type=MovieSceneScriptingFloatChannel, default=-14.4208, thinned_from=31, keys [[0,-14.4208],[7,-14.4208],[8,9.0783],[30,9.0783]]}
       index_01_r_ctrl.Scale.X {type=MovieSceneScriptingFloatChannel, default=1, const [1,31]}
       index_01_r_ctrl.Scale.Y {type=MovieSceneScriptingFloatChannel, default=1, const [1,31]}
       index_01_r_ctrl.Scale.Z {type=MovieSceneScriptingFloatChannel, default=1, const [1,31]}
       index_02_r_ctrl.Location.X {type=MovieSceneScriptingFloatChannel, default=0, thinned_from=31, keys [[0,0],[30,0]]}
       index_02_r_ctrl.Location.Y {type=MovieSceneScriptingFloatChannel, default=0, thinned_from=31, keys [[0,0],[30,0]]}
       index_02_r_ctrl.Location.Z {type=MovieSceneScriptingFloatChannel, default=0, thinned_from=31, keys [[0,0],[30,0]]}
       index_02_r_ctrl.Rotation.X {type=MovieSceneScriptingFloatChannel, default=0, thinned_from=31, keys [[0,0],[30,0]]}
       index_02_r_ctrl.Rotation.Y {type=MovieSceneScriptingFloatChannel, default=0, thinned_from=31, keys [[0,0],[30,0]]}
       index_02_r_ctrl.Rotation.Z {type=MovieSceneScriptingFloatChannel, default=-4.4334, thinned_from=31, keys [[0,-4.4334],[7,-4.4334],[8,29.7411],[30,29.7411]]}
       index_02_r_ctrl.Scale.X {type=MovieSceneScriptingFloatChannel, default=1, const [1,31]}
       index_02_r_ctrl.Scale.Y {type=MovieSceneScriptingFloatChannel, default=1, const [1,31]}
       index_02_r_ctrl.Scale.Z {type=MovieSceneScriptingFloatChannel, default=1, const [1,31]}
       index_03_r_ctrl.Location.X {type=MovieSceneScriptingFloatChannel, default=0, thinned_from=31, keys [[0,0],[30,0]]}
       index_03_r_ctrl.Location.Y {type=MovieSceneScriptingFloatChannel, default=0.0001, thinned_from=31, keys [[0,0.0001],[30,0.0001]]}
       index_03_r_ctrl.Location.Z {type=MovieSceneScriptingFloatChannel, default=-0.0001, thinned_from=31, keys [[0,-0.0001],[30,-0.0001]]}
       index_03_r_ctrl.Rotation.X {type=MovieSceneScriptingFloatChannel, default=0, thinned_from=31, keys [[0,0],[30,0]]}
       index_03_r_ctrl.Rotation.Y {type=MovieSceneScriptingFloatChannel, default=0, thinned_from=31, keys [[0,0],[30,0]]}
       index_03_r_ctrl.Rotation.Z {type=MovieSceneScriptingFloatChannel, default=-7.9391, thinned_from=31, keys [[0,-7.9391],[7,-7.9391],[8,18.4001],[30,18.4001]]}
       index_03_r_ctrl.Scale.X {type=MovieSceneScriptingFloatChannel, default=1, const [1,31]}
       index_03_r_ctrl.Scale.Y {type=MovieSceneScriptingFloatChannel, default=1, const [1,31]}
       index_03_r_ctrl.Scale.Z {type=MovieSceneScriptingFloatChannel, default=1, const [1,31]}
       middle_metacarpal_r_ctrl.Location.X {type=MovieSceneScriptingFloatChannel, default=0, thinned_from=31, keys [[0,0],[30,0]]}
       middle_metacarpal_r_ctrl.Location.Y {type=MovieSceneScriptingFloatChannel, default=0, thinned_from=31, keys [[0,0],[30,0]]}
       middle_metacarpal_r_ctrl.Location.Z {type=MovieSceneScriptingFloatChannel, default=0, thinned_from=31, keys [[0,0],[30,0]]}
       middle_metacarpal_r_ctrl.Rotation.X {type=MovieSceneScriptingFloatChannel, default=0, thinned_from=31, keys [[0,0],[30,0]]}
       middle_metacarpal_r_ctrl.Rotation.Y {type=MovieSceneScriptingFloatChannel, default=0, thinned_from=31, keys [[0,0],[7,0],[8,0.1155],[30,0.1155]]}
       middle_metacarpal_r_ctrl.Rotation.Z {type=MovieSceneScriptingFloatChannel, default=0, thinned_from=31, keys [[0,0],[30,0]]}
       middle_metacarpal_r_ctrl.Scale.X {type=MovieSceneScriptingFloatChannel, default=1, const [1,31]}
       middle_metacarpal_r_ctrl.Scale.Y {type=MovieSceneScriptingFloatChannel, default=1, const [1,31]}
       middle_metacarpal_r_ctrl.Scale.Z {type=MovieSceneScriptingFloatChannel, default=1, const [1,31]}
       middle_01_r_ctrl.Location.X {type=MovieSceneScriptingFloatChannel, default=0, thinned_from=31, keys [[0,0],[30,0]]}
       middle_01_r_ctrl.Location.Y {type=MovieSceneScriptingFloatChannel, default=0, thinned_from=31, keys [[0,0],[30,0]]}
       middle_01_r_ctrl.Location.Z {type=MovieSceneScriptingFloatChannel, default=0, thinned_from=31, keys [[0,0],[30,0]]}
       middle_01_r_ctrl.Rotation.X {type=MovieSceneScriptingFloatChannel, default=-2.5986, thinned_from=31, keys [[0,-2.5986],[7,-2.5986],[8,-0.2193],[30,-0.2193]]}
       middle_01_r_ctrl.Rotation.Y {type=MovieSceneScriptingFloatChannel, default=-2.1686, thinned_from=31, keys [[0,-2.1686],[7,-2.1686],[8,-1.9525],[30,-1.9525]]}
       middle_01_r_ctrl.Rotation.Z {type=MovieSceneScriptingFloatChannel, default=-19.3508, thinned_from=31, keys [[0,-19.3508],[7,-19.3508],[8,1.283],[30,1.283]]}
       middle_01_r_ctrl.Scale.X {type=MovieSceneScriptingFloatChannel, default=1, const [1,31]}
       middle_01_r_ctrl.Scale.Y {type=MovieSceneScriptingFloatChannel, default=1, const [1,31]}
       middle_01_r_ctrl.Scale.Z {type=MovieSceneScriptingFloatChannel, default=1, const [1,31]}
       middle_02_r_ctrl.Location.X {type=MovieSceneScriptingFloatChannel, default=0.0009, thinned_from=31, keys [[0,0.0009],[30,0.0009]]}
       middle_02_r_ctrl.Location.Y {type=MovieSceneScriptingFloatChannel, default=0.0008, thinned_from=31, keys [[0,0.0008],[30,0.0008]]}
       middle_02_r_ctrl.Location.Z {type=MovieSceneScriptingFloatChannel, default=0.0476, const [0.0476,31]}
       middle_02_r_ctrl.Rotation.X {type=MovieSceneScriptingFloatChannel, default=-0.0078, thinned_from=31, keys [[0,-0.0078],[7,-0.0078],[8,-0.0098],[30,-0.0098]]}
       middle_02_r_ctrl.Rotation.Y {type=MovieSceneScriptingFloatChannel, default=-0.026, thinned_from=31, keys [[0,-0.026],[7,-0.026],[8,-0.0094],[30,-0.0094]]}
       middle_02_r_ctrl.Rotation.Z {type=MovieSceneScriptingFloatChannel, default=-12.5587, thinned_from=31, keys [[0,-12.5587],[7,-12.5587],[8,17.0567],[30,17.0567]]}
       middle_02_r_ctrl.Scale.X {type=MovieSceneScriptingFloatChannel, default=1, const [1,31]}
       middle_02_r_ctrl.Scale.Y {type=MovieSceneScriptingFloatChannel, default=1, const [1,31]}
       middle_02_r_ctrl.Scale.Z {type=MovieSceneScriptingFloatChannel, default=1, const [1,31]}
       middle_03_r_ctrl.Location.X {type=MovieSceneScriptingFloatChannel, default=0.0005, thinned_from=31, keys [[0,0.0005],[30,0.0005]]}
       middle_03_r_ctrl.Location.Y {type=MovieSceneScriptingFloatChannel, default=0, thinned_from=31, keys [[0,0],[30,0]]}
       middle_03_r_ctrl.Location.Z {type=MovieSceneScriptingFloatChannel, default=0.0061, thinned_from=31, keys [[0,0.0061],[30,0.0061]]}
       middle_03_r_ctrl.Rotation.X {type=MovieSceneScriptingFloatChannel, default=-0.0219, thinned_from=31, keys [[0,-0.0219],[7,-0.0219],[8,-0.0083],[30,-0.0083]]}
       middle_03_r_ctrl.Rotation.Y {type=MovieSceneScriptingFloatChannel, default=0.0029, thinned_from=31, keys [[0,0.0029],[7,0.0029],[8,0.0073],[30,0.0073]]}
       middle_03_r_ctrl.Rotation.Z {type=MovieSceneScriptingFloatChannel, default=-6.4756, thinned_from=31, keys [[0,-6.4756],[7,-6.4756],[8,27.2144],[30,27.2144]]}
       middle_03_r_ctrl.Scale.X {type=MovieSceneScriptingFloatChannel, default=1, const [1,31]}
       middle_03_r_ctrl.Scale.Y {type=MovieSceneScriptingFloatChannel, default=1, const [1,31]}
       middle_03_r_ctrl.Scale.Z {type=MovieSceneScriptingFloatChannel, default=1, const [1,31]}
       ring_metacarpal_r_ctrl.Location.X {type=MovieSceneScriptingFloatChannel, default=0, thinned_from=31, keys [[0,0],[30,0]]}
       ring_metacarpal_r_ctrl.Location.Y {type=MovieSceneScriptingFloatChannel, default=0, thinned_from=31, keys [[0,0],[30,0]]}
       ring_metacarpal_r_ctrl.Location.Z {type=MovieSceneScriptingFloatChannel, default=0, thinned_from=31, keys [[0,0],[30,0]]}
       ring_metacarpal_r_ctrl.Rotation.X {type=MovieSceneScriptingFloatChannel, default=0, thinned_from=31, keys [[0,0],[30,0]]}
       ring_metacarpal_r_ctrl.Rotation.Y {type=MovieSceneScriptingFloatChannel, default=0, thinned_from=31, keys [[0,0],[7,0],[8,-0.6309],[30,-0.6309]]}
       ring_metacarpal_r_ctrl.Rotation.Z {type=MovieSceneScriptingFloatChannel, default=0, thinned_from=31, keys [[0,0],[30,0]]}
       ring_metacarpal_r_ctrl.Scale.X {type=MovieSceneScriptingFloatChannel, default=1, const [1,31]}
       ring_metacarpal_r_ctrl.Scale.Y {type=MovieSceneScriptingFloatChannel, default=1, const [1,31]}
       ring_metacarpal_r_ctrl.Scale.Z {type=MovieSceneScriptingFloatChannel, default=1, const [1,31]}
       ring_01_r_ctrl.Location.X {type=MovieSceneScriptingFloatChannel, default=0, thinned_from=31, keys [[0,0],[30,0]]}
       ring_01_r_ctrl.Location.Y {type=MovieSceneScriptingFloatChannel, default=0, thinned_from=31, keys [[0,0],[30,0]]}
       ring_01_r_ctrl.Location.Z {type=MovieSceneScriptingFloatChannel, default=0, thinned_from=31, keys [[0,0],[30,0]]}
       ring_01_r_ctrl.Rotation.X {type=MovieSceneScriptingFloatChannel, default=-1.6667, thinned_from=31, keys [[0,-1.6667],[7,-1.6667],[8,-0.6532],[30,-0.6532]]}
       ring_01_r_ctrl.Rotation.Y {type=MovieSceneScriptingFloatChannel, default=-1.5698, thinned_from=31, keys [[0,-1.5698],[7,-1.5698],[8,0.2712],[30,0.2712]]}
       ring_01_r_ctrl.Rotation.Z {type=MovieSceneScriptingFloatChannel, default=-13.1407, thinned_from=31, keys [[0,-13.1407],[7,-13.1407],[8,-2.3913],[30,-2.3913]]}
       ring_01_r_ctrl.Scale.X {type=MovieSceneScriptingFloatChannel, default=1, const [1,31]}
       ring_01_r_ctrl.Scale.Y {type=MovieSceneScriptingFloatChannel, default=1, const [1,31]}
       ring_01_r_ctrl.Scale.Z {type=MovieSceneScriptingFloatChannel, default=1, const [1,31]}
       ring_02_r_ctrl.Location.X {type=MovieSceneScriptingFloatChannel, default=0, thinned_from=31, keys [[0,0],[30,0]]}
       ring_02_r_ctrl.Location.Y {type=MovieSceneScriptingFloatChannel, default=0, thinned_from=31, keys [[0,0],[30,0]]}
       ring_02_r_ctrl.Location.Z {type=MovieSceneScriptingFloatChannel, default=0, thinned_from=31, keys [[0,0],[30,0]]}
       ring_02_r_ctrl.Rotation.X {type=MovieSceneScriptingFloatChannel, default=0, thinned_from=31, keys [[0,0],[30,0]]}
       ring_02_r_ctrl.Rotation.Y {type=MovieSceneScriptingFloatChannel, default=0, thinned_from=31, keys [[0,0],[30,0]]}
       ring_02_r_ctrl.Rotation.Z {type=MovieSceneScriptingFloatChannel, default=-14.3916, thinned_from=31, keys [[0,-14.3916],[7,-14.3916],[8,12.4006],[30,12.4006]]}
       ring_02_r_ctrl.Scale.X {type=MovieSceneScriptingFloatChannel, default=1, const [1,31]}
       ring_02_r_ctrl.Scale.Y {type=MovieSceneScriptingFloatChannel, default=1, const [1,31]}
       ring_02_r_ctrl.Scale.Z {type=MovieSceneScriptingFloatChannel, default=1, const [1,31]}
       ring_03_r_ctrl.Location.X {type=MovieSceneScriptingFloatChannel, default=0, thinned_from=31, keys [[0,0],[30,0]]}
       ring_03_r_ctrl.Location.Y {type=MovieSceneScriptingFloatChannel, default=0, thinned_from=31, keys [[0,0],[30,0]]}
       ring_03_r_ctrl.Location.Z {type=MovieSceneScriptingFloatChannel, default=0, thinned_from=31, keys [[0,0],[30,0]]}
       ring_03_r_ctrl.Rotation.X {type=MovieSceneScriptingFloatChannel, default=0, thinned_from=31, keys [[0,0],[30,0]]}
       ring_03_r_ctrl.Rotation.Y {type=MovieSceneScriptingFloatChannel, default=0, thinned_from=31, keys [[0,0],[30,0]]}
       ring_03_r_ctrl.Rotation.Z {type=MovieSceneScriptingFloatChannel, default=-20.52, thinned_from=31, keys [[0,-20.52],[7,-20.52],[8,2.7965],[30,2.7965]]}
       ring_03_r_ctrl.Scale.X {type=MovieSceneScriptingFloatChannel, default=1, const [1,31]}
       ring_03_r_ctrl.Scale.Y {type=MovieSceneScriptingFloatChannel, default=1, const [1,31]}
       ring_03_r_ctrl.Scale.Z {type=MovieSceneScriptingFloatChannel, default=1, const [1,31]}
       pinky_metacarpal_r_ctrl.Location.X {type=MovieSceneScriptingFloatChannel, default=0, thinned_from=31, keys [[0,0],[30,0]]}
       pinky_metacarpal_r_ctrl.Location.Y {type=MovieSceneScriptingFloatChannel, default=0, thinned_from=31, keys [[0,0],[30,0]]}
       pinky_metacarpal_r_ctrl.Location.Z {type=MovieSceneScriptingFloatChannel, default=0, thinned_from=31, keys [[0,0],[30,0]]}
       pinky_metacarpal_r_ctrl.Rotation.X {type=MovieSceneScriptingFloatChannel, default=0, thinned_from=31, keys [[0,0],[30,0]]}
       pinky_metacarpal_r_ctrl.Rotation.Y {type=MovieSceneScriptingFloatChannel, default=0, thinned_from=31, keys [[0,0],[7,0],[8,-1.2466],[30,-1.2466]]}
       pinky_metacarpal_r_ctrl.Rotation.Z {type=MovieSceneScriptingFloatChannel, default=0, thinned_from=31, keys [[0,0],[30,0]]}
       pinky_metacarpal_r_ctrl.Scale.X {type=MovieSceneScriptingFloatChannel, default=1, const [1,31]}
       pinky_metacarpal_r_ctrl.Scale.Y {type=MovieSceneScriptingFloatChannel, default=1, const [1,31]}
       pinky_metacarpal_r_ctrl.Scale.Z {type=MovieSceneScriptingFloatChannel, default=1, const [1,31]}
       pinky_01_r_ctrl.Location.X {type=MovieSceneScriptingFloatChannel, default=0, thinned_from=31, keys [[0,0],[30,0]]}
       pinky_01_r_ctrl.Location.Y {type=MovieSceneScriptingFloatChannel, default=0, thinned_from=31, keys [[0,0],[30,0]]}
       pinky_01_r_ctrl.Location.Z {type=MovieSceneScriptingFloatChannel, default=0, thinned_from=31, keys [[0,0],[30,0]]}
       pinky_01_r_ctrl.Rotation.X {type=MovieSceneScriptingFloatChannel, default=3.2666, thinned_from=31, keys [[0,3.2666],[7,3.2666],[8,0],[30,0]]}
       pinky_01_r_ctrl.Rotation.Y {type=MovieSceneScriptingFloatChannel, default=-5.7299, thinned_from=31, keys [[0,-5.7299],[7,-5.7299],[8,5.6059],[30,5.6059]]}
       pinky_01_r_ctrl.Rotation.Z {type=MovieSceneScriptingFloatChannel, default=-4.1286, thinned_from=31, keys [[0,-4.1286],[7,-4.1286],[8,-0.5827],[30,-0.5827]]}
       pinky_01_r_ctrl.Scale.X {type=MovieSceneScriptingFloatChannel, default=1, const [1,31]}
       pinky_01_r_ctrl.Scale.Y {type=MovieSceneScriptingFloatChannel, default=1, const [1,31]}
       pinky_01_r_ctrl.Scale.Z {type=MovieSceneScriptingFloatChannel, default=1, const [1,31]}
       pinky_02_r_ctrl.Location.X {type=MovieSceneScriptingFloatChannel, default=0, thinned_from=31, keys [[0,0],[30,0]]}
       pinky_02_r_ctrl.Location.Y {type=MovieSceneScriptingFloatChannel, default=0, thinned_from=31, keys [[0,0],[30,0]]}
       pinky_02_r_ctrl.Location.Z {type=MovieSceneScriptingFloatChannel, default=0, thinned_from=31, keys [[0,0],[30,0]]}
       pinky_02_r_ctrl.Rotation.X {type=MovieSceneScriptingFloatChannel, default=0, thinned_from=31, keys [[0,0],[30,0]]}
       pinky_02_r_ctrl.Rotation.Y {type=MovieSceneScriptingFloatChannel, default=0, thinned_from=31, keys [[0,0],[30,0]]}
       pinky_02_r_ctrl.Rotation.Z {type=MovieSceneScriptingFloatChannel, default=-13.151, thinned_from=31, keys [[0,-13.151],[7,-13.151],[8,-2.4658],[30,-2.4658]]}
       pinky_02_r_ctrl.Scale.X {type=MovieSceneScriptingFloatChannel, default=1, const [1,31]}
       pinky_02_r_ctrl.Scale.Y {type=MovieSceneScriptingFloatChannel, default=1, const [1,31]}
       pinky_02_r_ctrl.Scale.Z {type=MovieSceneScriptingFloatChannel, default=1, const [1,31]}
       pinky_03_r_ctrl.Location.X {type=MovieSceneScriptingFloatChannel, default=0, thinned_from=31, keys [[0,0],[30,0]]}
       pinky_03_r_ctrl.Location.Y {type=MovieSceneScriptingFloatChannel, default=0, thinned_from=31, keys [[0,0],[30,0]]}
       pinky_03_r_ctrl.Location.Z {type=MovieSceneScriptingFloatChannel, default=0, thinned_from=31, keys [[0,0],[30,0]]}
       pinky_03_r_ctrl.Rotation.X {type=MovieSceneScriptingFloatChannel, default=0, thinned_from=31, keys [[0,0],[30,0]]}
       pinky_03_r_ctrl.Rotation.Y {type=MovieSceneScriptingFloatChannel, default=0, thinned_from=31, keys [[0,0],[30,0]]}
       pinky_03_r_ctrl.Rotation.Z {type=MovieSceneScriptingFloatChannel, default=-22.5109, thinned_from=31, keys [[0,-22.5109],[7,-22.5109],[8,-4.2208],[30,-4.2208]]}
       pinky_03_r_ctrl.Scale.X {type=MovieSceneScriptingFloatChannel, default=1, const [1,31]}
       pinky_03_r_ctrl.Scale.Y {type=MovieSceneScriptingFloatChannel, default=1, const [1,31]}
       pinky_03_r_ctrl.Scale.Z {type=MovieSceneScriptingFloatChannel, default=1, const [1,31]}
       hand_r_ik_ctrl.Location.X {
        type=MovieSceneScriptingFloatChannel
        default=-42.8119
        thinned_from=31
        keys [[0,-42.8119],[1,-54.6874],[2,-66.7809],[3,-76.3212],[4,-79.1002],[5,-75.2966],[6,-59.5797],[7,-32.8735],[8,-3.0071],[9,13.3272],[10,13.4288],[11,11.5324],[12,8.6923],[13,5.813],[14,3.711],[15,2.5292],[16,1.8226],[17,1.7317],[18,1.8646],[19,1.462],[20,0.0758],[21,-1.8454],[22,-3.7158],[23,-5.4535],[24,-7.2424],[25,-8.8957],[26,-10.2253],[27,-11.0448],[28,-11.2845],[30,-11.1087]]
       }
       hand_r_ik_ctrl.Location.Y {
        type=MovieSceneScriptingFloatChannel
        default=9.1082
        keys [[0,9.1082],[1,6.2609],[2,2.2687],[3,-2.466],[4,-7.7847],[5,-16.3192],[6,-29.1638],[7,-23.9901],[8,2.5796],[9,31.2568],[10,44.5872],[11,52.2564],[12,54.7676],[13,53.151],[14,49.3679],[15,44.3039],[16,37.988],[17,30.4607],[18,22.4682],[19,15.6829],[20,10.3261],[21,6.1768],[22,4.244],[23,4.035],[24,4.2524],[25,4.6996],[26,5.1804],[27,5.4975],[28,5.6828],[29,5.8488],[30,5.9207]]
       }
       hand_r_ik_ctrl.Location.Z {
        type=MovieSceneScriptingFloatChannel
        default=-39.6867
        keys [[0,-39.6867],[1,-35.2086],[2,-28.637],[3,-20.6758],[4,-12.1021],[5,0.9559],[6,28.2579],[7,63.3692],[8,77.8393],[9,61.8546],[10,52.4861],[11,44.6493],[12,38.0732],[13,32.9041],[14,29.4231],[15,26.905],[16,24.261],[17,21.5232],[18,18.6507],[19,14.9201],[20,9.5938],[21,3.4092],[22,-2.1444],[23,-6.5021],[24,-10.0911],[25,-12.9107],[26,-14.9611],[27,-16.2409],[28,-16.9251],[29,-17.2585],[30,-17.3463]]
       }
       hand_r_ik_ctrl.Rotation.X {
        type=MovieSceneScriptingFloatChannel
        default=6.4263
        keys [[0,6.4263],[1,9.4321],[2,-3.4318],[3,-15.8251],[4,-28.1647],[5,-53.0814],[6,-100.575],[7,-103.851],[8,-64.321],[9,-1.6618],[10,31.0647],[11,55.2465],[12,74.2951],[13,81.4548],[14,75.6949],[15,68.7621],[16,60.9971],[17,49.9373],[18,34.8223],[19,20.4691],[20,11.1392],[21,5.8282],[22,1.8464],[23,-1.8435],[24,-5.0336],[25,-8.6503],[26,-12.6544],[27,-15.8469],[28,-16.6994],[29,-15.9623],[30,-15.3259]]
       }
       hand_r_ik_ctrl.Rotation.Y {
        type=MovieSceneScriptingFloatChannel
        default=-36.1331
        thinned_from=31
        keys [[0,-36.1331],[1,-17.1773],[2,-3.789],[3,8.2439],[4,21.0722],[5,13.4247],[6,-9.8134],[7,-15.939],[8,38.7097],[9,46.2251],[10,40.4043],[11,29.5971],[12,19.2612],[13,7.3454],[14,-0.0154],[15,5.4818],[17,26.7153],[18,32.271],[19,32.5169],[20,31.4495],[21,28.5913],[22,23.9558],[23,16.3908],[24,6.6161],[25,-3.1036],[26,-11.0419],[27,-16.1238],[28,-18.347],[29,-18.8817],[30,-18.8267]]
       }
       hand_r_ik_ctrl.Rotation.Z {
        type=MovieSceneScriptingFloatChannel
        default=-86.2861
        keys [[0,-86.2861],[1,-121.559],[2,-138.6221],[3,-152.3282],[4,-167.4032],[5,-184.1616],[6,-176.2124],[7,-154.1761],[8,-79.4835],[9,-16.0074],[10,17.3485],[11,37.9935],[12,52.8495],[13,66.9973],[14,74.4895],[15,76.5103],[16,72.8856],[17,58.4801],[18,37.1234],[19,17.7896],[20,3.1496],[21,-8.7595],[22,-17.788],[23,-23.2194],[24,-24.9077],[25,-23.9259],[26,-21.6633],[27,-20.0457],[28,-19.5966],[29,-19.3756],[30,-19.3306]]
       }
       hand_r_ik_ctrl.Scale.X {type=MovieSceneScriptingFloatChannel, default=1, thinned_from=31, keys [[0,1],[30,1]]}
       hand_r_ik_ctrl.Scale.Y {type=MovieSceneScriptingFloatChannel, default=1, const [1,31]}
       hand_r_ik_ctrl.Scale.Z {type=MovieSceneScriptingFloatChannel, default=1, const [1,31]}
       Softness_2 {type=MovieSceneScriptingFloatChannel, default=0, const [0,31]}
       hand_r_ikorient_ctrl.Rotation.Roll {type=MovieSceneScriptingFloatChannel, default=0, const [0,31]}
       hand_r_ikorient_ctrl.Rotation.Pitch {type=MovieSceneScriptingFloatChannel, default=0, const [0,31]}
       hand_r_ikorient_ctrl.Rotation.Yaw {type=MovieSceneScriptingFloatChannel, default=0, const [0,31]}
       arm_r_pv_ik_ctrl.Location.X {
        type=MovieSceneScriptingFloatChannel
        default=-52.7275
        keys [[0,-52.7275],[1,-55.555],[2,-39.0147],[3,-15.1948],[4,-0.3418],[5,11.6577],[6,17.8782],[7,0.3973],[8,-14.0206],[9,-19.9624],[10,-17.2681],[11,-19.3354],[12,-20.5068],[13,-16.4405],[14,-10.3706],[15,-3.8781],[16,2.5095],[17,5.5413],[18,2.8354],[19,-1.8146],[20,-6.4338],[21,-10.815],[22,-13.9176],[23,-16.3007],[24,-18.8713],[25,-21.3458],[26,-23.3883],[27,-24.6353],[28,-25.0181],[29,-24.9685],[30,-24.8679]]
       }
       arm_r_pv_ik_ctrl.Location.Y {
        type=MovieSceneScriptingFloatChannel
        default=-26.928
        keys [[0,-26.928],[1,-19.1367],[2,-15.9775],[3,-14.2503],[4,-6.5768],[5,8.7899],[6,44.9675],[7,74.7672],[8,80.8353],[9,65.4434],[10,65.1009],[11,69.3789],[12,69.8903],[13,66.2347],[14,59.9675],[15,52.1229],[16,41.8846],[17,27.6978],[18,14.1342],[19,5.261],[20,-1.5097],[21,-6.7589],[22,-10.4391],[23,-13.3232],[24,-15.6984],[25,-17.4942],[26,-18.6823],[27,-19.2911],[28,-19.4869],[29,-19.5071],[30,-19.4849]]
       }
       arm_r_pv_ik_ctrl.Location.Z {
        type=MovieSceneScriptingFloatChannel
        default=18.2316
        keys [[0,18.2316],[1,8.1408],[2,-0.9962],[3,5.5789],[4,15.5409],[5,21.6813],[6,22.9019],[7,12.2257],[8,-0.3741],[9,-14.706],[10,-16.5701],[11,-18.0069],[12,-19.0261],[13,-18.9664],[14,-17.9839],[15,-15.9213],[16,-12.3777],[17,-7.5359],[18,-3.5925],[19,-1.7378],[20,-0.1534],[21,1.104],[22,2.1753],[23,3.4688],[24,4.7936],[25,6.0916],[26,7.2306],[27,8.0357],[28,8.3457],[29,8.3237],[30,8.2533]]
       }
       arm_r_pv_ik_ctrl.Rotation.X {
        type=MovieSceneScriptingFloatChannel
        default=19.7866
        keys [[0,19.7866],[1,-16.9259],[2,-52.4537],[3,-79.9235],[4,-96.6066],[5,-105.3503],[6,-106.0575],[7,-90.8042],[8,-71.7464],[9,-35.7103],[10,-20.2555],[11,-16.4921],[12,-15.8475],[13,-13.1304],[14,-9.8792],[15,-5.4556],[16,1.5756],[17,9.8177],[18,14.6095],[19,15.0064],[20,15.1366],[21,14.7807],[22,15.0091],[23,16.6049],[24,18.4883],[25,20.4604],[26,22.2363],[27,23.4953],[28,24.0553],[29,24.1752],[30,24.1507]]
       }
       arm_r_pv_ik_ctrl.Rotation.Y {
        type=MovieSceneScriptingFloatChannel
        default=38.2178
        keys [[0,38.2178],[1,38.3223],[2,14.9817],[3,-8.5262],[4,-19.0357],[5,-19.6415],[6,-11.6782],[7,-6.1062],[8,0.4574],[9,-3.7828],[10,-2.8039],[11,-3.6629],[12,-5.803],[13,-10.0405],[14,-14.6461],[15,-18.0947],[16,-20.7935],[17,-21.6792],[18,-18.4408],[19,-13.7316],[20,-9.2877],[21,-4.9465],[22,-2.1135],[23,-0.7418],[24,0.3403],[25,1.2344],[26,1.9833],[27,2.5966],[28,2.8921],[29,2.8615],[30,2.787]]
       }
       arm_r_pv_ik_ctrl.Rotation.Z {
        type=MovieSceneScriptingFloatChannel
        default=-74.4995
        thinned_from=31
        keys [[0,-74.4995],[1,-115.2018],[2,-137.3589],[3,-132.0775],[4,-117.7031],[5,-96.4572],[6,-55.6735],[7,-16.4444],[8,9.311],[9,34.2585],[10,61.485],[11,85.0946],[12,98.7649],[13,100.7301],[14,95.4478],[15,86.2477],[16,72.9748],[17,52.313],[18,29.679],[19,12.8713],[20,-0.3021],[21,-11.2464],[22,-18.0801],[23,-22.1206],[24,-25.9622],[25,-29.3942],[26,-32.1225],[27,-33.7957],[28,-34.3157],[30,-34.0501]]
       }
       arm_r_pv_ik_ctrl.Scale.X {type=MovieSceneScriptingFloatChannel, default=1, thinned_from=31, keys [[0,1],[30,1]]}
       arm_r_pv_ik_ctrl.Scale.Y {type=MovieSceneScriptingFloatChannel, default=1, const [1,31]}
       arm_r_pv_ik_ctrl.Scale.Z {type=MovieSceneScriptingFloatChannel, default=1, const [1,31]}
       body_ctrl.Location.X {
        type=MovieSceneScriptingFloatChannel
        default=2.9482
        thinned_from=31
        keys [[0,2.9482],[1,3.0084],[2,2.8933],[3,2.5879],[4,2.1043],[5,1.4728],[6,0.5712],[7,-0.3113],[8,-0.8541],[9,-1.2698],[10,-1.6653],[11,-2.0422],[12,-2.2621],[13,-2.2637],[14,-2.0904],[15,-1.8476],[16,-1.6244],[17,-1.2982],[18,-0.7605],[19,-0.137],[20,0.4776],[21,1.0444],[22,1.4746],[23,1.7791],[24,2.0224],[25,2.2018],[26,2.3143],[27,2.358],[29,2.3813],[30,2.3869]]
       }
       body_ctrl.Location.Y {
        type=MovieSceneScriptingFloatChannel
        default=-0.4106
        keys [[0,-0.4106],[1,0.3393],[2,0.9644],[3,1.6953],[4,2.6403],[5,3.9155],[6,5.4063],[7,6.5907],[8,7.0808],[9,7.0688],[10,6.8159],[11,6.3717],[12,5.7829],[13,5.0681],[14,4.2705],[15,3.4419],[16,2.6052],[17,1.7169],[18,0.839],[19,0.1041],[20,-0.4851],[21,-0.9696],[22,-1.2932],[23,-1.5021],[24,-1.6756],[25,-1.8041],[26,-1.8801],[27,-1.897],[28,-1.8841],[29,-1.876],[30,-1.8732]]
       }
       body_ctrl.Location.Z {
        type=MovieSceneScriptingFloatChannel
        default=-5.2971
        thinned_from=31
        keys [[0,-5.2971],[1,-5.3164],[2,-5.242],[3,-5.054],[4,-4.8174],[5,-4.636],[6,-4.692],[7,-4.9547],[8,-5.1834],[10,-5.5909],[11,-5.7263],[12,-5.7552],[14,-5.6478],[15,-5.5611],[16,-5.4172],[17,-5.2012],[18,-4.9688],[19,-4.7709],[20,-4.5956],[21,-4.4356],[22,-4.3189],[23,-4.2351],[24,-4.1621],[25,-4.1053],[26,-4.0684],[27,-4.0553],[30,-4.0562]]
       }
       body_ctrl.Rotation.X {
        type=MovieSceneScriptingFloatChannel
        default=3.251
        keys [[0,3.251],[1,4.3926],[2,4.2925],[3,4.4716],[4,6.2969],[5,11.6961],[6,21.6264],[7,30.9518],[8,34.1304],[9,33.4542],[10,32.8806],[11,32.5277],[12,31.2781],[13,29.5147],[14,28.6753],[15,28.0822],[16,26.7642],[17,23.9118],[18,20.2592],[19,16.9276],[20,13.905],[21,10.9239],[22,8.7213],[23,7.044],[24,5.3319],[25,3.843],[26,2.8337],[27,2.5631],[28,2.7043],[29,2.7722],[30,2.7896]]
       }
       body_ctrl.Rotation.Y {
        type=MovieSceneScriptingFloatChannel
        default=-0.5041
        keys [[0,-0.5041],[1,0.0898],[2,2.0064],[3,3.5838],[4,3.6714],[5,2.0184],[6,1.7657],[7,3.9095],[8,2.7108],[9,1.7396],[10,3.1346],[11,4.7614],[12,5.0806],[13,3.5823],[14,0.6733],[15,-2.4144],[16,-4.4794],[17,-5.4932],[18,-5.9721],[19,-5.9586],[20,-5.555],[21,-4.695],[22,-3.4747],[23,-2.3084],[24,-1.4646],[25,-0.9504],[26,-0.7467],[27,-0.8221],[28,-1.0527],[29,-1.271],[30,-1.3681]]
       }
       body_ctrl.Rotation.Z {
        type=MovieSceneScriptingFloatChannel
        default=35.3655
        thinned_from=31
        keys [[0,35.3655],[1,37.2943],[2,36.5211],[3,33.5195],[4,28.3349],[5,20.8121],[6,9.6701],[7,-1.1923],[8,-10.1855],[9,-16.7682],[10,-19.3199],[11,-19.68],[12,-18.9636],[13,-18.1197],[14,-17.005],[15,-14.9149],[16,-11.3512],[17,-6.2755],[18,-0.5786],[19,4.4514],[20,8.323],[21,11.397],[22,13.5751],[23,15.0063],[24,15.9666],[25,16.5505],[26,16.8522],[27,16.9606],[30,16.9546]]
       }
       body_ctrl.Scale.X {
        type=MovieSceneScriptingFloatChannel
        default=1.001
        thinned_from=31
        keys [[0,1.001],[5,1.0045],[6,1.0078],[7,1.0146],[8,1.0132],[9,1.0092],[12,1.0108],[30,1.0016]]
       }
       body_ctrl.Scale.Y {type=MovieSceneScriptingFloatChannel, default=1, const [1,31]}
       body_ctrl.Scale.Z {type=MovieSceneScriptingFloatChannel, default=1, const [1,31]}
       hips_ctrl.Location.X {type=MovieSceneScriptingFloatChannel, default=0, thinned_from=31, keys [[0,0],[30,0]]}
       hips_ctrl.Location.Y {
        type=MovieSceneScriptingFloatChannel
        default=0.0001
        thinned_from=31
        keys [[0,0.0001],[2,0.0001],[5,0.0005],[6,0.0009],[7,0.0017],[8,0.0015],[9,0.0011],[12,0.0013],[30,0.0002]]
       }
       hips_ctrl.Location.Z {
        type=MovieSceneScriptingFloatChannel
        default=-0.0037
        thinned_from=31
        keys [[0,-0.0037],[1,-0.0031],[2,-0.0035],[3,-0.0087],[4,-0.0146],[5,-0.0165],[6,-0.0283],[7,-0.0529],[8,-0.0478],[9,-0.0333],[10,-0.0321],[11,-0.0369],[12,-0.039],[13,-0.0341],[14,-0.026],[15,-0.0201],[18,-0.0118],[19,-0.0098],[21,-0.0069],[22,-0.0059],[23,-0.0053],[24,-0.0052],[26,-0.0057],[30,-0.0057]]
       }
       hips_ctrl.Rotation.X {
        type=MovieSceneScriptingFloatChannel
        default=-6.3421
        keys [[0,-6.3421],[1,-8.207],[2,-7.9116],[3,-7.0311],[4,-6.9565],[5,-9.5165],[6,-15.3401],[7,-20.7245],[8,-20.731],[9,-17.628],[10,-16.5072],[11,-16.949],[12,-16.5109],[13,-15.1173],[14,-14.6932],[15,-14.8563],[16,-14.8865],[17,-14.1372],[18,-12.9982],[19,-11.9733],[20,-10.869],[21,-9.5383],[22,-8.5387],[23,-7.735],[24,-6.7702],[25,-5.8675],[26,-5.2446],[27,-5.1164],[28,-5.2639],[29,-5.3372],[30,-5.3569]]
       }
       hips_ctrl.Rotation.Y {
        type=MovieSceneScriptingFloatChannel
        default=4.2005
        thinned_from=31
        keys [[0,4.2005],[1,2.3708],[2,-0.6002],[3,-3.0604],[4,-3.8811],[5,-2.719],[6,-1.4455],[7,-0.5875],[8,3.3279],[9,6.1563],[10,5.6312],[11,4.164],[12,3.295],[13,3.9647],[14,6.0548],[15,8.1955],[16,9.1779],[17,8.6744],[18,7.3152],[19,5.885],[21,3.5755],[22,2.4717],[23,1.5926],[24,1.01],[25,0.7048],[26,0.6376],[27,0.7571],[28,0.9829],[29,1.1997],[30,1.2967]]
       }
       hips_ctrl.Rotation.Z {
        type=MovieSceneScriptingFloatChannel
        default=-6.4583
        keys [[0,-6.4583],[1,-6.9687],[2,-7.4021],[3,-8.0126],[4,-8.0628],[5,-6.6778],[6,-2.3449],[7,2.5117],[8,4.567],[9,6.2403],[10,7.5512],[11,7.8948],[12,7.2326],[13,6.1256],[14,4.8643],[15,3.5624],[16,2.5584],[17,2.059],[18,1.7132],[19,1.2439],[20,0.5381],[21,-0.2344],[22,-0.8143],[23,-1.1189],[24,-1.2948],[25,-1.3848],[26,-1.4273],[27,-1.4655],[28,-1.4953],[29,-1.495],[30,-1.4894]]
       }
       hips_ctrl.Scale.X {
        type=MovieSceneScriptingFloatChannel
        default=0.999
        thinned_from=31
        keys [[0,0.999],[5,0.9955],[6,0.9923],[7,0.9856],[8,0.987],[9,0.9909],[12,0.9894],[30,0.9984]]
       }
       hips_ctrl.Scale.Y {type=MovieSceneScriptingFloatChannel, default=1, const [1,31]}
       hips_ctrl.Scale.Z {type=MovieSceneScriptingFloatChannel, default=1, const [1,31]}
       hips_tan_ctrl.Location.X {
        type=MovieSceneScriptingFloatChannel
        default=0.0344
        thinned_from=31
        keys [[0,0.0344],[1,0.0186],[2,-0.0124],[3,-0.0472],[4,-0.0785],[5,-0.1065],[6,-0.1509],[7,-0.2015],[8,-0.221],[9,-0.2331],[10,-0.2509],[11,-0.2646],[12,-0.268],[13,-0.2553],[14,-0.225],[15,-0.1868],[17,-0.1175],[18,-0.0799],[19,-0.0495],[20,-0.0289],[21,-0.0148],[22,-0.0072],[23,-0.0028],[24,0.001],[25,0.0042],[26,0.0069],[27,0.0089],[29,0.0121],[30,0.0128]]
       }
       hips_tan_ctrl.Location.Y {
        type=MovieSceneScriptingFloatChannel
        default=-0.0731
        thinned_from=31
        keys [[0,-0.0731],[1,-0.075],[2,-0.0897],[3,-0.1099],[4,-0.1232],[5,-0.1145],[6,-0.0871],[7,-0.0671],[8,-0.0688],[9,-0.0897],[10,-0.1199],[11,-0.1564],[12,-0.1859],[13,-0.2011],[14,-0.2022],[15,-0.1965],[16,-0.1922],[17,-0.1898],[18,-0.1841],[21,-0.1627],[22,-0.1571],[23,-0.1539],[24,-0.1527],[26,-0.1528],[27,-0.1521],[28,-0.1512],[29,-0.1508],[30,-0.1506]]
       }
       hips_tan_ctrl.Location.Z {
        type=MovieSceneScriptingFloatChannel
        default=0.0121
        thinned_from=31
        keys [[0,0.0121],[1,0.0071],[2,0.0084],[3,0.0395],[4,0.0761],[5,0.0888],[6,0.1651],[7,0.3253],[8,0.2983],[9,0.2116],[10,0.2071],[11,0.241],[12,0.2556],[13,0.2236],[14,0.1691],[15,0.1286],[17,0.0878],[18,0.0664],[19,0.0514],[20,0.0406],[21,0.0312],[22,0.0247],[23,0.0212],[24,0.0205],[25,0.0219],[26,0.0237],[27,0.024],[28,0.0234],[30,0.0236]]
       }
       hips_tan_ctrl.Rotation.X {type=MovieSceneScriptingFloatChannel, default=0, const [0,31]}
       hips_tan_ctrl.Rotation.Y {type=MovieSceneScriptingFloatChannel, default=0, const [0,31]}
       hips_tan_ctrl.Rotation.Z {type=MovieSceneScriptingFloatChannel, default=0, const [0,31]}
       hips_tan_ctrl.Scale.X {type=MovieSceneScriptingFloatChannel, default=1, const [1,31]}
       hips_tan_ctrl.Scale.Y {type=MovieSceneScriptingFloatChannel, default=1, const [1,31]}
       hips_tan_ctrl.Scale.Z {type=MovieSceneScriptingFloatChannel, default=1, const [1,31]}
       spine_01_ctrl.Location.X {
        type=MovieSceneScriptingFloatChannel
        default=-0.0135
        thinned_from=31
        keys [[0,-0.0135],[3,-0.0167],[4,-0.0163],[5,-0.0133],[6,-0.0043],[7,0.0057],[8,0.0122],[9,0.0169],[10,0.0197],[11,0.0206],[12,0.0194],[13,0.0173],[15,0.0119],[16,0.0096],[19,0.0045],[21,0.0007],[22,-0.0007],[24,-0.0022],[26,-0.003],[30,-0.0033]]
       }
       spine_01_ctrl.Location.Y {
        type=MovieSceneScriptingFloatChannel
        default=-0.0008
        thinned_from=31
        keys [[0,-0.0008],[2,-0.0009],[4,-0.0025],[5,-0.0026],[6,-0.0042],[7,-0.0079],[8,-0.0073],[9,-0.0053],[10,-0.0052],[11,-0.006],[12,-0.0062],[13,-0.0054],[14,-0.0041],[15,-0.0031],[18,-0.0017],[22,-0.0008],[30,-0.0008]]
       }
       spine_01_ctrl.Location.Z {
        type=MovieSceneScriptingFloatChannel
        default=0.0088
        thinned_from=31
        keys [[0,0.0088],[1,0.0077],[2,0.0086],[3,0.0234],[4,0.0409],[5,0.0478],[6,0.0866],[7,0.1636],[8,0.1472],[9,0.1009],[10,0.0947],[11,0.1076],[12,0.1126],[13,0.0971],[14,0.0727],[15,0.0557],[16,0.0478],[17,0.0396],[18,0.0311],[19,0.025],[20,0.0204],[21,0.0161],[22,0.013],[23,0.0112],[24,0.0106],[25,0.0109],[26,0.0115],[27,0.0116],[28,0.0114],[30,0.0115]]
       }
       spine_01_ctrl.Rotation.X {
        type=MovieSceneScriptingFloatChannel
        default=1.4951
        keys [[0,1.4951],[1,1.0368],[2,1.2932],[3,1.7804],[4,1.9708],[5,1.128],[6,-0.9139],[7,-2.4029],[8,-1.8887],[9,-0.2072],[10,1.282],[11,2.5177],[12,3.4557],[13,3.9419],[14,3.6921],[15,3.1659],[16,2.8531],[17,2.8541],[18,2.8875],[19,2.8845],[20,2.9155],[21,3.0233],[22,3.0945],[23,3.1877],[24,3.3791],[25,3.5893],[26,3.7407],[27,3.7562],[28,3.6978],[29,3.671],[30,3.6647]]
       }
       spine_01_ctrl.Rotation.Y {
        type=MovieSceneScriptingFloatChannel
        default=2.1246
        keys [[0,2.1246],[1,1.3302],[2,-0.3217],[3,-2.0355],[4,-3.2303],[5,-3.6608],[6,-4.703],[7,-6.6539],[8,-6.7415],[9,-6.457],[10,-7.0042],[11,-7.6091],[12,-7.6971],[13,-6.9788],[14,-5.5604],[15,-3.9623],[16,-2.7106],[17,-1.7399],[18,-0.8613],[19,-0.2226],[20,0.1733],[21,0.359],[22,0.3472],[23,0.2758],[24,0.2471],[25,0.2631],[26,0.3181],[27,0.4013],[28,0.5061],[29,0.604],[30,0.6472]]
       }
       spine_01_ctrl.Rotation.Z {
        type=MovieSceneScriptingFloatChannel
        default=6.3969
        keys [[0,6.3969],[1,6.8453],[2,7.3828],[3,7.8494],[4,7.6515],[5,6.3374],[6,2.5027],[7,-1.7224],[8,-4.9601],[9,-7.5652],[10,-9.1041],[11,-9.7171],[12,-9.3231],[13,-8.3898],[14,-7.0917],[15,-5.6825],[16,-4.5241],[17,-3.6781],[18,-2.9058],[19,-2.1154],[20,-1.2137],[21,-0.3147],[22,0.3383],[23,0.7396],[24,1.0529],[25,1.2878],[26,1.4521],[27,1.5508],[28,1.5975],[29,1.614],[30,1.6161]]
       }
       spine_01_ctrl.Scale.X {
        type=MovieSceneScriptingFloatChannel
        default=0.999
        thinned_from=31
        keys [[0,0.999],[5,0.9955],[6,0.9923],[7,0.9856],[8,0.987],[9,0.9909],[12,0.9894],[30,0.9984]]
       }
       spine_01_ctrl.Scale.Y {type=MovieSceneScriptingFloatChannel, default=1, const [1,31]}
       spine_01_ctrl.Scale.Z {type=MovieSceneScriptingFloatChannel, default=1, const [1,31]}
       spine_02_ctrl.Location.X {
        type=MovieSceneScriptingFloatChannel
        default=-0.0091
        thinned_from=31
        keys [[0,-0.0091],[2,-0.0108],[3,-0.012],[4,-0.0127],[5,-0.011],[6,-0.007],[7,-0.0061],[8,-0.0005],[9,0.0059],[10,0.0076],[11,0.0067],[12,0.0053],[13,0.0055],[14,0.0062],[15,0.0061],[18,0.004],[19,0.0031],[21,0.0006],[22,-0.0004],[24,-0.0015],[26,-0.002],[30,-0.0022]]
       }
       spine_02_ctrl.Location.Y {
        type=MovieSceneScriptingFloatChannel
        default=-0.0004
        thinned_from=31
        keys [[0,-0.0004],[3,-0.0006],[4,-0.0004],[5,0.0002],[6,0.0025],[7,0.0068],[8,0.0052],[9,0.0017],[10,0.0001],[11,-0.001],[12,-0.0019],[13,-0.002],[15,-0.0006],[16,-0.0003],[30,-0.0001]]
       }
       spine_02_ctrl.Location.Z {
        type=MovieSceneScriptingFloatChannel
        default=-0.0013
        thinned_from=31
        keys [[0,-0.0013],[1,0.0001],[2,0.0002],[3,-0.0001],[4,-0.0001],[5,0.0015],[6,0.0054],[7,0.0083],[8,0.007],[9,0.0036],[10,0.0008],[11,-0.0013],[12,-0.0028],[13,-0.0034],[14,-0.0025],[15,-0.0012],[16,-0.0005],[20,-0.0016],[26,-0.0042],[30,-0.0041]]
       }
       spine_02_ctrl.Rotation.X {
        type=MovieSceneScriptingFloatChannel
        default=0.8842
        keys [[0,0.8842],[1,-0.03],[2,-0.1066],[3,0.1083],[4,0.1119],[5,-0.9596],[6,-3.5456],[7,-5.3218],[8,-4.5005],[9,-2.3393],[10,-0.4949],[11,0.8943],[12,1.8573],[13,2.2358],[14,1.6432],[15,0.7953],[16,0.3495],[17,0.4273],[18,0.6406],[19,0.8271],[20,1.0738],[21,1.4345],[22,1.6877],[23,1.9154],[24,2.2644],[25,2.6233],[26,2.8802],[27,2.9216],[28,2.8449],[29,2.8119],[30,2.8048]]
       }
       spine_02_ctrl.Rotation.Y {
        type=MovieSceneScriptingFloatChannel
        default=2.7464
        keys [[0,2.7464],[1,1.8282],[2,-0.2938],[3,-2.4612],[4,-3.789],[5,-3.8047],[6,-4.5701],[7,-7.0228],[8,-6.9655],[9,-6.2109],[10,-6.6379],[11,-7.2541],[12,-7.2257],[13,-6.2478],[14,-4.5667],[15,-2.7397],[16,-1.3701],[17,-0.4527],[18,0.249],[19,0.6929],[20,0.9339],[21,0.9485],[22,0.7542],[23,0.5276],[24,0.3827],[25,0.3239],[26,0.3474],[27,0.4421],[28,0.5846],[29,0.7185],[30,0.7778]]
       }
       spine_02_ctrl.Rotation.Z {
        type=MovieSceneScriptingFloatChannel
        default=6.3642
        keys [[0,6.3642],[1,6.8141],[2,7.4242],[3,8.0023],[4,7.8874],[5,6.5529],[6,2.6363],[7,-1.6159],[8,-4.9898],[9,-7.6555],[10,-9.1741],[11,-9.7419],[12,-9.2628],[13,-8.2483],[14,-6.9492],[15,-5.6105],[16,-4.5194],[17,-3.7013],[18,-2.9437],[19,-2.1615],[20,-1.2597],[21,-0.3512],[22,0.3137],[23,0.7246],[24,1.0429],[25,1.2788],[26,1.4414],[27,1.5372],[28,1.58],[29,1.5928],[30,1.5933]]
       }
       spine_02_ctrl.Scale.X {type=MovieSceneScriptingFloatChannel, default=1, const [1,31]}
       spine_02_ctrl.Scale.Y {type=MovieSceneScriptingFloatChannel, default=1, const [1,31]}
       spine_02_ctrl.Scale.Z {type=MovieSceneScriptingFloatChannel, default=1, const [1,31]}
       spine_03_ctrl.Location.X {
        type=MovieSceneScriptingFloatChannel
        default=0.0053
        thinned_from=31
        keys [[0,0.0053],[1,0.0055],[3,0.0053],[4,0.0049],[5,0.0038],[6,0.0005],[7,-0.0038],[8,-0.0059],[9,-0.0072],[10,-0.0082],[11,-0.0087],[12,-0.0082],[13,-0.0072],[15,-0.0045],[16,-0.0034],[21,0],[25,0.001],[30,0.0013]]
       }
       spine_03_ctrl.Location.Y {
        type=MovieSceneScriptingFloatChannel
        default=0.0003
        thinned_from=31
        keys [[0,0.0003],[1,0.0006],[4,0.0008],[5,0.001],[7,0.0026],[8,0.0025],[10,0.0015],[13,0.0006],[16,0.0008],[27,-0.0003],[30,-0.0003]]
       }
       spine_03_ctrl.Location.Z {
        type=MovieSceneScriptingFloatChannel
        default=0.0001
        thinned_from=31
        keys [[0,0.0001],[1,-0.0012],[2,-0.0017],[4,-0.0021],[5,-0.0032],[6,-0.0059],[7,-0.0075],[8,-0.0067],[10,-0.0027],[11,-0.0012],[12,-0.0003],[13,-0.0001],[15,-0.0023],[16,-0.0028],[17,-0.0026],[20,-0.0012],[22,-0.0001],[26,0.0016],[30,0.0015]]
       }
       spine_03_ctrl.Rotation.X {
        type=MovieSceneScriptingFloatChannel
        default=0.1589
        keys [[0,0.1589],[1,-1.6429],[2,-2.3349],[3,-2.5811],[4,-2.8517],[5,-4.323],[6,-8.1587],[7,-10.6551],[8,-9.4334],[9,-6.4529],[10,-3.7745],[11,-1.7937],[12,-0.5421],[13,-0.2683],[14,-1.5258],[15,-3.0529],[16,-3.7647],[17,-3.4794],[18,-2.8614],[19,-2.277],[20,-1.5985],[21,-0.7443],[22,-0.15],[23,0.329],[24,0.9842],[25,1.6365],[26,2.1015],[27,2.1903],[28,2.0724],[29,2.0251],[30,2.0164]]
       }
       spine_03_ctrl.Rotation.Y {
        type=MovieSceneScriptingFloatChannel
        default=3.8981
        thinned_from=31
        keys [[0,3.8981],[1,2.7175],[2,-0.4632],[3,-3.6874],[4,-5.4313],[5,-4.7958],[6,-5.3241],[7,-9.1928],[8,-9.1049],[9,-7.6054],[10,-7.9938],[11,-8.833],[12,-8.6519],[13,-7.0625],[14,-4.6103],[15,-1.941],[16,0.0423],[17,1.2106],[18,1.8976],[19,2.1981],[20,2.2862],[21,2.052],[23,1.0126],[24,0.6482],[25,0.4503],[26,0.4195],[27,0.5458],[28,0.7739],[29,0.9894],[30,1.085]]
       }
       spine_03_ctrl.Rotation.Z {
        type=MovieSceneScriptingFloatChannel
        default=6.3089
        keys [[0,6.3089],[1,6.7117],[2,7.3796],[3,8.1068],[4,8.1028],[5,6.6741],[6,2.618],[7,-1.4787],[8,-4.7491],[9,-7.4295],[10,-8.9702],[11,-9.5647],[12,-9.1116],[13,-8.1009],[14,-6.8406],[15,-5.6067],[16,-4.6141],[17,-3.8366],[18,-3.081],[19,-2.2854],[20,-1.3658],[21,-0.4301],[22,0.2623],[23,0.6942],[24,1.0267],[25,1.2697],[26,1.4337],[27,1.5266],[28,1.564],[29,1.5719],[30,1.5701]]
       }
       spine_03_ctrl.Scale.X {type=MovieSceneScriptingFloatChannel, default=1, const [1,31]}
       spine_03_ctrl.Scale.Y {type=MovieSceneScriptingFloatChannel, default=1, const [1,31]}
       spine_03_ctrl.Scale.Z {type=MovieSceneScriptingFloatChannel, default=1, const [1,31]}
       spine_04_ctrl.Location.X {
        type=MovieSceneScriptingFloatChannel
        default=0
        thinned_from=31
        keys [[0,-0.0006],[1,-0.0003],[2,0.0003],[3,0.0025],[4,0.0057],[5,0.0054],[6,0.0087],[7,0.0274],[8,0.0232],[9,0.0124],[10,0.0129],[11,0.0176],[12,0.0186],[13,0.0126],[14,0.0051],[15,0.0002],[16,-0.0022],[17,-0.0028],[22,-0.0007],[25,-0.0001],[30,-0.0003]]
       }
       spine_04_ctrl.Location.Y {
        type=MovieSceneScriptingFloatChannel
        default=0
        thinned_from=31
        keys [[0,-0.0022],[1,-0.0023],[2,-0.0027],[3,-0.0064],[4,-0.0108],[5,-0.0139],[6,-0.0308],[7,-0.0664],[8,-0.0578],[9,-0.0354],[10,-0.0292],[11,-0.0293],[12,-0.0279],[13,-0.0239],[14,-0.0207],[15,-0.0181],[16,-0.0163],[17,-0.0133],[18,-0.01],[19,-0.0078],[20,-0.0062],[21,-0.0048],[22,-0.0038],[24,-0.003],[30,-0.0028]]
       }
       spine_04_ctrl.Location.Z {
        type=MovieSceneScriptingFloatChannel
        default=0
        thinned_from=31
        keys [[0,0.0262],[1,0.0229],[2,0.0258],[3,0.0571],[4,0.0932],[5,0.1051],[6,0.1756],[7,0.3227],[8,0.2929],[9,0.2086],[10,0.2033],[11,0.2351],[12,0.2497],[13,0.2208],[14,0.1702],[15,0.1327],[16,0.1145],[17,0.0971],[18,0.079],[19,0.0662],[20,0.0566],[21,0.048],[22,0.0417],[23,0.0383],[24,0.0375],[25,0.0386],[26,0.0401],[27,0.0403],[28,0.0397],[30,0.0399]]
       }
       spine_04_ctrl.Rotation.X {
        type=MovieSceneScriptingFloatChannel
        default=0
        thinned_from=31
        keys [[0,-1.8],[1,-1.6986],[2,-1.6808],[3,-1.7055],[4,-1.6343],[5,-1.2926],[6,-1.0013],[7,-1.8511],[8,-2.7736],[9,-3.6578],[10,-4.5248],[11,-5.1444],[12,-5.2838],[13,-5.0426],[14,-4.5089],[15,-3.8743],[16,-3.4029],[17,-3.0812],[18,-2.7935],[19,-2.5717],[20,-2.4162],[21,-2.3241],[22,-2.2725],[23,-2.2584],[24,-2.2835],[25,-2.3236],[26,-2.3554],[27,-2.3576],[28,-2.3436],[30,-2.3377]]
       }
       spine_04_ctrl.Rotation.Y {
        type=MovieSceneScriptingFloatChannel
        default=0
        keys [[0,-1.3624],[1,-1.149],[2,-0.5882],[3,0.0243],[4,0.5015],[5,0.8543],[6,1.7592],[7,2.7732],[8,3.246],[9,3.3729],[10,3.2208],[11,2.9302],[12,2.7922],[13,2.7778],[14,2.6711],[15,2.3941],[16,2.0315],[17,1.5928],[18,1.0906],[19,0.6677],[20,0.3594],[21,0.1351],[22,0.0147],[23,-0.0469],[24,-0.1005],[25,-0.1483],[26,-0.1911],[27,-0.2287],[28,-0.2653],[29,-0.2967],[30,-0.3101]]
       }
       spine_04_ctrl.Rotation.Z {
        type=MovieSceneScriptingFloatChannel
        default=0
        thinned_from=31
        keys [[0,6.4534],[1,6.9023],[2,7.3709],[3,7.8014],[4,7.61],[5,6.2221],[6,1.9605],[7,-2.9898],[8,-6.1999],[9,-8.4293],[10,-9.7451],[11,-10.211],[12,-9.6683],[13,-8.5932],[14,-7.2924],[15,-5.9034],[16,-4.7199],[17,-3.8106],[19,-2.1546],[20,-1.232],[21,-0.321],[22,0.3374],[23,0.7414],[24,1.0549],[25,1.2874],[26,1.4478],[27,1.5437],[28,1.5886],[29,1.603],[30,1.604]]
       }
       spine_04_ctrl.Scale.X {type=MovieSceneScriptingFloatChannel, default=1, const [1,31]}
       spine_04_ctrl.Scale.Y {type=MovieSceneScriptingFloatChannel, default=1, const [1,31]}
       spine_04_ctrl.Scale.Z {type=MovieSceneScriptingFloatChannel, default=1, const [1,31]}
       head_ctrl.Location.X {type=MovieSceneScriptingFloatChannel, default=0, thinned_from=31, keys [[0,0],[30,0]]}
       head_ctrl.Location.Y {type=MovieSceneScriptingFloatChannel, default=0, thinned_from=31, keys [[0,0],[30,0]]}
       head_ctrl.Location.Z {type=MovieSceneScriptingFloatChannel, default=0, thinned_from=31, keys [[0,0],[30,0]]}
       head_ctrl.Rotation.X {
        type=MovieSceneScriptingFloatChannel
        default=3.0818
        thinned_from=31
        keys [[0,3.0818],[1,1.7554],[2,2.4257],[3,2.6316],[4,0.6751],[5,-5.1427],[6,-14.4278],[7,-22.5115],[8,-24.2705],[9,-22.5534],[10,-21.9374],[11,-21.8442],[12,-20.7448],[13,-18.8774],[14,-17.8202],[15,-17.3224],[16,-16.6714],[17,-14.857],[18,-12.2249],[19,-9.6837],[20,-7.1732],[21,-4.4442],[22,-2.2425],[24,1.3222],[25,2.8312],[26,3.816],[27,4.0388],[28,3.8361],[29,3.7072],[30,3.6621]]
       }
       head_ctrl.Rotation.Y {
        type=MovieSceneScriptingFloatChannel
        default=-4.8387
        keys [[0,-4.8387],[1,-7.4435],[2,-10.5689],[3,-13.2559],[4,-14.4446],[5,-13.2394],[6,-8.9957],[7,-3.3395],[8,4.1932],[9,9.2253],[10,9.628],[11,8.3243],[12,6.8618],[13,6.5365],[14,7.4405],[15,8.1991],[16,7.5851],[17,5.6478],[18,3.6891],[19,2.2106],[20,1.2419],[21,0.5051],[22,-0.2084],[23,-0.6982],[24,-0.7404],[25,-0.4663],[26,-0.0235],[27,0.4381],[28,0.8421],[29,1.1508],[30,1.2739]]
       }
       head_ctrl.Rotation.Z {
        type=MovieSceneScriptingFloatChannel
        default=-25.6952
        thinned_from=31
        keys [[0,-25.6952],[1,-24.7412],[2,-23.3524],[3,-20.6951],[4,-16.1794],[5,-9.1986],[6,0.3126],[7,7.7743],[8,10.7356],[9,12.834],[10,14.4024],[11,15.0047],[12,14.8535],[13,14.2376],[14,12.976],[15,10.8321],[16,7.7373],[17,3.5938],[18,-1.238],[19,-5.6329],[20,-9.0684],[21,-11.816],[22,-13.7627],[23,-15.0611],[24,-15.9759],[25,-16.5585],[26,-16.8731],[27,-17.0027],[28,-17.0458],[30,-17.0509]]
       }
       head_ctrl.Scale.X {type=MovieSceneScriptingFloatChannel, default=1, thinned_from=31, keys [[0,1],[30,1]]}
       head_ctrl.Scale.Y {type=MovieSceneScriptingFloatChannel, default=1, const [1,31]}
       head_ctrl.Scale.Z {type=MovieSceneScriptingFloatChannel, default=1, const [1,31]}
       clavicle_l_ctrl.Location.X {type=MovieSceneScriptingFloatChannel, default=0.1526, thinned_from=31, keys [[0,0],[30,0]]}
       clavicle_l_ctrl.Location.Y {type=MovieSceneScriptingFloatChannel, default=0.1462, thinned_from=31, keys [[0,0],[30,0]]}
       clavicle_l_ctrl.Location.Z {type=MovieSceneScriptingFloatChannel, default=-0.0378, thinned_from=31, keys [[0,0],[30,0]]}
       clavicle_l_ctrl.Rotation.X {type=MovieSceneScriptingFloatChannel, default=-1.2758, thinned_from=31, keys [[0,0],[30,0]]}
       clavicle_l_ctrl.Rotation.Y {
        type=MovieSceneScriptingFloatChannel
        default=6.7355
        thinned_from=31
        keys [[0,8.2368],[1,11.2719],[2,13.6911],[3,15.6806],[4,17.2503],[5,18.4098],[6,19.0933],[7,19.3182],[8,19.22],[9,18.9337],[10,18.4157],[11,17.4833],[12,16.0633],[13,14.0352],[15,9.1564],[16,7.1159],[17,5.4557],[18,3.8968],[19,2.4561],[20,1.1506],[21,-0.0023],[22,-0.986],[23,-1.7369],[24,-2.2359],[25,-2.5317],[26,-2.6733],[27,-2.7096],[30,-2.7096]]
       }
       clavicle_l_ctrl.Rotation.Z {
        type=MovieSceneScriptingFloatChannel
        default=5.2643
        keys [[0,-1.0972],[1,-4.8807],[2,-6.5551],[3,-7.5849],[4,-8.1044],[5,-8.2477],[6,-8.0626],[7,-7.5263],[8,-6.6672],[9,-5.5138],[10,-4.166],[11,-2.8015],[12,-1.5649],[13,-0.514],[14,0.3363],[15,0.9056],[16,1.1132],[17,1.0628],[18,0.9281],[19,0.7336],[20,0.5041],[21,0.2645],[22,0.0392],[23,-0.1926],[24,-0.4537],[25,-0.7224],[26,-0.9766],[27,-1.1948],[28,-1.3605],[29,-1.4662],[30,-1.5033]]
       }
       clavicle_l_ctrl.Scale.X {type=MovieSceneScriptingFloatChannel, default=1, const [1,31]}
       clavicle_l_ctrl.Scale.Y {type=MovieSceneScriptingFloatChannel, default=1, const [1,31]}
       clavicle_l_ctrl.Scale.Z {type=MovieSceneScriptingFloatChannel, default=1, const [1,31]}
       clavicle_r_ctrl.Location.X {type=MovieSceneScriptingFloatChannel, default=-0.1719, thinned_from=31, keys [[0,0],[30,0]]}
       clavicle_r_ctrl.Location.Y {type=MovieSceneScriptingFloatChannel, default=-0.3924, thinned_from=31, keys [[0,0],[30,0]]}
       clavicle_r_ctrl.Location.Z {type=MovieSceneScriptingFloatChannel, default=0.051, thinned_from=31, keys [[0,0],[30,0]]}
       clavicle_r_ctrl.Rotation.X {type=MovieSceneScriptingFloatChannel, default=-2.5082, thinned_from=31, keys [[0,0],[10,0.0001],[27,0],[30,0]]}
       clavicle_r_ctrl.Rotation.Y {
        type=MovieSceneScriptingFloatChannel
        default=4.9556
        thinned_from=31
        keys [[0,4.2087],[2,10.3719],[3,13.1484],[4,15.1533],[5,15.9244],[6,15.5497],[7,13.9475],[8,10.3482],[9,6.103],[10,2.3017],[11,-1.1898],[12,-3.7445],[13,-5.2721],[14,-6.1504],[15,-6.4544],[16,-6.259],[17,-5.5246],[18,-4.4599],[19,-3.5135],[20,-2.8261],[21,-2.3222],[22,-2.1266],[23,-2.2131],[24,-2.4295],[26,-2.9922],[27,-3.2085],[28,-3.3367],[29,-3.4049],[30,-3.4249]]
       }
       clavicle_r_ctrl.Rotation.Z {
        type=MovieSceneScriptingFloatChannel
        default=-18.3645
        keys [[0,-11.8666],[1,-12.5575],[2,-12.6104],[3,-12.3197],[4,-11.5409],[5,-10.1297],[6,-6.7217],[7,-1.9041],[8,3.5866],[9,8.2501],[10,10.5405],[11,11.7135],[12,12.0436],[13,11.6146],[14,10.4584],[15,8.7716],[16,6.751],[17,4.5462],[18,2.3796],[19,0.5562],[20,-0.8412],[21,-1.8953],[22,-2.5494],[23,-2.9135],[24,-3.1419],[25,-3.2441],[26,-3.23],[27,-3.1089],[28,-2.8877],[29,-2.6564],[30,-2.5494]]
       }
       clavicle_r_ctrl.Scale.X {type=MovieSceneScriptingFloatChannel, default=1, const [1,31]}
       clavicle_r_ctrl.Scale.Y {type=MovieSceneScriptingFloatChannel, default=1, const [1,31]}
       clavicle_r_ctrl.Scale.Z {type=MovieSceneScriptingFloatChannel, default=1, const [1,31]}
       neck_01_ctrl.Location.X {type=MovieSceneScriptingFloatChannel, default=0.39, thinned_from=31, keys [[0,0],[30,0]]}
       neck_01_ctrl.Location.Y {type=MovieSceneScriptingFloatChannel, default=-0.3193, thinned_from=31, keys [[0,0],[30,0]]}
       neck_01_ctrl.Location.Z {type=MovieSceneScriptingFloatChannel, default=-0.0063, thinned_from=31, keys [[0,0],[30,0]]}
       neck_01_ctrl.Rotation.X {
        type=MovieSceneScriptingFloatChannel
        default=-2.8935
        thinned_from=31
        keys [[0,-2.8779],[1,-4.5257],[2,-6.3022],[3,-8.044],[4,-8.974],[5,-7.2367],[6,-3.179],[7,0.7457],[8,0.8574],[9,-1.6788],[10,-3.2426],[11,-4.0704],[12,-4.6242],[13,-4.8894],[14,-5.0252],[15,-5.5465],[16,-6.1818],[17,-6.677],[18,-6.8403],[19,-6.7536],[20,-6.7354],[21,-6.8508],[22,-6.8685],[23,-6.8351],[24,-6.921],[26,-7.189],[28,-7.2965],[29,-7.3874],[30,-7.4364]]
       }
       neck_01_ctrl.Rotation.Y {
        type=MovieSceneScriptingFloatChannel
        default=-7.6315
        keys [[0,-5.3742],[1,-3.1349],[2,-0.9839],[3,0.84],[4,2.3146],[5,3.7031],[6,6.646],[7,9.8429],[8,9.0905],[9,8.2841],[10,9.4014],[11,10.609],[12,10.662],[13,9.2317],[14,6.4869],[15,3.4884],[16,1.1879],[17,-0.6026],[18,-2.044],[19,-2.6626],[20,-2.759],[21,-2.6926],[22,-2.4389],[23,-2.1649],[24,-2.0246],[25,-1.9894],[26,-2.0083],[27,-2.0181],[28,-2.0427],[29,-2.1069],[30,-2.1434]]
       }
       neck_01_ctrl.Rotation.Z {
        type=MovieSceneScriptingFloatChannel
        default=-46.0426
        keys [[0,-52.4764],[1,-57.0657],[2,-59.113],[3,-59.0665],[4,-54.8663],[5,-44.3905],[6,-20.1821],[7,6.4497],[8,22.6704],[9,33.0112],[10,39.4203],[11,42.0982],[12,40.9786],[13,38.0397],[14,33.5791],[15,28.2432],[16,22.6301],[17,16.6996],[18,10.4317],[19,4.6679],[20,-0.5209],[21,-5.1298],[22,-8.3284],[23,-10.2125],[24,-11.5234],[25,-12.3897],[26,-12.9359],[27,-13.2853],[28,-13.4723],[29,-13.5138],[30,-13.5062]]
       }
       neck_01_ctrl.Scale.X {type=MovieSceneScriptingFloatChannel, default=1, const [1,31]}
       neck_01_ctrl.Scale.Y {type=MovieSceneScriptingFloatChannel, default=1, const [1,31]}
       neck_01_ctrl.Scale.Z {type=MovieSceneScriptingFloatChannel, default=1, const [1,31]}
       neck_02_ctrl.Location.X {type=MovieSceneScriptingFloatChannel, default=0, const [0,31]}
       neck_02_ctrl.Location.Y {type=MovieSceneScriptingFloatChannel, default=0, const [0,31]}
       neck_02_ctrl.Location.Z {type=MovieSceneScriptingFloatChannel, default=0, const [0,31]}
       neck_02_ctrl.Rotation.X {type=MovieSceneScriptingFloatChannel, default=0, const [0,31]}
       neck_02_ctrl.Rotation.Y {type=MovieSceneScriptingFloatChannel, default=0, const [0,31]}
       neck_02_ctrl.Rotation.Z {type=MovieSceneScriptingFloatChannel, default=0, const [0,31]}
       neck_02_ctrl.Scale.X {type=MovieSceneScriptingFloatChannel, default=1, const [1,31]}
       neck_02_ctrl.Scale.Y {type=MovieSceneScriptingFloatChannel, default=1, const [1,31]}
       neck_02_ctrl.Scale.Z {type=MovieSceneScriptingFloatChannel, default=1, const [1,31]}
       head_ik_ctrl.Location.X {
        type=MovieSceneScriptingFloatChannel
        default=3.4129
        thinned_from=31
        keys [[0,2.8937],[1,2.592],[2,2.2281],[3,1.8559],[4,1.4669],[5,1.0075],[6,-0.271],[7,-1.8567],[8,-2.4581],[9,-2.7752],[10,-3.1757],[11,-3.4217],[12,-3.3638],[13,-3.0411],[14,-2.4597],[15,-1.7628],[16,-1.1523],[17,-0.629],[18,-0.1635],[19,0.1507],[20,0.3711],[21,0.549],[22,0.6437],[23,0.6824],[24,0.7179],[25,0.7508],[26,0.7778],[27,0.7932],[29,0.811],[30,0.815]]
       }
       head_ik_ctrl.Location.Y {
        type=MovieSceneScriptingFloatChannel
        default=-1.329
        thinned_from=31
        keys [[0,-0.8696],[1,-0.9873],[2,-1.0192],[3,-1.0576],[4,-1.088],[5,-0.8524],[6,-0.336],[7,0.2215],[8,0.3172],[9,0.0822],[10,-0.063],[11,-0.1683],[12,-0.3407],[13,-0.5096],[14,-0.6393],[15,-0.7992],[16,-0.9425],[17,-1.0504],[18,-1.1197],[19,-1.1532],[20,-1.1829],[21,-1.2152],[22,-1.2148],[23,-1.1993],[24,-1.2012],[26,-1.2252],[27,-1.2318],[28,-1.2413],[29,-1.258],[30,-1.267]]
       }
       head_ik_ctrl.Location.Z {
        type=MovieSceneScriptingFloatChannel
        default=0.5714
        keys [[0,0.4173],[1,0.2646],[2,0.076],[3,-0.0212],[4,-0.021],[5,-0.0337],[6,0.0632],[7,0.524],[8,0.6123],[9,0.557],[10,0.7436],[11,0.9764],[12,1.044],[13,0.8725],[14,0.5438],[15,0.2828],[16,0.1773],[17,0.1734],[18,0.2008],[19,0.2291],[20,0.2621],[21,0.3002],[22,0.3169],[23,0.3255],[24,0.3479],[25,0.3761],[26,0.4004],[27,0.4096],[28,0.4121],[29,0.4191],[30,0.4231]]
       }
       head_ik_ctrl.Rotation.X {
        type=MovieSceneScriptingFloatChannel
        default=3.0819
        thinned_from=31
        keys [[0,3.0818],[1,1.7554],[2,2.4257],[3,2.6316],[4,0.6751],[5,-5.1427],[6,-14.4278],[7,-22.5115],[8,-24.2705],[9,-22.5534],[10,-21.9374],[11,-21.8442],[12,-20.7448],[13,-18.8774],[14,-17.8202],[15,-17.3224],[16,-16.6714],[17,-14.857],[18,-12.2249],[19,-9.6837],[20,-7.1732],[21,-4.4442],[22,-2.2425],[24,1.3222],[25,2.8312],[26,3.816],[27,4.0388],[28,3.8361],[29,3.7072],[30,3.6621]]
       }
       head_ik_ctrl.Rotation.Y {
        type=MovieSceneScriptingFloatChannel
        default=-4.8387
        keys [[0,-4.8387],[1,-7.4435],[2,-10.5689],[3,-13.2559],[4,-14.4446],[5,-13.2394],[6,-8.9957],[7,-3.3395],[8,4.1932],[9,9.2253],[10,9.628],[11,8.3243],[12,6.8618],[13,6.5365],[14,7.4405],[15,8.1991],[16,7.5851],[17,5.6478],[18,3.6891],[19,2.2106],[20,1.2419],[21,0.5051],[22,-0.2084],[23,-0.6982],[24,-0.7404],[25,-0.4663],[26,-0.0235],[27,0.4381],[28,0.8421],[29,1.1508],[30,1.2739]]
       }
       head_ik_ctrl.Rotation.Z {
        type=MovieSceneScriptingFloatChannel
        default=-25.6952
        thinned_from=31
        keys [[0,-25.6952],[1,-24.7412],[2,-23.3524],[3,-20.6951],[4,-16.1794],[5,-9.1986],[6,0.3126],[7,7.7743],[8,10.7356],[9,12.834],[10,14.4024],[11,15.0047],[12,14.8535],[13,14.2376],[14,12.976],[15,10.8321],[16,7.7373],[17,3.5938],[18,-1.238],[19,-5.6329],[20,-9.0684],[21,-11.816],[22,-13.7627],[23,-15.0611],[24,-15.9759],[25,-16.5585],[26,-16.8731],[27,-17.0027],[28,-17.0458],[30,-17.0509]]
       }
       head_ik_ctrl.Scale.X {type=MovieSceneScriptingFloatChannel, default=1, thinned_from=31, keys [[0,1],[30,1]]}
       head_ik_ctrl.Scale.Y {type=MovieSceneScriptingFloatChannel, default=1, const [1,31]}
       head_ik_ctrl.Scale.Z {type=MovieSceneScriptingFloatChannel, default=1, const [1,31]}
       chest_ctrl.Location.X {
        type=MovieSceneScriptingFloatChannel
        default=-0.7314
        thinned_from=31
        keys [[0,-0.7314],[1,-0.3556],[2,0.3476],[3,1.0448],[4,1.4991],[5,1.6309],[6,1.8992],[7,2.3912],[9,1.9965],[10,2.2391],[11,2.5487],[12,2.6616],[13,2.429],[14,1.8715],[15,1.2312],[16,0.7487],[17,0.4069],[18,0.1154],[19,-0.0848],[20,-0.1963],[21,-0.2267],[22,-0.1864],[23,-0.1321],[24,-0.1032],[25,-0.0998],[26,-0.118],[27,-0.1515],[28,-0.196],[29,-0.2383],[30,-0.2572]]
       }
       chest_ctrl.Location.Y {
        type=MovieSceneScriptingFloatChannel
        default=0.3666
        keys [[0,0.3666],[1,0.1768],[2,0.3535],[3,0.6465],[4,0.776],[5,0.3842],[6,-0.6057],[7,-1.4213],[8,-1.3643],[9,-0.7651],[10,-0.2694],[11,0.1273],[12,0.5026],[13,0.7943],[14,0.8258],[15,0.7285],[16,0.6747],[17,0.7303],[18,0.7949],[19,0.8313],[20,0.8739],[21,0.9464],[22,0.996],[23,1.0503],[24,1.1465],[25,1.2485],[26,1.3211],[27,1.3292],[28,1.3021],[29,1.289],[30,1.2857]]
       }
       chest_ctrl.Location.Z {
        type=MovieSceneScriptingFloatChannel
        default=0.019
        thinned_from=31
        keys [[0,0.019],[1,0.0177],[2,0.0314],[3,0.0236],[4,-0.0006],[5,-0.0318],[6,-0.1439],[7,-0.311],[8,-0.2723],[9,-0.1625],[10,-0.1344],[11,-0.1369],[12,-0.1259],[13,-0.0778],[14,-0.0059],[15,0.0492],[16,0.0743],[17,0.0833],[18,0.083],[19,0.0786],[20,0.0746],[21,0.0725],[22,0.0718],[23,0.0721],[25,0.0747],[26,0.0753],[27,0.075],[29,0.0736],[30,0.0733]]
       }
       chest_ctrl.Rotation.X {
        type=MovieSceneScriptingFloatChannel
        default=3.2929
        keys [[0,3.2929],[1,-0.133],[2,-1.31],[3,-1.6658],[4,-2.2567],[5,-5.4754],[6,-13.3623],[7,-18.2768],[8,-14.3288],[9,-6.446],[10,0.4317],[11,5.6691],[12,8.6972],[13,8.9657],[14,5.7199],[15,1.9172],[16,-0.0555],[17,0.0338],[18,0.728],[19,1.4196],[20,2.3646],[21,3.7086],[22,4.6434],[23,5.4463],[24,6.6419],[25,7.8644],[26,8.742],[27,8.896],[28,8.6543],[29,8.5578],[30,8.5405]]
       }
       chest_ctrl.Rotation.Y {
        type=MovieSceneScriptingFloatChannel
        default=8.2681
        keys [[0,8.2681],[1,5.586],[2,-1.3822],[3,-8.5879],[4,-12.8457],[5,-12.2839],[6,-14.3043],[7,-23.0432],[8,-23.3505],[9,-20.3506],[10,-20.8467],[11,-22.0824],[12,-21.5606],[13,-18.3993],[14,-13.4161],[15,-7.8474],[16,-3.5167],[17,-0.5648],[18,1.623],[19,2.9196],[20,3.5454],[21,3.4039],[22,2.5883],[23,1.709],[24,1.1115],[25,0.8163],[26,0.8218],[27,1.1056],[28,1.5773],[29,2.0234],[30,2.2215]]
       }
       chest_ctrl.Rotation.Z {
        type=MovieSceneScriptingFloatChannel
        default=19.3797
        thinned_from=31
        keys [[0,19.3797],[1,20.4881],[3,23.8649],[4,23.5598],[5,19.6852],[6,8.3686],[7,-3.3287],[8,-13.8991],[9,-22.9585],[10,-28.4209],[11,-30.9882],[12,-29.9545],[13,-26.5686],[14,-21.8765],[15,-17.2553],[16,-13.7275],[17,-11.157],[18,-8.7896],[19,-6.3783],[21,-0.885],[22,1.0817],[23,2.2742],[24,3.2022],[25,3.8956],[26,4.3858],[27,4.6937],[28,4.8547],[29,4.9245],[30,4.9398]]
       }
       chest_ctrl.Scale.X {
        type=MovieSceneScriptingFloatChannel
        default=0.999
        thinned_from=31
        keys [[0,0.999],[5,0.9955],[6,0.9923],[7,0.9856],[8,0.987],[9,0.9909],[12,0.9894],[30,0.9984]]
       }
       chest_ctrl.Scale.Y {type=MovieSceneScriptingFloatChannel, default=1, const [1,31]}
       chest_ctrl.Scale.Z {type=MovieSceneScriptingFloatChannel, default=1, const [1,31]}
       chest_tan_ctrl.Location.X {
        type=MovieSceneScriptingFloatChannel
        default=-0.4709
        keys [[0,-0.4709],[1,-0.2985],[2,0.1622],[3,0.6308],[4,0.8788],[5,0.7687],[6,0.7909],[7,1.29],[8,1.2358],[9,0.9873],[10,1.0225],[11,1.1355],[12,1.1168],[13,0.9042],[14,0.571],[15,0.2045],[16,-0.0681],[17,-0.2261],[18,-0.3145],[19,-0.3463],[20,-0.346],[21,-0.2996],[22,-0.2159],[23,-0.1358],[24,-0.0792],[25,-0.0477],[26,-0.041],[27,-0.0578],[28,-0.0898],[29,-0.1204],[30,-0.1341]]
       }
       chest_tan_ctrl.Location.Y {
        type=MovieSceneScriptingFloatChannel
        default=-0.014
        keys [[0,-0.014],[1,-0.2712],[2,-0.3713],[3,-0.4061],[4,-0.4431],[5,-0.6516],[6,-1.1946],[7,-1.5379],[8,-1.3689],[9,-0.9541],[10,-0.5742],[11,-0.2918],[12,-0.1131],[13,-0.0741],[14,-0.2543],[15,-0.4745],[16,-0.5771],[17,-0.5357],[18,-0.4463],[19,-0.362],[20,-0.2644],[21,-0.1422],[22,-0.0574],[23,0.0108],[24,0.1041],[25,0.197],[26,0.2631],[27,0.2758],[28,0.259],[29,0.2523],[30,0.251]]
       }
       chest_tan_ctrl.Location.Z {
        type=MovieSceneScriptingFloatChannel
        default=0.2698
        thinned_from=31
        keys [[0,0.2698],[1,0.2408],[2,0.2285],[3,0.2364],[4,0.2453],[5,0.2265],[6,0.2099],[7,0.2444],[8,0.2398],[9,0.2277],[10,0.2467],[11,0.2729],[12,0.2846],[13,0.271],[14,0.2347],[15,0.206],[16,0.1977],[17,0.2035],[18,0.2138],[20,0.2326],[21,0.2438],[22,0.2514],[23,0.2578],[24,0.2676],[25,0.2781],[26,0.286],[27,0.2877],[28,0.2859],[30,0.2854]]
       }
       chest_tan_ctrl.Rotation.X {
        type=MovieSceneScriptingFloatChannel
        default=0.2707
        thinned_from=31
        keys [[0,0.2707],[1,0],[2,1.8237],[3,2.742],[4,2.0258],[5,0],[6,0],[7,2.7305],[8,10.0947],[9,12.3335],[10,9.9283],[11,7.3294],[12,5.4308],[13,3.3024],[14,0.9763],[15,1.0899],[16,2.6185],[17,3.8311],[18,3.5871],[19,2.8635],[20,2.3672],[21,1.907],[22,0.9808],[23,0.4425],[24,0.2049],[25,0],[30,0]]
       }
       chest_tan_ctrl.Rotation.Y {
        type=MovieSceneScriptingFloatChannel
        default=0
        thinned_from=31
        keys [[0,0],[3,0],[4,1.48],[5,5.4683],[6,4.0142],[7,2.8387],[8,10.5985],[9,10.7238],[10,6.4323],[11,6.3833],[12,8.4904],[13,10.2741],[14,10.144],[15,7.8284],[16,4.1922],[17,1.485],[18,0.1022],[19,0],[30,0]]
       }
       chest_tan_ctrl.Rotation.Z {
        type=MovieSceneScriptingFloatChannel
        default=0
        thinned_from=31
        keys [[0,0],[8,0],[9,1.2924],[10,5.004],[11,8.1445],[12,10.6992],[13,11.16],[14,10.6867],[15,9.6957],[17,6.3276],[18,5.3235],[19,4.5637],[20,3.9969],[21,3.1039],[22,1.725],[23,0.5574],[24,0],[30,0]]
       }
       chest_tan_ctrl.Scale.X {type=MovieSceneScriptingFloatChannel, default=1, thinned_from=31, keys [[0,1],[4,1.0016],[5,1.0005],[7,1.0067],[8,1],[30,1]]}
       chest_tan_ctrl.Scale.Y {type=MovieSceneScriptingFloatChannel, default=1, const [1,31]}
       chest_tan_ctrl.Scale.Z {type=MovieSceneScriptingFloatChannel, default=1, const [1,31]}
      ]
     }]
   }
  ]
 }]
display_rate {denominator=1, numerator=30}
tick_resolution {denominator=1, numerator=24000}
tracks []
